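# Physics asset of collision shapes for the cape of the F_MED_OutlineLove_Flank female player body cosmetic.
# Capsules cover the pelvis, spine, head, clavicles, arms and feet, while thigh, calf and neck_02 bodies are empty.
# Left and right dyn_cape and dyn_skirt bone chains carry wide capsules offset from their bones, repeated identically or mirrored along each chain.
# It has no constraints, so it defines collision volumes only, with no joints.
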PhysicsAsset /BRCosmetics/Characters/Player/Female/Medium/Bodies/F_MED_OutlineLove_Flank/Meshes/F_MED_OutlineLove_Cape_CapeColliders
bodies {
 calf_l []
 calf_r []
 clavicle_l [{shapeType=Capsule, length=6.5623, radius=7.5194, center=<3.9566 -1.0862 -2.1127>, rotation=<rot -77.5566 177.3915 7.7314>}]
 clavicle_r [{shapeType=Capsule, length=6.5623, radius=7.5194, center=<-3.9566 1.0862 2.1127>, rotation=<rot -77.5566 177.3915 -172.2686>}]
 dyn_cape_l_00 [{shapeType=Capsule, length=5.1179, radius=12.6668, shapeName=dyn_cape_r_00_capsule, center=<11.0323 -10.7078 0>, rotation=<rot 90 0 0>}]
 dyn_cape_l_01 [{shapeType=Capsule, length=5.1179, radius=12.6668, shapeName=dyn_cape_r_01_capsule, center=<11.0323 -10.7156 0>, rotation=<rot 90 0 0>}]
 dyn_cape_l_02 [{shapeType=Capsule, length=5.1179, radius=12.6668, shapeName=dyn_cape_r_02_capsule, center=<11.0323 -10.7156 0>, rotation=<rot 90 0 0>}]
 dyn_cape_l_03 [{shapeType=Capsule, length=5.1179, radius=12.6668, shapeName=dyn_cape_r_03_capsule, center=<11.0323 -10.7156 0>, rotation=<rot 90 0 0>}]
 dyn_cape_l_04 [{shapeType=Capsule, length=5.1179, radius=12.6668, shapeName=dyn_cape_r_04_capsule, center=<11.0323 -10.7156 0>, rotation=<rot 90 0 0>}]
 dyn_cape_l_05 [{shapeType=Capsule, length=5.1179, radius=12.6668, shapeName=dyn_cape_r_05_capsule, center=<11.0323 -10.7156 0>, rotation=<rot 90 0 0>}]
 dyn_cape_r_00 [{shapeType=Capsule, length=5.1179, radius=12.6668, shapeName=dyn_cape_r_00_capsule, center=<11.0323 -10.7156 0>, rotation=<rot 90 0 0>}]
 dyn_cape_r_01 [{shapeType=Capsule, length=5.1179, radius=12.6668, shapeName=dyn_cape_r_01_capsule, center=<11.0323 -10.7156 0>, rotation=<rot 90 0 0>}]
 dyn_cape_r_02 [{shapeType=Capsule, length=5.1179, radius=12.6668, shapeName=dyn_cape_r_02_capsule, center=<11.0323 -10.7156 0>, rotation=<rot 90 0 0>}]
 dyn_cape_r_03 [{shapeType=Capsule, length=5.1179, radius=12.6668, shapeName=dyn_cape_r_03_capsule, center=<11.0323 -10.7156 0>, rotation=<rot 90 0 0>}]
 dyn_cape_r_04 [{shapeType=Capsule, length=5.1179, radius=12.6668, shapeName=dyn_cape_r_04_capsule, center=<11.0323 -10.7156 0>, rotation=<rot 90 0 0>}]
 dyn_cape_r_05 [{shapeType=Capsule, length=5.1179, radius=12.6668, shapeName=dyn_cape_r_05_capsule, center=<11.0323 -10.7156 0>, rotation=<rot 90 0 0>}]
 dyn_skirt_l_2_01 [{shapeType=Capsule, length=1.5334, radius=14.4666, shapeName=dyn_skirt_r_2_05_capsule, center=<3 13.0577 1.0772>, rotation=<rot 90 0 0>}]
 dyn_skirt_l_2_02 [{shapeType=Capsule, length=1.5334, radius=14.4666, shapeName=dyn_skirt_r_2_05_capsule, center=<3 13.0577 1.0772>, rotation=<rot 90 0 0>}]
 dyn_skirt_l_2_03 [{shapeType=Capsule, length=1.5334, radius=14.4666, shapeName=dyn_skirt_r_2_05_capsule, center=<3 13.0577 1.0772>, rotation=<rot 90 0 0>}]
 dyn_skirt_l_2_04 [{shapeType=Capsule, length=1.5334, radius=14.4666, shapeName=dyn_skirt_r_2_05_capsule, center=<3 13.0577 1.0772>, rotation=<rot 90 0 0>}]
 dyn_skirt_l_2_05 [{shapeType=Capsule, length=1.5334, radius=14.4666, shapeName=dyn_skirt_r_2_05_capsule, center=<3 13.0577 1.0772>, rotation=<rot 90 0 0>}]
 dyn_skirt_l_3_01 [{shapeType=Capsule, length=1.6574, radius=14.3426, shapeName=dyn_skirt_r_3_05_capsule, center=<2 13.9704 2.1031>, rotation=<rot 90 0 0>}]
 dyn_skirt_l_3_02 [{shapeType=Capsule, length=1.6574, radius=14.3426, shapeName=dyn_skirt_r_3_05_capsule, center=<2 13.9704 2.1031>, rotation=<rot 90 0 0>}]
 dyn_skirt_l_3_03 [{shapeType=Capsule, length=1.6574, radius=14.3426, shapeName=dyn_skirt_r_3_05_capsule, center=<2 13.9704 2.1031>, rotation=<rot 90 0 0>}]
 dyn_skirt_l_3_04 [{shapeType=Capsule, length=1.6574, radius=14.3426, shapeName=dyn_skirt_r_3_05_capsule, center=<2 13.9704 2.1031>, rotation=<rot 90 0 0>}]
 dyn_skirt_l_3_05 [{shapeType=Capsule, length=1.6574, radius=14.3426, shapeName=dyn_skirt_r_3_05_capsule, center=<2 13.9704 2.1031>, rotation=<rot 90 0 0>}]
 dyn_skirt_l_4_01 [{shapeType=Capsule, length=1.3225, radius=15.6204, shapeName=dyn_skirt_r_4_05_capsule, center=<2 13.9374 -1.103>, rotation=<rot 90 0 0>}]
 dyn_skirt_l_4_02 [{shapeType=Capsule, length=1.3225, radius=15.6204, shapeName=dyn_skirt_r_4_05_capsule, center=<2 13.9374 -1.103>, rotation=<rot 90 0 0>}]
 dyn_skirt_l_4_03 [{shapeType=Capsule, length=1.3225, radius=15.6204, shapeName=dyn_skirt_r_4_05_capsule, center=<2 13.9374 -1.103>, rotation=<rot 90 0 0>}]
 dyn_skirt_l_4_04 [{shapeType=Capsule, length=1.3225, radius=15.6204, shapeName=dyn_skirt_r_4_05_capsule, center=<2 13.9374 -1.103>, rotation=<rot 90 0 0>}]
 dyn_skirt_l_4_05 [{shapeType=Capsule, length=1.3225, radius=15.6204, shapeName=dyn_skirt_r_4_05_capsule, center=<2 13.9374 -1.103>, rotation=<rot 90 0 0>}]
 dyn_skirt_r_2_01 [{shapeType=Capsule, length=1.5334, radius=14.4666, shapeName=dyn_skirt_r_2_05_capsule, center=<3 13.0577 -1.0772>, rotation=<rot 90 0 0>}]
 dyn_skirt_r_2_02 [{shapeType=Capsule, length=1.5334, radius=14.4666, shapeName=dyn_skirt_r_2_05_capsule, center=<3 13.0577 -1.0772>, rotation=<rot 90 0 0>}]
 dyn_skirt_r_2_03 [{shapeType=Capsule, length=1.5334, radius=14.4666, shapeName=dyn_skirt_r_2_05_capsule, center=<3 13.0577 -1.0772>, rotation=<rot 90 0 0>}]
 dyn_skirt_r_2_04 [{shapeType=Capsule, length=1.5334, radius=14.4666, shapeName=dyn_skirt_r_2_05_capsule, center=<3 13.0577 -1.0772>, rotation=<rot 90 0 0>}]
 dyn_skirt_r_2_05 [{shapeType=Capsule, length=1.5334, radius=14.4666, shapeName=dyn_skirt_r_2_05_capsule, center=<3 13.0577 -1.0772>, rotation=<rot 90 0 0>}]
 dyn_skirt_r_3_01 [{shapeType=Capsule, length=1.6574, radius=14.3426, shapeName=dyn_skirt_r_3_05_capsule, center=<2 13.9704 -2.1031>, rotation=<rot 90 0 0>}]
 dyn_skirt_r_3_02 [{shapeType=Capsule, length=1.6574, radius=14.3426, shapeName=dyn_skirt_r_3_05_capsule, center=<2 13.9704 -2.1031>, rotation=<rot 90 0 0>}]
 dyn_skirt_r_3_03 [{shapeType=Capsule, length=1.6574, radius=14.3426, shapeName=dyn_skirt_r_3_05_capsule, center=<2 13.9704 -2.1031>, rotation=<rot 90 0 0>}]
 dyn_skirt_r_3_04 [{shapeType=Capsule, length=1.6574, radius=14.3426, shapeName=dyn_skirt_r_3_05_capsule, center=<2 13.9704 -2.1031>, rotation=<rot 90 0 0>}]
 dyn_skirt_r_3_05 [{shapeType=Capsule, length=1.6574, radius=14.3426, shapeName=dyn_skirt_r_3_05_capsule, center=<2 13.9704 -2.1031>, rotation=<rot 90 0 0>}]
 dyn_skirt_r_4_01 [{shapeType=Capsule, length=1.3225, radius=15.6204, shapeName=dyn_skirt_r_4_05_capsule, center=<2 13.9374 1.103>, rotation=<rot 90 0 0>}]
 dyn_skirt_r_4_02 [{shapeType=Capsule, length=1.3225, radius=15.6204, shapeName=dyn_skirt_r_4_05_capsule, center=<2 13.9374 1.103>, rotation=<rot 90 0 0>}]
 dyn_skirt_r_4_03 [{shapeType=Capsule, length=1.3225, radius=15.6204, shapeName=dyn_skirt_r_4_05_capsule, center=<2 13.9374 1.103>, rotation=<rot 90 0 0>}]
 dyn_skirt_r_4_04 [{shapeType=Capsule, length=1.3225, radius=15.6204, shapeName=dyn_skirt_r_4_05_capsule, center=<2 13.9374 1.103>, rotation=<rot 90 0 0>}]
 dyn_skirt_r_4_05 [{shapeType=Capsule, length=1.3225, radius=15.6204, shapeName=dyn_skirt_r_4_05_capsule, center=<2 13.9374 1.103>, rotation=<rot 90 0 0>}]
 foot_l [{shapeType=Capsule, length=11.0765, radius=8.0344, center=<-4.1811 6.0027 -0.338>, rotation=<rot 0.0388 1.4701 -89.3982>}]
 foot_r [{shapeType=Capsule, length=11.0765, radius=8.0344, center=<4.1811 -6.0027 0.338>, rotation=<rot 0.0388 1.4701 90.6018>}]
 head [{shapeType=Capsule, length=0, radius=18, center=<9.0697 0.0746 0>, rotation=<rot 0 -90 0>}]
 lowerarm_l [{shapeType=Capsule, length=20.8668, radius=5.1988, shapeName=lowerarm_l_capsule, center=<10.5647 0 0>, rotation=<rot 90 0 0>}]
 lowerarm_r [{shapeType=Capsule, length=21.7, radius=4.3, shapeName=lowerarm_l_capsule, center=<-10.5647 0 0>, rotation=<rot 90 0 0>}]
 neck_02 []
 pelvis [{
   shapeType=Capsule
   length=8.7071
   radius=10.4491
   shapeName=pelvis_capsule
   center=<-0.3945 -1.4285 0.0224>
   rotation=<rot 4.5484 -90 1.2156>
  }]
 spine_01 [{shapeType=Capsule, length=9.2863, radius=7.6878, shapeName=spine_01_capsule, center=<3.2393 0.1393 0>}]
 spine_02 [{shapeType=Capsule, length=9.7561, radius=7.4892, shapeName=spine_02_capsule, center=<3.0268 0.0383 -0.2238>}]
 spine_03 [{shapeType=Capsule, length=11.8086, radius=7.1416, shapeName=spine_03_capsule, center=<2.9926 -0.6577 0>, rotation=<rot 0 -90 0>}]
 spine_04 [{shapeType=Capsule, length=14.3512, radius=7.3397, shapeName=spine_04_capsule, center=<2.7688 -1.2033 0>}]
 spine_05 [{shapeType=Capsule, length=9.678, radius=12.3571, center=<0.6969 1.91 0>}]
 thigh_l []
 thigh_r []
 upperarm_l [{shapeType=Capsule, length=24.23, radius=3.99, shapeName=upperarm_l_capsule, center=<14.2986 -0.5368 -0.3564>, rotation=<rot 90 0 0>}]
 upperarm_r [{shapeType=Capsule, length=24.4, radius=3.8, shapeName=upperarm_l_capsule, center=<-14.2986 0.5368 0.3564>, rotation=<rot 90 0 0>}]
}
constraints []
