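# Physics asset for the cape and skirt dynamic bones of the F_MED_OutlineLove_Flank female cosmetic body.
# Left and right six-bone cape chains hang from spine_04 with limited swing and very tight twist.
# Six five-bone skirt chains hang from the pelvis with free angular motion, so their degree limits are inactive.
# Pelvis, spine, arm and hand capsules provide body shapes, plus an oversized ground_mover capsule acting as a floor.
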PhysicsAsset /BRCosmetics/Characters/Player/Female/Medium/Bodies/F_MED_OutlineLove_Flank/Meshes/F_MED_OutlineLove_Flank_Cape_Physics
bodies {
 dyn_cape_l_00 [{shapeType=Capsule, length=5.1179, radius=12.6668, shapeName=dyn_cape_r_00_capsule, center=<11.0323 10.7577 0>}]
 dyn_cape_l_01 [{shapeType=Capsule, length=5.1179, radius=12.6668, shapeName=dyn_cape_r_01_capsule, center=<11.0323 10.75 0>}]
 dyn_cape_l_02 [{shapeType=Capsule, length=5.1179, radius=12.6668, shapeName=dyn_cape_r_02_capsule, center=<11.0323 10.75 0>}]
 dyn_cape_l_03 [{shapeType=Capsule, length=5.1179, radius=12.6668, shapeName=dyn_cape_r_03_capsule, center=<11.0323 10.75 0>}]
 dyn_cape_l_04 [{shapeType=Capsule, length=13.2655, radius=12.6668, shapeName=dyn_cape_r_04_capsule, center=<11.0323 10.75 0>}]
 dyn_cape_l_05 [{shapeType=Capsule, length=12.3901, radius=12.6668, shapeName=dyn_cape_r_05_capsule, center=<11.0323 10.75 0>}]
 dyn_cape_r_00 [{shapeType=Capsule, length=5.1179, radius=12.6668, shapeName=dyn_cape_r_00_capsule, center=<11.0323 10.7577 0>}]
 dyn_cape_r_01 [{shapeType=Capsule, length=5.1179, radius=12.6668, shapeName=dyn_cape_r_01_capsule, center=<11.0323 10.75 0>}]
 dyn_cape_r_02 [{shapeType=Capsule, length=5.1179, radius=12.6668, shapeName=dyn_cape_r_02_capsule, center=<11.0323 10.75 0>}]
 dyn_cape_r_03 [{shapeType=Capsule, length=5.1179, radius=12.6668, shapeName=dyn_cape_r_03_capsule, center=<11.0323 10.75 0>}]
 dyn_cape_r_04 [{shapeType=Capsule, length=13.2655, radius=12.6668, shapeName=dyn_cape_r_04_capsule, center=<11.0323 10.75 0>}]
 dyn_cape_r_05 [{shapeType=Capsule, length=12.3901, radius=12.6668, shapeName=dyn_cape_r_05_capsule, center=<11.0323 10.75 0>}]
 dyn_skirt_l_2_01 [{shapeType=Capsule, length=1.5334, radius=14.4666, shapeName=dyn_skirt_r_2_05_capsule, center=<3 13.0577 1.0772>, rotation=<rot 90 0 0>}]
 dyn_skirt_l_2_02 [{shapeType=Capsule, length=1.5334, radius=14.4666, shapeName=dyn_skirt_r_2_05_capsule, center=<3 13.0577 1.0772>, rotation=<rot 90 0 0>}]
 dyn_skirt_l_2_03 [{shapeType=Capsule, length=1.5334, radius=14.4666, shapeName=dyn_skirt_r_2_05_capsule, center=<3 13.0577 1.0772>, rotation=<rot 90 0 0>}]
 dyn_skirt_l_2_04 [{shapeType=Capsule, length=1.5334, radius=14.4666, shapeName=dyn_skirt_r_2_05_capsule, center=<3 13.0577 1.0772>, rotation=<rot 90 0 0>}]
 dyn_skirt_l_2_05 [{shapeType=Capsule, length=1.5334, radius=14.4666, shapeName=dyn_skirt_r_2_05_capsule, center=<3 13.0577 1.0772>, rotation=<rot 90 0 0>}]
 dyn_skirt_l_3_01 [{shapeType=Capsule, length=1.6574, radius=14.3426, shapeName=dyn_skirt_r_3_05_capsule, center=<2 13.9704 2.1031>, rotation=<rot 90 0 0>}]
 dyn_skirt_l_3_02 [{shapeType=Capsule, length=1.6574, radius=14.3426, shapeName=dyn_skirt_r_3_05_capsule, center=<2 13.9704 2.1031>, rotation=<rot 90 0 0>}]
 dyn_skirt_l_3_03 [{shapeType=Capsule, length=1.6574, radius=14.3426, shapeName=dyn_skirt_r_3_05_capsule, center=<2 13.9704 2.1031>, rotation=<rot 90 0 0>}]
 dyn_skirt_l_3_04 [{shapeType=Capsule, length=1.6574, radius=14.3426, shapeName=dyn_skirt_r_3_05_capsule, center=<2 13.9704 2.1031>, rotation=<rot 90 0 0>}]
 dyn_skirt_l_3_05 [{shapeType=Capsule, length=1.6574, radius=14.3426, shapeName=dyn_skirt_r_3_05_capsule, center=<2 13.9704 2.1031>, rotation=<rot 90 0 0>}]
 dyn_skirt_l_4_01 [{shapeType=Capsule, length=1.3225, radius=15.6204, shapeName=dyn_skirt_r_4_05_capsule, center=<2 13.9374 -1.103>, rotation=<rot 90 0 0>}]
 dyn_skirt_l_4_02 [{shapeType=Capsule, length=1.3225, radius=15.6204, shapeName=dyn_skirt_r_4_05_capsule, center=<2 13.9374 -1.103>, rotation=<rot 90 0 0>}]
 dyn_skirt_l_4_03 [{shapeType=Capsule, length=1.3225, radius=15.6204, shapeName=dyn_skirt_r_4_05_capsule, center=<2 13.9374 -1.103>, rotation=<rot 90 0 0>}]
 dyn_skirt_l_4_04 [{shapeType=Capsule, length=1.3225, radius=15.6204, shapeName=dyn_skirt_r_4_05_capsule, center=<2 13.9374 -1.103>, rotation=<rot 90 0 0>}]
 dyn_skirt_l_4_05 [{shapeType=Capsule, length=1.3225, radius=15.6204, shapeName=dyn_skirt_r_4_05_capsule, center=<2 13.9374 -1.103>, rotation=<rot 90 0 0>}]
 dyn_skirt_r_2_01 [{shapeType=Capsule, length=1.5334, radius=14.4666, shapeName=dyn_skirt_r_2_05_capsule, center=<3 13.0577 -1.0772>, rotation=<rot 90 0 0>}]
 dyn_skirt_r_2_02 [{shapeType=Capsule, length=1.5334, radius=14.4666, shapeName=dyn_skirt_r_2_05_capsule, center=<3 13.0577 -1.0772>, rotation=<rot 90 0 0>}]
 dyn_skirt_r_2_03 [{shapeType=Capsule, length=1.5334, radius=14.4666, shapeName=dyn_skirt_r_2_05_capsule, center=<3 13.0577 -1.0772>, rotation=<rot 90 0 0>}]
 dyn_skirt_r_2_04 [{shapeType=Capsule, length=1.5334, radius=14.4666, shapeName=dyn_skirt_r_2_05_capsule, center=<3 13.0577 -1.0772>, rotation=<rot 90 0 0>}]
 dyn_skirt_r_2_05 [{shapeType=Capsule, length=1.5334, radius=14.4666, shapeName=dyn_skirt_r_2_05_capsule, center=<3 13.0577 -1.0772>, rotation=<rot 90 0 0>}]
 dyn_skirt_r_3_01 [{shapeType=Capsule, length=1.6574, radius=14.3426, shapeName=dyn_skirt_r_3_05_capsule, center=<2 13.9704 -2.1031>, rotation=<rot 90 0 0>}]
 dyn_skirt_r_3_02 [{shapeType=Capsule, length=1.6574, radius=14.3426, shapeName=dyn_skirt_r_3_05_capsule, center=<2 13.9704 -2.1031>, rotation=<rot 90 0 0>}]
 dyn_skirt_r_3_03 [{shapeType=Capsule, length=1.6574, radius=14.3426, shapeName=dyn_skirt_r_3_05_capsule, center=<2 13.9704 -2.1031>, rotation=<rot 90 0 0>}]
 dyn_skirt_r_3_04 [{shapeType=Capsule, length=1.6574, radius=14.3426, shapeName=dyn_skirt_r_3_05_capsule, center=<2 13.9704 -2.1031>, rotation=<rot 90 0 0>}]
 dyn_skirt_r_3_05 [{shapeType=Capsule, length=1.6574, radius=14.3426, shapeName=dyn_skirt_r_3_05_capsule, center=<2 13.9704 -2.1031>, rotation=<rot 90 0 0>}]
 dyn_skirt_r_4_01 [{shapeType=Capsule, length=1.3225, radius=15.6204, shapeName=dyn_skirt_r_4_05_capsule, center=<2 13.9374 1.103>, rotation=<rot 90 0 0>}]
 dyn_skirt_r_4_02 [{shapeType=Capsule, length=1.3225, radius=15.6204, shapeName=dyn_skirt_r_4_05_capsule, center=<2 13.9374 1.103>, rotation=<rot 90 0 0>}]
 dyn_skirt_r_4_03 [{shapeType=Capsule, length=1.3225, radius=15.6204, shapeName=dyn_skirt_r_4_05_capsule, center=<2 13.9374 1.103>, rotation=<rot 90 0 0>}]
 dyn_skirt_r_4_04 [{shapeType=Capsule, length=1.3225, radius=15.6204, shapeName=dyn_skirt_r_4_05_capsule, center=<2 13.9374 1.103>, rotation=<rot 90 0 0>}]
 dyn_skirt_r_4_05 [{shapeType=Capsule, length=1.3225, radius=15.6204, shapeName=dyn_skirt_r_4_05_capsule, center=<2 13.9374 1.103>, rotation=<rot 90 0 0>}]
 ground_mover [{shapeType=Capsule, length=0.5056, radius=337.6725, shapeName=floor_joint_capsule, center=<2.6954 -3.0549 -337.3088>, rotation=<rot 0 90 -90>}]
 hand_l [{
   shapeType=Capsule
   length=4.8886
   radius=7.2965
   shapeName=hand_r_capsule
   center=<9.0987 -1.803 -0.5956>
   rotation=<rot 2.0987 -90 -86.1237>
  }]
 hand_r [{shapeType=Capsule, length=4.8886, radius=7.2965, shapeName=hand_r_capsule, center=<-9.0987 1.803 0.5956>, rotation=<rot 2.0987 -90 -86.1237>}]
 lowerarm_l [{
   shapeType=Capsule
   length=20.9587
   radius=5.0413
   shapeName=lowerarm_l_capsule
   center=<10.5622 -0.1765 -0.1183>
   rotation=<rot 88.8121 -0.0026 -0.0026>
  }]
 lowerarm_r [{
   shapeType=Capsule
   length=20.9587
   radius=5.0413
   shapeName=lowerarm_l_capsule
   center=<-10.5622 0.1765 0.1183>
   rotation=<rot 88.8121 -0.0026 -0.0026>
  }]
 pelvis [{
   shapeType=Capsule
   length=8.7071
   radius=10.4491
   shapeName=pelvis_capsule
   center=<-0.3945 -1.4285 0.0224>
   rotation=<rot 4.5484 -90 1.2156>
  }]
 spine_01 [{shapeType=Capsule, length=9.2863, radius=7.6878, shapeName=spine_01_capsule, center=<3.2393 0.1393 0>}]
 spine_02 [{shapeType=Capsule, length=9.7561, radius=7.4892, shapeName=spine_02_capsule, center=<3.0268 0.0383 -0.2238>}]
 spine_03 [{shapeType=Capsule, length=11.8086, radius=7.1416, shapeName=spine_03_capsule, center=<2.9926 -0.6577 0>, rotation=<rot 0 -90 0>}]
 spine_04 [{shapeType=Capsule, length=14.3512, radius=7.3397, shapeName=spine_04_capsule, center=<2.7688 -1.2033 0>}]
 upperarm_l [{shapeType=Capsule, length=24.23, radius=3.99, shapeName=upperarm_l_capsule, center=<14.2986 -0.5368 -0.3564>, rotation=<rot 90 0 0>}]
 upperarm_r [{shapeType=Capsule, length=24.4, radius=3.8, shapeName=upperarm_l_capsule, center=<-14.2986 0.5368 0.3564>, rotation=<rot 90 0 0>}]
}
constraints [
 {bone1Name=dyn_cape_l_00, bone2Name=spine_04, swing1LimitDegrees=25, swing2LimitDegrees=25, twistLimitDegrees=1}
 {bone1Name=dyn_cape_l_01, bone2Name=dyn_cape_l_00, swing1LimitDegrees=25, swing2LimitDegrees=25, twistLimitDegrees=1}
 {bone1Name=dyn_cape_l_02, bone2Name=dyn_cape_l_01, swing1LimitDegrees=25, swing2LimitDegrees=25, twistLimitDegrees=1}
 {bone1Name=dyn_cape_l_03, bone2Name=dyn_cape_l_02, swing1LimitDegrees=25, swing2LimitDegrees=25, twistLimitDegrees=1}
 {bone1Name=dyn_cape_l_04, bone2Name=dyn_cape_l_03, swing1LimitDegrees=25, swing2LimitDegrees=25, twistLimitDegrees=1}
 {bone1Name=dyn_cape_l_05, bone2Name=dyn_cape_l_04, swing1LimitDegrees=25, swing2LimitDegrees=25, twistLimitDegrees=1}
 {bone1Name=dyn_cape_r_00, bone2Name=spine_04, swing1LimitDegrees=25, swing2LimitDegrees=25, twistLimitDegrees=1}
 {bone1Name=dyn_cape_r_01, bone2Name=dyn_cape_r_00, swing1LimitDegrees=25, swing2LimitDegrees=25, twistLimitDegrees=1}
 {bone1Name=dyn_cape_r_02, bone2Name=dyn_cape_r_01, swing1LimitDegrees=25, swing2LimitDegrees=25, twistLimitDegrees=1}
 {bone1Name=dyn_cape_r_03, bone2Name=dyn_cape_r_02, swing1LimitDegrees=25, swing2LimitDegrees=25, twistLimitDegrees=1}
 {bone1Name=dyn_cape_r_04, bone2Name=dyn_cape_r_03, swing1LimitDegrees=25, swing2LimitDegrees=25, twistLimitDegrees=1}
 {bone1Name=dyn_cape_r_05, bone2Name=dyn_cape_r_04, swing1LimitDegrees=25, swing2LimitDegrees=25, twistLimitDegrees=1}
 {
  bone1Name=dyn_skirt_r_3_05
  bone2Name=dyn_skirt_r_3_04
  swing1LimitDegrees=45
  swing1Motion=Free
  swing2LimitDegrees=45
  swing2Motion=Free
  twistLimitDegrees=45
  twistMotion=Free
 }
 {
  bone1Name=dyn_skirt_l_2_02
  bone2Name=dyn_skirt_l_2_01
  swing1LimitDegrees=45
  swing1Motion=Free
  swing2LimitDegrees=45
  swing2Motion=Free
  twistLimitDegrees=45
  twistMotion=Free
 }
 {
  bone1Name=dyn_skirt_l_2_03
  bone2Name=dyn_skirt_l_2_02
  swing1LimitDegrees=45
  swing1Motion=Free
  swing2LimitDegrees=45
  swing2Motion=Free
  twistLimitDegrees=45
  twistMotion=Free
 }
 {
  bone1Name=dyn_skirt_l_2_04
  bone2Name=dyn_skirt_l_2_03
  swing1LimitDegrees=45
  swing1Motion=Free
  swing2LimitDegrees=45
  swing2Motion=Free
  twistLimitDegrees=45
  twistMotion=Free
 }
 {
  bone1Name=dyn_skirt_l_2_05
  bone2Name=dyn_skirt_l_2_04
  swing1LimitDegrees=45
  swing1Motion=Free
  swing2LimitDegrees=45
  swing2Motion=Free
  twistLimitDegrees=45
  twistMotion=Free
 }
 {
  bone1Name=dyn_skirt_l_3_02
  bone2Name=dyn_skirt_l_3_01
  swing1LimitDegrees=45
  swing1Motion=Free
  swing2LimitDegrees=45
  swing2Motion=Free
  twistLimitDegrees=45
  twistMotion=Free
 }
 {
  bone1Name=dyn_skirt_l_3_03
  bone2Name=dyn_skirt_l_3_02
  swing1LimitDegrees=45
  swing1Motion=Free
  swing2LimitDegrees=45
  swing2Motion=Free
  twistLimitDegrees=45
  twistMotion=Free
 }
 {
  bone1Name=dyn_skirt_l_3_04
  bone2Name=dyn_skirt_l_3_03
  swing1LimitDegrees=45
  swing1Motion=Free
  swing2LimitDegrees=45
  swing2Motion=Free
  twistLimitDegrees=45
  twistMotion=Free
 }
 {
  bone1Name=dyn_skirt_l_3_05
  bone2Name=dyn_skirt_l_3_04
  swing1LimitDegrees=45
  swing1Motion=Free
  swing2LimitDegrees=45
  swing2Motion=Free
  twistLimitDegrees=45
  twistMotion=Free
 }
 {
  bone1Name=dyn_skirt_r_4_05
  bone2Name=dyn_skirt_r_4_04
  swing1LimitDegrees=45
  swing1Motion=Free
  swing2LimitDegrees=45
  swing2Motion=Free
  twistLimitDegrees=45
  twistMotion=Free
 }
 {
  bone1Name=dyn_skirt_l_4_02
  bone2Name=dyn_skirt_l_4_01
  swing1LimitDegrees=45
  swing1Motion=Free
  swing2LimitDegrees=45
  swing2Motion=Free
  twistLimitDegrees=45
  twistMotion=Free
 }
 {
  bone1Name=dyn_skirt_l_4_03
  bone2Name=dyn_skirt_l_4_02
  swing1LimitDegrees=45
  swing1Motion=Free
  swing2LimitDegrees=45
  swing2Motion=Free
  twistLimitDegrees=45
  twistMotion=Free
 }
 {
  bone1Name=dyn_skirt_l_4_04
  bone2Name=dyn_skirt_l_4_03
  swing1LimitDegrees=45
  swing1Motion=Free
  swing2LimitDegrees=45
  swing2Motion=Free
  twistLimitDegrees=45
  twistMotion=Free
 }
 {
  bone1Name=dyn_skirt_l_4_05
  bone2Name=dyn_skirt_l_4_04
  swing1LimitDegrees=45
  swing1Motion=Free
  swing2LimitDegrees=45
  swing2Motion=Free
  twistLimitDegrees=45
  twistMotion=Free
 }
 {
  bone1Name=dyn_skirt_r_4_02
  bone2Name=dyn_skirt_r_4_01
  swing1LimitDegrees=45
  swing1Motion=Free
  swing2LimitDegrees=45
  swing2Motion=Free
  twistLimitDegrees=45
  twistMotion=Free
 }
 {
  bone1Name=dyn_skirt_r_4_03
  bone2Name=dyn_skirt_r_4_02
  swing1LimitDegrees=45
  swing1Motion=Free
  swing2LimitDegrees=45
  swing2Motion=Free
  twistLimitDegrees=45
  twistMotion=Free
 }
 {
  bone1Name=dyn_skirt_r_2_02
  bone2Name=dyn_skirt_r_2_01
  swing1LimitDegrees=45
  swing1Motion=Free
  swing2LimitDegrees=45
  swing2Motion=Free
  twistLimitDegrees=45
  twistMotion=Free
 }
 {
  bone1Name=dyn_skirt_r_2_03
  bone2Name=dyn_skirt_r_2_02
  swing1LimitDegrees=45
  swing1Motion=Free
  swing2LimitDegrees=45
  swing2Motion=Free
  twistLimitDegrees=45
  twistMotion=Free
 }
 {
  bone1Name=dyn_skirt_r_2_04
  bone2Name=dyn_skirt_r_2_03
  swing1LimitDegrees=45
  swing1Motion=Free
  swing2LimitDegrees=45
  swing2Motion=Free
  twistLimitDegrees=45
  twistMotion=Free
 }
 {
  bone1Name=dyn_skirt_r_2_05
  bone2Name=dyn_skirt_r_2_04
  swing1LimitDegrees=45
  swing1Motion=Free
  swing2LimitDegrees=45
  swing2Motion=Free
  twistLimitDegrees=45
  twistMotion=Free
 }
 {
  bone1Name=dyn_skirt_r_4_04
  bone2Name=dyn_skirt_r_4_03
  swing1LimitDegrees=45
  swing1Motion=Free
  swing2LimitDegrees=45
  swing2Motion=Free
  twistLimitDegrees=45
  twistMotion=Free
 }
 {
  bone1Name=dyn_skirt_r_3_04
  bone2Name=dyn_skirt_r_3_03
  swing1LimitDegrees=45
  swing1Motion=Free
  swing2LimitDegrees=45
  swing2Motion=Free
  twistLimitDegrees=45
  twistMotion=Free
 }
 {
  bone1Name=dyn_skirt_r_3_02
  bone2Name=dyn_skirt_r_3_01
  swing1LimitDegrees=45
  swing1Motion=Free
  swing2LimitDegrees=45
  swing2Motion=Free
  twistLimitDegrees=45
  twistMotion=Free
 }
 {
  bone1Name=dyn_skirt_r_3_03
  bone2Name=dyn_skirt_r_3_02
  swing1LimitDegrees=45
  swing1Motion=Free
  swing2LimitDegrees=45
  swing2Motion=Free
  twistLimitDegrees=45
  twistMotion=Free
 }
 {
  bone1Name=dyn_skirt_l_2_01
  bone2Name=pelvis
  swing1LimitDegrees=45
  swing1Motion=Free
  swing2LimitDegrees=45
  swing2Motion=Free
  twistLimitDegrees=45
  twistMotion=Free
 }
 {
  bone1Name=dyn_skirt_l_3_01
  bone2Name=pelvis
  swing1LimitDegrees=45
  swing1Motion=Free
  swing2LimitDegrees=45
  swing2Motion=Free
  twistLimitDegrees=45
  twistMotion=Free
 }
 {
  bone1Name=dyn_skirt_l_4_01
  bone2Name=pelvis
  swing1LimitDegrees=45
  swing1Motion=Free
  swing2LimitDegrees=45
  swing2Motion=Free
  twistLimitDegrees=45
  twistMotion=Free
 }
 {
  bone1Name=dyn_skirt_r_2_01
  bone2Name=pelvis
  swing1LimitDegrees=45
  swing1Motion=Free
  swing2LimitDegrees=45
  swing2Motion=Free
  twistLimitDegrees=45
  twistMotion=Free
 }
 {
  bone1Name=dyn_skirt_r_3_01
  bone2Name=pelvis
  swing1LimitDegrees=45
  swing1Motion=Free
  swing2LimitDegrees=45
  swing2Motion=Free
  twistLimitDegrees=45
  twistMotion=Free
 }
 {
  bone1Name=dyn_skirt_r_4_01
  bone2Name=pelvis
  swing1LimitDegrees=45
  swing1Motion=Free
  swing2LimitDegrees=45
  swing2Motion=Free
  twistLimitDegrees=45
  twistMotion=Free
 }
]
