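# Control Rig for the F_MED_OutlineLove_Flank female cosmetic body, correcting skirt, shoulder pad and cape bones as the body moves.
# After the shared DeformRig function, spherical pose readers on thighs and upper arms, calf bend and DeformRig curve attributes weight additive bone offsets.
# Smoothed, curve-remapped crouch, knee-slide and baseball-slide variables add skirt poses, with crouch reduced by movement and pelvis height.
# Most offsets target base_ bones that parent dyn_ chains.
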ControlRigBlueprint /BRCosmetics/Characters/Player/Female/Medium/Bodies/F_MED_OutlineLove_Flank/Meshes/F_MED_OutlineLove_Flank_CtrlRig
graphs [
 RigVMFunctionLibrary {links [], nodes []}
 RigVMModel {nodes [
   n1 = &RigUnit_PrepareForExecution()
   addFKCtrl = RigVMFunctionReferenceNode(
    Custom_FKCtrl=()
    Options_FKCtrl="(pelvis_1_#1=true,spine_3_#2=False,clavicle_5_#3=true,arm_7_#4=true,leg_9_#5=true,head_11_46DC15EF465CD297 ...(truncated)"
    Shape_Settings="(pelvis_2_#6=(bVisible=True,Name=\"Hexagon_Thin\",Color=(R=0.0117,G=0.8281,B=1,A=1),Transform=(Rotation=(X=0,Y=-0.7071,Z=0,W=0.7071),Translation=(X=2,Y=0,Z=0),Scale3D=(X=5,Y=5,Z=7))),spine_4_#7=(bVi ...(truncated)"
    ExecuteContext=@n1.ExecutePin
   )
   RigUnit_BeginExecution = &RigUnit_BeginExecution()
   setFKCtrl = RigVMFunctionReferenceNode(setFK_Ctrl=false, ExecuteContext=@RigUnit_BeginExecution.ExecutePin)
   > FN_DeformRig_FNC_1 = RigVMFunctionReferenceNode(
    DeformRig_ Curve="(deformRig_l_upperarm_up_alpha_72_#8=true,deformRig_r_upperarm_up_alpha_73_#9=true,deformRig_l_upperarm_fr_alpha_74_#10=true,deformRig_r_upperarm_fr_alpha_75_CB51B80C4 ...(truncated)"
    DeformRig_Bool="(upperarm_1_#11=True,elbow_3_#12=True,wrist_7_#13=True,groin_8_#14=True,glute_17_#15=True,knee_9_40C408324CA261 ...(truncated)"
    DeformRig_Debug=(upperarm_debug_18_#11=False,wrist_debug_20_#13=False,lat_debug_26_#16=False,groinGlute_debug_27_#14=False)
    DeformRig_Value="(upperarm_up_18_#17=(Rotation=(X=0,Y=0,Z=0,W=1),Translation=(X=0,Y=0,Z=8),Scale3D=(X=0.7,Y=1,Z=1)),upperarm_fwd_fr_19_#18=(Rotation=(X=0,Y=0,Z=0,W=1),Translation=(X=0,Y=10,Z=0),Scale3D=(X=0.5,Y=1,Z= ...(truncated)"
   )
   RigVMFunction_Sequence = &RigVMFunction_Sequence(ExecuteContext=@FN_DeformRig_FNC_1.ExecuteContext)
   n2 = &RigUnit_SphericalPoseReader(
    ActiveRegionScaleFactors=(PositiveWidth=3.5,NegativeWidth=0.5,PositiveHeight=3,NegativeHeight=3)
    ActiveRegionSize=0.1
    Debug=(bDrawDebug=false,bDraw2D=False,bDrawLocalAxes=False,DebugScale=25,DebugSegments=20,DebugThickness=0.25)
    DriverAxis=(X=1,Y=0,Z=0)
    DriverItem=(Type=Bone,Name="thigh_r")
    FalloffRegionScaleFactors=(PositiveWidth=1,NegativeWidth=1.1,PositiveHeight=1,NegativeHeight=0.9)
    FalloffSize=0.45
    FlipHeightScaling=False
    FlipWidthScaling=False
    OptionalParentItem=(Type=Bone,Name="None")
    RotationOffset=(X=90,Y=0,Z=180)
    ExecutePin=@RigVMFunction_Sequence.A
   )
   > RigVMVariableNode(Value=@n2.OutputParam) {variable=RightLeg Fwd Alpha}
   > n3 = &RigUnit_SphericalPoseReader(
    ActiveRegionScaleFactors=(PositiveWidth=3.5,NegativeWidth=0.7,PositiveHeight=3,NegativeHeight=3)
    ActiveRegionSize=0.1
    Debug=(bDrawDebug=false,bDraw2D=False,bDrawLocalAxes=False,DebugScale=25,DebugSegments=20,DebugThickness=0.25)
    DriverAxis=(X=1,Y=0,Z=0)
    DriverItem=(Type=Bone,Name="thigh_r")
    FalloffRegionScaleFactors=(PositiveWidth=1,NegativeWidth=1.15,PositiveHeight=1,NegativeHeight=1)
    FalloffSize=0.45
    FlipHeightScaling=False
    FlipWidthScaling=False
    OptionalParentItem=(Type=Bone,Name="None")
    RotationOffset=(X=0,Y=0,Z=180)
   )
   > RigVMVariableNode(Value=@n3.OutputParam) {variable=RightLeg Out Alpha}
   > n4 = &RigUnit_SphericalPoseReader(
    ActiveRegionScaleFactors=(PositiveWidth=3.5,NegativeWidth=0.5,PositiveHeight=3,NegativeHeight=3)
    ActiveRegionSize=0.1
    Debug=(bDrawDebug=false,bDraw2D=False,bDrawLocalAxes=False,DebugScale=25,DebugSegments=20,DebugThickness=0.25)
    DriverAxis=(X=1,Y=0,Z=0)
    DriverItem=(Type=Bone,Name="thigh_r")
    FalloffRegionScaleFactors=(PositiveWidth=1,NegativeWidth=1.1,PositiveHeight=1,NegativeHeight=1)
    FalloffSize=0.45
    FlipHeightScaling=False
    FlipWidthScaling=False
    OptionalParentItem=(Type=Bone,Name="None")
    RotationOffset=(X=-90,Y=0,Z=180)
   )
   > RigVMVariableNode(Value=@n4.OutputParam) {variable=RightLeg Bwd Alpha}
   > n5 = &RigUnit_SphericalPoseReader(
    ActiveRegionScaleFactors=(PositiveWidth=2,NegativeWidth=2.7,PositiveHeight=2,NegativeHeight=2)
    ActiveRegionSize=0.1
    Debug=(bDrawDebug=false,bDraw2D=False,bDrawLocalAxes=False,DebugScale=25,DebugSegments=20,DebugThickness=0.25)
    DriverAxis=(X=1,Y=0,Z=0)
    DriverItem=(Type=Bone,Name="thigh_r")
    FalloffRegionScaleFactors=(PositiveWidth=1,NegativeWidth=0.9,PositiveHeight=1,NegativeHeight=1)
    FalloffSize=0.35
    FlipHeightScaling=False
    FlipWidthScaling=False
    OptionalParentItem=(Type=Bone,Name="None")
    RotationOffset=(X=90,Y=6,Z=125)
   )
   > RigVMVariableNode(Value=@n5.OutputParam) {variable=RightLeg Fwd Extreme Alpha}
   > n6 = &RigUnit_SphericalPoseReader(
    ActiveRegionScaleFactors=(PositiveWidth=3.5,NegativeWidth=0.5,PositiveHeight=3,NegativeHeight=3)
    ActiveRegionSize=0.1
    Debug=(bDrawDebug=false,bDraw2D=False,bDrawLocalAxes=False,DebugScale=25,DebugSegments=20,DebugThickness=0.25)
    DriverAxis=(X=-1,Y=0,Z=0)
    DriverItem=(Type=Bone,Name="thigh_l")
    FalloffRegionScaleFactors=(PositiveWidth=1,NegativeWidth=1.1,PositiveHeight=1,NegativeHeight=0.9)
    FalloffSize=0.45
    FlipHeightScaling=False
    FlipWidthScaling=False
    OptionalParentItem=(Type=Bone,Name="None")
    RotationOffset=(X=90,Y=-6,Z=0)
   )
   > RigVMVariableNode(Value=@n6.OutputParam) {variable=LeftLeg Fwd Alpha}
   > n7 = &RigUnit_SphericalPoseReader(
    ActiveRegionScaleFactors=(PositiveWidth=3.5,NegativeWidth=0.7,PositiveHeight=3,NegativeHeight=3)
    ActiveRegionSize=0.1
    Debug=(bDrawDebug=false,bDraw2D=False,bDrawLocalAxes=False,DebugScale=25,DebugSegments=20,DebugThickness=0.25)
    DriverAxis=(X=-1,Y=0,Z=0)
    DriverItem=(Type=Bone,Name="thigh_l")
    FalloffRegionScaleFactors=(PositiveWidth=1,NegativeWidth=1.15,PositiveHeight=1,NegativeHeight=1.1)
    FalloffSize=0.45
    FlipHeightScaling=False
    FlipWidthScaling=False
    OptionalParentItem=(Type=Bone,Name="None")
    RotationOffset=(X=180,Y=0,Z=0)
   )
   > RigVMVariableNode(Value=@n7.OutputParam) {variable=LeftLeg Out Alpha}
   > n8 = &RigUnit_SphericalPoseReader(
    ActiveRegionScaleFactors=(PositiveWidth=3.5,NegativeWidth=0.5,PositiveHeight=3,NegativeHeight=3)
    ActiveRegionSize=0.1
    Debug=(bDrawDebug=false,bDraw2D=False,bDrawLocalAxes=False,DebugScale=25,DebugSegments=20,DebugThickness=0.25)
    DriverAxis=(X=-1,Y=0,Z=0)
    DriverItem=(Type=Bone,Name="thigh_l")
    FalloffRegionScaleFactors=(PositiveWidth=1,NegativeWidth=1.1,PositiveHeight=1,NegativeHeight=1)
    FalloffSize=0.45
    FlipHeightScaling=False
    FlipWidthScaling=False
    OptionalParentItem=(Type=Bone,Name="None")
    RotationOffset=(X=-90,Y=0,Z=0)
   )
   > RigVMVariableNode(Value=@n8.OutputParam) {variable=LeftLeg Bwd Alpha}
   > n9 = &RigUnit_SphericalPoseReader(
    ActiveRegionScaleFactors=(PositiveWidth=2,NegativeWidth=2.7,PositiveHeight=2,NegativeHeight=2)
    ActiveRegionSize=0.1
    Debug=(bDrawDebug=false,bDraw2D=False,bDrawLocalAxes=False,DebugScale=25,DebugSegments=20,DebugThickness=0.25)
    DriverAxis=(X=-1,Y=0,Z=0)
    DriverItem=(Type=Bone,Name="thigh_l")
    FalloffRegionScaleFactors=(PositiveWidth=1,NegativeWidth=0.95,PositiveHeight=1,NegativeHeight=1)
    FalloffSize=0.35
    FlipHeightScaling=False
    FlipWidthScaling=False
    OptionalParentItem=(Type=Bone,Name="None")
    RotationOffset=(X=90,Y=-6,Z=-55)
   )
   > RigVMVariableNode(Value=@n9.OutputParam) {variable=LeftLeg Fwd Extreme Alpha}
   n10 = &RigUnit_GetRelativeTransformForItem(Child=(Type=Bone,Name="calf_l"), Parent=(Type=Bone,Name="thigh_l"), bChildInitial=False, bParentInitial=False)
   > RigVMVariableNode(Value=&RigVMFunction_MathDoubleRemap(
     SourceMaximum=90
     SourceMinimum=14
     TargetMaximum=1
     TargetMinimum=0
     Value=&RigVMFunction_MathQuaternionToEuler(RotationOrder=ZXY, Value=@n10.RelativeTransform.Rotation).Result.Z
    ).Result) {variable=LeftCalf Bend Alpha}
   n11 = &RigUnit_GetRelativeTransformForItem(Child=(Type=Bone,Name="calf_r"), Parent=(Type=Bone,Name="thigh_r"), bChildInitial=False, bParentInitial=False)
   > RigVMVariableNode(Value=&RigVMFunction_MathDoubleRemap(
     SourceMaximum=90
     SourceMinimum=14
     TargetMaximum=1
     TargetMinimum=0
     Value=&RigVMFunction_MathQuaternionToEuler(RotationOrder=ZXY, Value=@n11.RelativeTransform.Rotation).Result.Z
    ).Result) {variable=RightCalf Bend Alpha}
   Maximum = &RigVMFunction_MathDoubleMax(A=RigVMVariableNode() {variable=RightLeg Fwd Alpha}.Value, B=RigVMVariableNode() {variable=LeftLeg Fwd Alpha}.Value)
   > &RigUnit_ModifyTransforms(
    ItemToModify=((Item=(Type=Bone,Name="base_skirt_f_01"),Transform=(Rotation=(X=0,Y=0,Z=-0.6191,W=0.7853),Translation=(X=0,Y=0,Z=0),Scale3D=(X=1,Y=1,Z=1))))
    WeightMinimum=0
    Weight=@Maximum.Result
   )
   > &RigUnit_ModifyTransforms(
    ItemToModify=((Item=(Type=Bone,Name="base_skirt_f_01"),Transform=(Rotation=(X=0,Y=0,Z=-0.14,W=0.9901),Translation=(X=0,Y=0,Z=0),Scale3D=(X=1,Y=1,Z=1))))
    WeightMinimum=0
    Weight=&RigVMFunction_MathDoubleMax(A=RigVMVariableNode() {variable=LeftLeg Fwd Extreme Alpha}.Value, B=RigVMVariableNode() {variable=RightLeg Fwd Extreme Alpha}.Value).Result
   )
   n12 = RigVMVariableNode() {variable=RightLeg Fwd Alpha}
   > &RigUnit_ModifyTransforms(
    ItemToModify="((Item=(Type=Bone,Name=\"base_skirt_r_1_01\"),Transform=(Rotation=(X=0.0861,Y=0.0726,Z=-0.6272,W=0.7707),Translation=(X=-4.5943,Y=-3.133,Z=-0.0626),Scale3D=(X=1,Y=1,Z=1))),(Item=(Type=Bone,Name=\"base_skirt_r_2_01\"),Transform=(Rotation=(X=0.0781,Y=0.294,Z=-0. ...(truncated)"
    WeightMinimum=0
    Weight=@n12.Value
   )
   > &RigUnit_ModifyTransforms(
    ItemToModify="((Item=(Type=Bone,Name=\"base_skirt_r_1_01\"),Transform=(Rotation=(X=0.0146,Y=-0.535,Z=-0.3221,W=0.7809),Translation=(X=0,Y=0,Z=0),Scale3D=(X=1,Y=1,Z=1))),(Item=(Type=Bone,Name=\"base_skirt_r_2_01\"),Transform=(Rotation=(X=-0.0199,Y=-0.0493,Z=-0.6772,W=0.7339) ...(truncated)"
    WeightMinimum=0
    Weight=RigVMVariableNode() {variable=RightLeg Out Alpha}.Value
   )
   > &RigUnit_ModifyTransforms(
    ItemToModify="((Item=(Type=Bone,Name=\"base_skirt_l_4_01\"),Transform=(Rotation=(X=-0.0311,Y=-0.0399,Z=-0.5941,W=0.8028),Translation=(X=0,Y=0,Z=0),Scale3D=(X=1,Y=1,Z=1))),(Item=(Type=Bone,Name=\"base_skirt_r_2_01\"),Transform=(Rotation=(X=0,Y=-0.2207,Z=0,W=0.9753),Translati ...(truncated)"
    WeightMinimum=0
    Weight=RigVMVariableNode() {variable=RightLeg Bwd Alpha}.Value
   )
   > &RigUnit_ModifyTransforms(
    ItemToModify="((Item=(Type=Bone,Name=\"base_skirt_r_1_01\"),Transform=(Rotation=(X=0,Y=0,Z=-0.1694,W=0.9856),Translation=(X=0.36,Y=0.8,Z=0),Scale3D=(X=1,Y=1,Z=1))),(Item=(Type=Bone,Name=\"base_skirt_r_2_01\"),Transform=(Rotation=(X=0,Y=0.1316,Z=0,W=0.9913),Translation=(X=0. ...(truncated)"
    WeightMinimum=0
    Weight=RigVMVariableNode() {variable=RightLeg Fwd Extreme Alpha}.Value
   )
   Subtract_2 = RigVMAggregateNode(
    B=&RigVMFunction_MathDoubleRemap(SourceMaximum=1, SourceMinimum=0, TargetMaximum=3, TargetMinimum=0, Value=RigVMVariableNode() {variable=RightLeg Fwd Alpha}.Value).Result
    A=RigVMVariableNode() {variable=RightCalf Bend Alpha}.Value
    C=RigVMVariableNode() {variable="is Knee Sliding "}.Value
   )
   > &RigUnit_ModifyTransforms(
    ItemToModify="((Item=(Type=Bone,Name=\"dyn_skirt_r_4_03\"),Transform=(Rotation=(X=0,Y=0,Z=-0.5546,W=0.8321),Translation=(X=0,Y=0,Z=0),Scale3D=(X=1,Y=1,Z=1))),(Item=(Type=Bone,Name=\"dyn_skirt_r_3_03\"),Transform=(Rotation=(X=0,Y=0,Z=-0.3387,W=0.9409),Translation=(X=0,Y=0,Z= ...(truncated)"
    WeightMinimum=0
    Weight=@Subtract_2.Result
   )
   > &RigUnit_ModifyTransforms(
    ItemToModify="((Item=(Type=Bone,Name=\"base_skirt_l_1_01\"),Transform=(Rotation=(X=-0.0457,Y=-0.1029,Z=-0.5651,W=0.8173),Translation=(X=-4.5943,Y=-3.133,Z=-0.0626),Scale3D=(X=1,Y=1,Z=1))),(Item=(Type=Bone,Name=\"base_skirt_l_2_01\"),Transform=(Rotation=(X=-0.0227,Y=0.2621,Z ...(truncated)"
    WeightMinimum=0
    Weight=RigVMVariableNode() {variable=LeftLeg Fwd Alpha}.Value
   )
   > &RigUnit_ModifyTransforms(
    ItemToModify="((Item=(Type=Bone,Name=\"base_skirt_l_1_01\"),Transform=(Rotation=(X=-0.0146,Y=0.535,Z=-0.3221,W=0.7809),Translation=(X=0,Y=0,Z=0),Scale3D=(X=1,Y=1,Z=1))),(Item=(Type=Bone,Name=\"base_skirt_l_2_01\"),Transform=(Rotation=(X=0.0199,Y=0.0493,Z=-0.6772,W=0.7339),T ...(truncated)"
    WeightMinimum=0
    Weight=RigVMVariableNode() {variable=LeftLeg Out Alpha}.Value
   )
   > &RigUnit_ModifyTransforms(
    ItemToModify="((Item=(Type=Bone,Name=\"base_skirt_r_4_01\"),Transform=(Rotation=(X=0.029,Y=0.0414,Z=-0.594,W=0.8029),Translation=(X=0,Y=0,Z=0),Scale3D=(X=1,Y=1,Z=1))),(Item=(Type=Bone,Name=\"base_skirt_l_2_01\"),Transform=(Rotation=(X=0,Y=0.2207,Z=0,W=0.9753),Translation=(X ...(truncated)"
    WeightMinimum=0
    Weight=RigVMVariableNode() {variable=LeftLeg Bwd Alpha}.Value
   )
   > &RigUnit_ModifyTransforms(
    ItemToModify="((Item=(Type=Bone,Name=\"base_skirt_l_1_01\"),Transform=(Rotation=(X=0,Y=0,Z=-0.1694,W=0.9856),Translation=(X=-0.36,Y=-0.8,Z=0),Scale3D=(X=1,Y=1,Z=1))),(Item=(Type=Bone,Name=\"base_skirt_l_2_01\"),Transform=(Rotation=(X=0,Y=-0.1316,Z=0,W=0.9913),Translation=(X ...(truncated)"
    WeightMinimum=0
    Weight=RigVMVariableNode() {variable=LeftLeg Fwd Extreme Alpha}.Value
   )
   Remap = &RigVMFunction_MathDoubleRemap(SourceMaximum=1, SourceMinimum=0, TargetMaximum=3, TargetMinimum=0, Value=RigVMVariableNode() {variable=LeftLeg Fwd Alpha}.Value)
   Subtract_1 = RigVMAggregateNode(B=@Remap.Result, A=RigVMVariableNode() {variable=LeftCalf Bend Alpha}.Value, C=RigVMVariableNode() {variable="is Knee Sliding "}.Value)
   > &RigUnit_ModifyTransforms(
    ItemToModify="((Item=(Type=Bone,Name=\"dyn_skirt_l_4_03\"),Transform=(Rotation=(X=0,Y=0,Z=-0.5546,W=0.8321),Translation=(X=0,Y=0,Z=0),Scale3D=(X=1,Y=1,Z=1))),(Item=(Type=Bone,Name=\"dyn_skirt_l_3_03\"),Transform=(Rotation=(X=0,Y=0,Z=-0.3387,W=0.9409),Translation=(X=0,Y=0,Z= ...(truncated)"
    WeightMinimum=0
    Weight=@Subtract_1.Result
   )
   Subtract = RigVMAggregateNode(A=1, B=RigVMVariableNode() {variable=is Moving 2D}.Value, C=RigVMVariableNode() {variable=Crouching high position}.Value)
   n13 = &RigVMFunction_AlphaInterp(
    InterpSpeedDecreasing=15
    InterpSpeedIncreasing=7
    Scale=1
    bClampResult=true
    Value=&RigVMFunction_MathDoubleMul(B=@Subtract.Result, A=RigVMVariableNode() {variable=is Crouching}.Value).Result
   )
   n14 = &RigVMFunction_AnimEvalRichCurve(
    Curve="(EditorCurveData=(Keys=((InterpMode=RCIM_Linear,TangentMode=RCTM_Auto,TangentWeightMode=RCTWM_WeightedNone,Time=0,Value=0,ArriveTangent=0,ArriveTangentWeight=0,LeaveTangent=0,LeaveTangentWeight=0),(InterpMode=RCIM_Linear,TangentMode=RCTM_Auto,TangentWeight ...(truncated)"
    SourceMinimum=0
    TargetMinimum=0
    Value=@n13.Result
   )
   > &RigUnit_ModifyTransforms(
    ItemToModify="((Item=(Type=Bone,Name=\"base_skirt_l_1_01\"),Transform=(Rotation=(X=0.002,Y=0.1374,Z=0.2924,W=0.9464),Translation=(X=0,Y=-4,Z=0),Scale3D=(X=1,Y=1,Z=1))),(Item=(Type=Bone,Name=\"base_skirt_l_2_01\"),Transform=(Rotation=(X=-0.0449,Y=0.1677,Z=0.2549,W=0.9513),Tr ...(truncated)"
    WeightMinimum=0
    Weight=@n14.Result
   )
   n15 = &RigVMFunction_AlphaInterp(InterpSpeedDecreasing=15, InterpSpeedIncreasing=7, Scale=1, bClampResult=true, Value=RigVMVariableNode() {variable="is Knee Sliding "}.Value)
   n16 = &RigVMFunction_AnimEvalRichCurve(
    Curve="(EditorCurveData=(Keys=((InterpMode=RCIM_Linear,TangentMode=RCTM_Auto,TangentWeightMode=RCTWM_WeightedNone,Time=0,Value=0,ArriveTangent=0,ArriveTangentWeight=0,LeaveTangent=0,LeaveTangentWeight=0),(InterpMode=RCIM_Linear,TangentMode=RCTM_Auto,TangentWeight ...(truncated)"
    SourceMinimum=0
    TargetMinimum=0
    Value=@n15.Result
   )
   > &RigUnit_ModifyTransforms(
    ItemToModify="((Item=(Type=Bone,Name=\"base_skirt_l_4_01\"),Transform=(Rotation=(X=-0.0714,Y=-0.05,Z=-0.816,W=0.5714),Translation=(X=0,Y=0,Z=0),Scale3D=(X=1,Y=1,Z=1))),(Item=(Type=Bone,Name=\"base_skirt_r_4_01\"),Transform=(Rotation=(X=0,Y=0,Z=-0.8192,W=0.5736),Translation= ...(truncated)"
    WeightMinimum=0
    Weight=@n16.Result
   )
   n17 = &RigVMFunction_AlphaInterp(
    InterpSpeedDecreasing=15
    InterpSpeedIncreasing=7
    Scale=1
    bClampResult=true
    Value=RigVMVariableNode() {variable=is Baseball Sliding}.Value
   )
   n18 = &RigVMFunction_AnimEvalRichCurve(
    Curve="(EditorCurveData=(Keys=((InterpMode=RCIM_Linear,TangentMode=RCTM_Auto,TangentWeightMode=RCTWM_WeightedNone,Time=0,Value=0,ArriveTangent=0,ArriveTangentWeight=0,LeaveTangent=0,LeaveTangentWeight=0),(InterpMode=RCIM_Linear,TangentMode=RCTM_Auto,TangentWeight ...(truncated)"
    SourceMinimum=0
    TargetMinimum=0
    Value=@n17.Result
   )
   > &RigUnit_ModifyTransforms(
    ItemToModify="((Item=(Type=Bone,Name=\"base_skirt_l_4_01\"),Transform=(Rotation=(X=0.0985,Y=-0.1392,Z=0.8043,W=-0.5692),Translation=(X=0,Y=0,Z=0),Scale3D=(X=1,Y=1,Z=1))),(Item=(Type=Bone,Name=\"base_skirt_r_4_01\"),Transform=(Rotation=(X=0,Y=0,Z=-0.866,W=0.5),Translation=(X ...(truncated)"
    WeightMinimum=0
    Weight=@n18.Result
   )
   n19 = &RigVMFunction_AnimEvalRichCurve(
    Curve=(EditorCurveData=(Keys=((),(Time=0.5,Value=1),(Time=0.8))))
    SourceMaximum=2
    SourceMinimum=0
    TargetMinimum=0
    Value=&RigVMFunction_MathDoubleRemap(
     SourceMaximum=75
     SourceMinimum=45
     TargetMaximum=1
     TargetMinimum=0
     bClamp=false
     Value=&RigUnit_GetRelativeTransformForItem(Child=(Type=Bone,Name="pelvis"), Parent=(Type=Bone,Name="root"), bChildInitial=False, bParentInitial=False).RelativeTransform.Translation.Z
    ).Result
   )
   > RigVMVariableNode(Value=@n19.Result) {variable=Crouching high position}
   > n20 = &RigUnit_SphericalPoseReader(
    ActiveRegionScaleFactors=(PositiveWidth=1,NegativeWidth=1,PositiveHeight=1,NegativeHeight=1)
    ActiveRegionSize=0.1
    Debug=(bDrawDebug=True,bDraw2D=False,bDrawLocalAxes=False,DebugScale=25,DebugSegments=20,DebugThickness=0.25)
    DriverAxis=(X=0,Y=1,Z=0)
    DriverItem=(Type=Bone,Name="head")
    FalloffRegionScaleFactors=(PositiveWidth=1,NegativeWidth=1,PositiveHeight=1,NegativeHeight=1)
    FalloffSize=0.2
    FlipHeightScaling=False
    FlipWidthScaling=False
    OptionalParentItem=(Type=Bone,Name="None")
    RotationOffset=(X=58,Y=0,Z=-15)
   )
   > &RigUnit_ModifyTransforms(
    ItemToModify="((Item=(Type=Bone,Name=\"base_skirt_r_1_01\"),Transform=(Rotation=(X=-0.0038,Y=-0.0435,Z=-0.0871,W=0.9952),Translation=(X=0,Y=0,Z=0),Scale3D=(X=1,Y=1,Z=1))),(Item=(Type=Bone,Name=\"base_skirt_r_2_01\"),Transform=(Rotation=(X=0,Y=0.0872,Z=0,W=0.9962),Translatio ...(truncated)"
    WeightMinimum=0
    Weight=&RigVMFunction_MathDoubleMul(A=@n20.OutputParam, B=RigVMVariableNode() {variable=is Crouching}.Value).Result
   )
   &RigUnit_ModifyTransforms(
    ItemToModify="((Item=(Type=Bone,Name=\"base_shoulderPad_l_01\"),Transform=(Rotation=(X=0,Y=0,Z=0,W=1),Translation=(X=-3.0413,Y=-7.7053,Z=0),Scale3D=(X=1,Y=1,Z=1))),(Item=(Type=Bone,Name=\"base_shoulderPad_l_02\"),Transform=(Rotation=(X=0.0374,Y=0,Z=0,W=0.9993),Translation=( ...(truncated)"
    WeightMinimum=0
    Weight=RigVMDispatchNode(BoneName=root, Default=0, Name=deformRig_l_upperarm_fr_alpha) {template="DISPATCH_RigDispatch_GetAnimAttribute(in Name,in BoneName,in Default,out Value,out Found)"}.Value
    ExecutePin=@RigVMFunction_Sequence.B
   )
   > n21 = &RigUnit_SphericalPoseReader(
    ActiveRegionScaleFactors=(PositiveWidth=2.7,NegativeWidth=3,PositiveHeight=3,NegativeHeight=3)
    ActiveRegionSize=0.1
    Debug=(bDrawDebug=false,bDraw2D=False,bDrawLocalAxes=False,DebugScale=25,DebugSegments=20,DebugThickness=0.25)
    DriverAxis=(X=1,Y=0,Z=0)
    DriverItem=(Type=Bone,Name="upperarm_l")
    FalloffRegionScaleFactors=(PositiveWidth=1,NegativeWidth=1,PositiveHeight=1,NegativeHeight=1)
    FalloffSize=0.5
    FlipHeightScaling=False
    FlipWidthScaling=False
    OptionalParentItem=(Type=Bone,Name="None")
    RotationOffset=(X=0,Y=45,Z=0)
   )
   > &RigUnit_ModifyTransforms(
    ItemToModify="((Item=(Type=Bone,Name=\"base_shoulderPad_l_03\"),Transform=(Rotation=(X=0,Y=0,Z=0,W=1),Translation=(X=-0.8,Y=-5.6,Z=0),Scale3D=(X=1,Y=1,Z=1))),(Item=(Type=Bone,Name=\"base_shoulderPad_l_02\"),Transform=(Rotation=(X=-0.0428,Y=0,Z=0,W=0.9991),Translation=(X=0,Y ...(truncated)"
    WeightMinimum=0
    Weight=@n21.OutputParam
   )
   > &RigUnit_ModifyTransforms(
    ItemToModify="((Item=(Type=Bone,Name=\"base_shoulderPad_l_05\"),Transform=(Rotation=(X=0,Y=0,Z=-0.0168,W=0.9999),Translation=(X=-6.1824,Y=-5.5183,Z=0),Scale3D=(X=1,Y=1,Z=1))),(Item=(Type=Bone,Name=\"base_shoulderPad_l_04\"),Transform=(Rotation=(X=0,Y=0,Z=0,W=1),Translation= ...(truncated)"
    WeightMinimum=0
    Weight=RigVMDispatchNode(BoneName=root, Default=0, Name=deformRig_l_upperarm_bk_alpha) {template="DISPATCH_RigDispatch_GetAnimAttribute(in Name,in BoneName,in Default,out Value,out Found)"}.Value
   )
   > n22 = &RigUnit_SphericalPoseReader(
    ActiveRegionScaleFactors=(PositiveWidth=3,NegativeWidth=1,PositiveHeight=1.5,NegativeHeight=1.5)
    ActiveRegionSize=0.1
    Debug=(bDrawDebug=false,DebugScale=25,DebugSegments=20,DebugThickness=0.25)
    DriverAxis=(X=1,Y=0,Z=0)
    DriverItem=(Type=Bone,Name="upperarm_l")
    FalloffRegionScaleFactors=(PositiveWidth=1,NegativeWidth=1.25,PositiveHeight=1.25,NegativeHeight=1)
    FalloffSize=0.4
    FlipHeightScaling=true
    FlipWidthScaling=false
    OptionalParentItem=(Type=Bone)
    RotationOffset=(X=90,Y=0,Z=180)
   )
   > &RigUnit_ModifyTransforms(
    ItemToModify=((Item=(Type=Bone,Name="base_cape_l_00"),Transform=(Rotation=(X=0.5608,Y=0,Z=0,W=0.8279),Translation=(X=-5.5678,Y=5.614,Z=5.7175),Scale3D=(X=1,Y=1,Z=1))))
    WeightMinimum=0
    Weight=@n22.OutputParam
   )
   > &RigUnit_ModifyTransforms(
    ItemToModify="((Item=(Type=Bone,Name=\"base_shoulderPad_r_01\"),Transform=(Rotation=(X=0,Y=0,Z=0,W=1),Translation=(X=3.0413,Y=7.7053,Z=0),Scale3D=(X=1,Y=1,Z=1))),(Item=(Type=Bone,Name=\"base_shoulderPad_r_02\"),Transform=(Rotation=(X=0.0374,Y=0,Z=0,W=0.9993),Translation=(X= ...(truncated)"
    WeightMinimum=0
    Weight=RigVMDispatchNode(BoneName=root, Default=0, Name=deformRig_r_upperarm_fr_alpha) {template="DISPATCH_RigDispatch_GetAnimAttribute(in Name,in BoneName,in Default,out Value,out Found)"}.Value
   )
   > n23 = &RigUnit_SphericalPoseReader(
    ActiveRegionScaleFactors=(PositiveWidth=2.7,NegativeWidth=3,PositiveHeight=3,NegativeHeight=3)
    ActiveRegionSize=0.1
    Debug=(bDrawDebug=false,bDraw2D=False,bDrawLocalAxes=False,DebugScale=25,DebugSegments=20,DebugThickness=0.25)
    DriverAxis=(X=-1,Y=0,Z=0)
    DriverItem=(Type=Bone,Name="upperarm_r")
    FalloffRegionScaleFactors=(PositiveWidth=1,NegativeWidth=1,PositiveHeight=1,NegativeHeight=1)
    FalloffSize=0.5
    FlipHeightScaling=False
    FlipWidthScaling=False
    OptionalParentItem=(Type=Bone,Name="None")
    RotationOffset=(X=0,Y=225,Z=0)
   )
   > &RigUnit_ModifyTransforms(
    ItemToModify="((Item=(Type=Bone,Name=\"base_shoulderPad_r_03\"),Transform=(Rotation=(X=0,Y=0,Z=0,W=1),Translation=(X=0.8,Y=5.6,Z=0),Scale3D=(X=1,Y=1,Z=1))),(Item=(Type=Bone,Name=\"base_shoulderPad_r_02\"),Transform=(Rotation=(X=-0.0428,Y=0,Z=0,W=0.9991),Translation=(X=0,Y=2 ...(truncated)"
    WeightMinimum=0
    Weight=@n23.OutputParam
   )
   > &RigUnit_ModifyTransforms(
    ItemToModify="((Item=(Type=Bone,Name=\"base_shoulderPad_r_05\"),Transform=(Rotation=(X=0,Y=0,Z=0.0168,W=0.9999),Translation=(X=6.1824,Y=5.5183,Z=0),Scale3D=(X=1,Y=1,Z=1))),(Item=(Type=Bone,Name=\"base_shoulderPad_r_04\"),Transform=(Rotation=(X=0,Y=0,Z=0,W=1),Translation=(X= ...(truncated)"
    WeightMinimum=0
    Weight=RigVMDispatchNode(BoneName=root, Default=0, Name=deformRig_r_upperarm_bk_alpha) {template="DISPATCH_RigDispatch_GetAnimAttribute(in Name,in BoneName,in Default,out Value,out Found)"}.Value
   )
   > n24 = &RigUnit_SphericalPoseReader(
    ActiveRegionScaleFactors=(PositiveWidth=3,NegativeWidth=1,PositiveHeight=1.5,NegativeHeight=1.5)
    ActiveRegionSize=0.1
    Debug=(bDrawDebug=false,DebugScale=25,DebugSegments=20,DebugThickness=0.25)
    DriverAxis=(X=-1,Y=0,Z=0)
    DriverItem=(Type=Bone,Name="upperarm_r")
    FalloffRegionScaleFactors=(PositiveWidth=1,NegativeWidth=1.25,PositiveHeight=1.25,NegativeHeight=1)
    FalloffSize=0.4
    FlipHeightScaling=true
    FlipWidthScaling=false
    OptionalParentItem=(Type=Bone)
    RotationOffset=(X=90,Y=0,Z=0)
   )
   > &RigUnit_ModifyTransforms(
    ItemToModify=((Item=(Type=Bone,Name="base_cape_r_00"),Transform=(Rotation=(X=-0.5608,Y=0,Z=0,W=0.8279),Translation=(X=-5.5678,Y=5.614,Z=-5.7175),Scale3D=(X=1,Y=1,Z=1))))
    WeightMinimum=0
    Weight=@n24.OutputParam
   )
   Multiply = &RigVMFunction_MathDoubleMul(A=RigVMVariableNode() {variable=is Crouching}.Value, B=RigVMVariableNode() {variable=is Moving 2D}.Value)
   Interpolate = /Script/RigVM.RigVMFunction_MathQuaternionSlerp(A=(X=-0.0457,Y=-0.1029,Z=-0.5651,W=0.8173), B=(X=-0.0548,Y=-0.0984,Z=-0.6369,W=0.7627), T=@Multiply.Result) {is=struct RigVMUnitNode, template="Interpolate::Execute(in A,in B,in T,out Result)"}
   &RigVMFunction_MathQuaternionSlerp(A=(X=0.0227,Y=-0.2621,Z=-0.0836,W=0.9611), B=(X=0.0086,Y=-0.304,Z=-0.1092,W=0.9463), T=@Multiply.Result) {template="Interpolate::Execute(in A,in B,in T,out Result)"}
   &RigUnit_ModifyTransforms(
    ExecutePin=()
    ItemToModify="((Item=(Type=Bone,Name=\"base_skirt_r_1_01\"),Transform=(Rotation=(X=0.0806,Y=0.0787,Z=-0.569,W=0.8146),Translation=(X=-4.5943,Y=-3.133,Z=-0.0626),Scale3D=(X=1,Y=1,Z=1))),(Item=(Type=Bone,Name=\"base_skirt_r_2_01\"),Transform=(Rotation=(X=0.0745,Y=0.2523,Z=-0. ...(truncated)"
    Weight=1
    WeightMinimum=0
   )
   &RigUnit_ModifyTransforms(
    ExecutePin=()
    ItemToModify="((Item=(Type=Bone,Name=\"base_skirt_l_1_01\"),Transform=(Rotation=(X=0.002,Y=0.1374,Z=0.2924,W=0.9464),Translation=(X=0,Y=-4,Z=0),Scale3D=(X=1,Y=1,Z=1))),(Item=(Type=Bone,Name=\"base_skirt_l_2_01\"),Transform=(Rotation=(X=-0.0449,Y=0.1677,Z=0.2549,W=0.9513),Tr ...(truncated)"
    Weight=1
    WeightMinimum=0
   )
   &RigUnit_ModifyTransforms(
    ExecutePin=()
    ItemToModify="((Item=(Type=Bone,Name=\"base_skirt_l_1_01\"),Transform=(Rotation=(X=0.0416,Y=0.0131,Z=0.3004,W=0.9528),Translation=(X=0,Y=-4,Z=0),Scale3D=(X=1,Y=1,Z=1))),(Item=(Type=Bone,Name=\"base_skirt_l_2_01\"),Transform=(Rotation=(X=-0.0338,Y=0.1261,Z=0.2566,W=0.9577),T ...(truncated)"
    Weight=1
    WeightMinimum=0
   )
   RigVMDispatchNode(Enabled=True, ExecuteContext=(), Prefix=crouching idle, ScreenColor=(R=1,G=1,B=1,A=1), ScreenDuration=0.0, Value=@n14.Result) {template="DISPATCH_RigVMDispatch_Print(in Prefix,in Value,in Enabled,in ScreenDuration,in ScreenColor)"}
   RigVMDispatchNode(Enabled=True, ExecuteContext=(), ScreenColor=(R=1,G=1,B=1,A=1), ScreenDuration=0.0, Value=@n19.Result) {template="DISPATCH_RigVMDispatch_Print(in Prefix,in Value,in Enabled,in ScreenDuration,in ScreenColor)"}
   RigVMDispatchNode(Enabled=True, ExecuteContext=(), ScreenColor=(R=1,G=1,B=1,A=1), ScreenDuration=0.0, Value=@n12.Value) {template="DISPATCH_RigVMDispatch_Print(in Prefix,in Value,in Enabled,in ScreenDuration,in ScreenColor)"}
  ]}
 Subtract_3_ContainedGraph {nodes [
   n1 = RigVMFunctionEntryNode()
   RigVMFunctionReturnNode(Result=&RigVMFunction_MathDoubleSub(B=@n1.C, A=&RigVMFunction_MathDoubleSub(A=@n1.A, B=@n1.B).Result).Result)
  ]}
 Subtract_3_ContainedGraph {nodes [
   n1 = RigVMFunctionEntryNode()
   RigVMFunctionReturnNode(Result=&RigVMFunction_MathDoubleSub(B=@n1.C, A=&RigVMFunction_MathDoubleSub(A=@n1.A, B=@n1.B).Result).Result)
  ]}
 Subtract_3_ContainedGraph {nodes [
   n1 = RigVMFunctionEntryNode()
   Subtract = &RigVMFunction_MathDoubleSub(A=@n1.A, B=@n1.B)
   RigVMFunctionReturnNode(Result=&RigVMFunction_MathDoubleSub(B=@n1.C, A=@Subtract.Result).Result)
  ]}
]
hierarchy [
 root {loc=<0 0 0>, rot=<0 0 0 1>}
 pelvis {parent=root, loc=<0 -0.5301 98.0571>, rot=<0.0127 0.707 -0.0127 -0.707>}
 base_skirt_r_1_01 {parent=pelvis, loc=<4.1661 6.9665 7.6793>, rot=<0.0656 -0.2585 0.9626 0.0469>}
 dyn_skirt_r_1_01 {parent=base_skirt_r_1_01, loc=<0 0 0>, rot=<0 0 0 1>}
 dyn_skirt_r_1_02 {parent=dyn_skirt_r_1_01, loc=<14.2501 -0.7852 0>, rot=<0 0 0 1>}
 dyn_skirt_r_1_03 {parent=dyn_skirt_r_1_02, loc=<14.25 0.0573 -0.0001>, rot=<0 0 0 1>}
 dyn_skirt_r_1_04 {parent=dyn_skirt_r_1_03, loc=<14.1709 0.4023 0>, rot=<0 0 0 1>}
 dyn_skirt_r_1_05 {parent=dyn_skirt_r_1_04, loc=<14.25 0.2496 0>, rot=<0 0 0 1>}
 base_skirt_r_2_01 {parent=pelvis, loc=<3.9352 0.8334 12.8604>, rot=<0.0754 -0.6804 0.7266 0.0591>}
 dyn_skirt_r_2_01 {parent=base_skirt_r_2_01, loc=<0 0 0>, rot=<0 0 0 1>}
 dyn_skirt_r_2_02 {parent=dyn_skirt_r_2_01, loc=<13.8747 -0.7938 0>, rot=<0 0 0 1>}
 dyn_skirt_r_2_03 {parent=dyn_skirt_r_2_02, loc=<13.875 0.2935 0>, rot=<0 0 0 1>}
 dyn_skirt_r_2_04 {parent=dyn_skirt_r_2_03, loc=<13.8791 0.443 0>, rot=<0 0 0 1>}
 dyn_skirt_r_2_05 {parent=dyn_skirt_r_2_04, loc=<13.875 0.0625 0>, rot=<0 0 0 1>}
 base_skirt_r_3_01 {parent=pelvis, loc=<3.6893 -6.0558 11.2447>, rot=<0.0598 -0.9231 0.3771 0.0466>}
 dyn_skirt_r_3_01 {parent=base_skirt_r_3_01, loc=<0 0 0>, rot=<0 0 0 1>}
 dyn_skirt_r_3_02 {parent=dyn_skirt_r_3_01, loc=<13.5 -1.1682 0.0001>, rot=<0 0 0 1>}
 dyn_skirt_r_3_03 {parent=dyn_skirt_r_3_02, loc=<13.5001 0.7842 0>, rot=<0 0 0 1>}
 dyn_skirt_r_3_04 {parent=dyn_skirt_r_3_03, loc=<13.5735 0.5762 0>, rot=<0 0 0 1>}
 dyn_skirt_r_3_05 {parent=dyn_skirt_r_3_04, loc=<13.5 -0.0514 0>, rot=<0 0 0 1>}
 base_skirt_r_4_01 {parent=pelvis, loc=<3.6194 -9.4021 4.6745>, rot=<0.049 -0.9948 0.0844 0.0281>}
 dyn_skirt_r_4_01 {parent=base_skirt_r_4_01, loc=<0 0 0>, rot=<0 0 0 -1>}
 dyn_skirt_r_4_02 {parent=dyn_skirt_r_4_01, loc=<13.3498 -1.5291 0>, rot=<0 0 0 -1>}
 dyn_skirt_r_4_03 {parent=dyn_skirt_r_4_02, loc=<13.35 0.6206 0>, rot=<0 0 0 -1>}
 dyn_skirt_r_4_04 {parent=dyn_skirt_r_4_03, loc=<13.355 0.4837 0>, rot=<0 0 0 -1>}
 dyn_skirt_r_4_05 {parent=dyn_skirt_r_4_04, loc=<13.3501 0.4241 0>, rot=<0 0 0 -1>}
 base_skirt_l_1_01 {parent=pelvis, loc=<4.166 6.9665 -7.6793>, rot=<-0.0656 0.2585 0.9626 0.0469>}
 dyn_skirt_l_1_01 {parent=base_skirt_l_1_01, loc=<0 0 0>, rot=<0 0 0 1>}
 dyn_skirt_l_1_02 {parent=dyn_skirt_l_1_01, loc=<14.25 -0.7852 0>, rot=<0 0 0 1>}
 dyn_skirt_l_1_03 {parent=dyn_skirt_l_1_02, loc=<14.25 0.0573 0>, rot=<0 0 0 1>}
 dyn_skirt_l_1_04 {parent=dyn_skirt_l_1_03, loc=<14.1708 0.4023 0>, rot=<0 0 0 1>}
 dyn_skirt_l_1_05 {parent=dyn_skirt_l_1_04, loc=<14.25 0.2496 0>, rot=<0 0 0 1>}
 base_skirt_l_2_01 {parent=pelvis, loc=<3.9355 0.8334 -12.8604>, rot=<-0.0754 0.6804 0.7266 0.0591>}
 dyn_skirt_l_2_01 {parent=base_skirt_l_2_01, loc=<0 0 0>, rot=<0 0 0 1>}
 dyn_skirt_l_2_02 {parent=dyn_skirt_l_2_01, loc=<13.875 -0.7937 0>, rot=<0 0 0 1>}
 dyn_skirt_l_2_03 {parent=dyn_skirt_l_2_02, loc=<13.875 0.2935 0>, rot=<0 0 0 1>}
 dyn_skirt_l_2_04 {parent=dyn_skirt_l_2_03, loc=<13.8792 0.4429 0>, rot=<0 0 0 1>}
 dyn_skirt_l_2_05 {parent=dyn_skirt_l_2_04, loc=<13.875 0.0625 0>, rot=<0 0 0 1>}
 base_skirt_l_3_01 {parent=pelvis, loc=<3.6892 -6.0558 -11.2447>, rot=<0.0598 -0.9231 -0.3771 -0.0466>}
 dyn_skirt_l_3_01 {parent=base_skirt_l_3_01, loc=<0 0 0>, rot=<0 0 0 1>}
 dyn_skirt_l_3_02 {parent=dyn_skirt_l_3_01, loc=<13.5 -1.1683 0>, rot=<0 0 0 1>}
 dyn_skirt_l_3_03 {parent=dyn_skirt_l_3_02, loc=<13.5 0.7842 0>, rot=<0 0 0 1>}
 dyn_skirt_l_3_04 {parent=dyn_skirt_l_3_03, loc=<13.5736 0.5761 0>, rot=<0 0 0 1>}
 dyn_skirt_l_3_05 {parent=dyn_skirt_l_3_04, loc=<13.5 -0.0513 0>, rot=<0 0 0 1>}
 base_skirt_l_4_01 {parent=pelvis, loc=<3.6196 -9.4021 -4.6745>, rot=<0.049 -0.9948 -0.0844 -0.0281>}
 dyn_skirt_l_4_01 {parent=base_skirt_l_4_01, loc=<0 0 0>, rot=<0 0 0 1>}
 dyn_skirt_l_4_02 {parent=dyn_skirt_l_4_01, loc=<13.35 -1.529 0>, rot=<0 0 0 1>}
 dyn_skirt_l_4_03 {parent=dyn_skirt_l_4_02, loc=<13.35 0.6206 0>, rot=<0 0 0 1>}
 dyn_skirt_l_4_04 {parent=dyn_skirt_l_4_03, loc=<13.3551 0.4836 0>, rot=<0 0 0 1>}
 dyn_skirt_l_4_05 {parent=dyn_skirt_l_4_04, loc=<13.35 0.4241 0>, rot=<0 0 0 1>}
 base_skirt_f_01 {parent=pelvis, loc=<4.257 8.909 0>, rot=<0 0 0.9995 0.033>}
 dyn_skirt_f_01 {parent=base_skirt_f_01, loc=<0 0 0>, rot=<0 0 0 1>}
 dyn_skirt_f_02 {parent=dyn_skirt_f_01, loc=<9.1819 -0.0498 0>, rot=<0 0 0 1>}
 dyn_skirt_f_03 {parent=dyn_skirt_f_02, loc=<13.4045 -0.0269 0>, rot=<0 0 0 1>}
 dyn_skirt_f_04 {parent=dyn_skirt_f_03, loc=<13.4207 0.1293 0>, rot=<0 0 0 1>}
 dyn_skirt_f_05 {parent=dyn_skirt_f_04, loc=<13.1369 0.0896 0>, rot=<0 0 0 1>}
 spine_01 {parent=pelvis, loc=<6.1208 -0.9328 0>, rot=<0 0 0.034 -0.9994>}
 spine_02 {parent=spine_01, loc=<6.5997 -0.2095 0>, rot=<0 0 0.0238 -0.9997>}
 spine_03 {parent=spine_02, loc=<6.4914 -0.1237 0>, rot=<0 0 0.0149 -0.9999>}
 spine_04 {parent=spine_03, loc=<6.9598 0.0794 0>, rot=<0 0 -0.0132 -0.9999>}
 spine_05 {parent=spine_04, loc=<6.9987 0 0>, rot=<0 0 -0.0015 -1>}
 neck_01 {parent=spine_05, loc=<11.3138 0.1529 0>, rot=<0 0 -0.2003 -0.9797>}
 neck_02 {parent=neck_01, loc=<4.5627 0 0>, rot=<0 0 0.0504 -0.9987>}
 head {parent=neck_02, loc=<5.3914 0 0>, rot=<0 0 0.115 -0.9934>}
 clavicle_l {parent=spine_05, loc=<5.4831 -0.1053 -3.2946>, rot=<0.0465 -0.7289 -0.009 -0.6829>}
 upperarm_l {parent=clavicle_l, loc=<9.8658 -1.2136 -0.4105>, rot=<0.0164 -0.3245 -0.0269 -0.9454>}
 lowerarm_l {parent=upperarm_l, loc=<26.465 -0.8345 0.259>, rot=<0.0036 0.0161 -0.2298 -0.9731>}
 hand_l {parent=lowerarm_l, loc=<22.7355 -0.3203 -0.6105>, rot=<-0.6784 -0.0223 0.0735 -0.7307>}
 thumb_01_l {parent=hand_l, loc=<1.3118 -0.3191 -2.459>, rot=<0.6546 -0.4042 0.0664 -0.6354>}
 thumb_02_l {parent=thumb_01_l, loc=<5.9336 -0.0448 0.0401>, rot=<-0.0037 0.0213 0.298 -0.9543>}
 thumb_03_l {parent=thumb_02_l, loc=<2.2223 0.3682 -0.0907>, rot=<0.0118 -0.013 0.0066 -0.9998>}
 index_metacarpal_l {parent=hand_l, loc=<1.6535 0.0183 -2.96>, rot=<0.0158 -0.0124 -0.0173 -0.9996>}
 index_01_l {parent=index_metacarpal_l, loc=<8.3008 0.0357 -0.0524>, rot=<-0.0044 0.0485 0.0924 -0.9945>}
 index_02_l {parent=index_01_l, loc=<3.3587 -0.3004 0.0452>, rot=<-0.002 0.0027 0.3225 -0.9466>}
 index_03_l {parent=index_02_l, loc=<2.5578 -0.123 -0.0252>, rot=<-0.002 0.0067 0.2558 -0.9667>}
 middle_metacarpal_l {parent=hand_l, loc=<2.1856 -0.1725 -0.5984>, rot=<0.006 0.0168 -0.0329 -0.9993>}
 middle_01_l {parent=middle_metacarpal_l, loc=<7.3275 -0.2447 0.0129>, rot=<-0.0208 0.0042 0.0858 -0.9961>}
 middle_02_l {parent=middle_01_l, loc=<4.0123 0.1401 0.1175>, rot=<-0.0218 0.0258 0.3512 -0.9357>}
 middle_03_l {parent=middle_02_l, loc=<2.6077 0.0534 0.1373>, rot=<0.0082 0.0188 0.1708 -0.9851>}
 ring_metacarpal_l {parent=hand_l, loc=<2.2074 -0.3625 1.3638>, rot=<-0.0284 0.0722 -0.0458 -0.9959>}
 ring_01_l {parent=ring_metacarpal_l, loc=<7.0295 -0.6197 -0.1446>, rot=<-0.0489 -0.0252 0.1892 -0.9804>}
 ring_02_l {parent=ring_01_l, loc=<3.6174 0.4647 -0.1027>, rot=<0 0 0.2806 -0.9598>}
 ring_03_l {parent=ring_02_l, loc=<2.4152 0.114 0.0199>, rot=<0 0 0.1189 -0.9929>}
 pinky_metacarpal_l {parent=hand_l, loc=<1.7616 -0.9753 3.3814>, rot=<-0.0397 0.0814 -0.069 -0.9935>}
 pinky_01_l {parent=pinky_metacarpal_l, loc=<7.1709 -0.4831 -0.1094>, rot=<-0.0654 -0.0035 0.2242 -0.9723>}
 pinky_02_l {parent=pinky_01_l, loc=<2.8274 0.0893 -0.1043>, rot=<0.0037 0.0011 0.2854 -0.9584>}
 pinky_03_l {parent=pinky_02_l, loc=<2.2046 0.0307 -0.0078>, rot=<-0.0002 0.0098 0.1445 -0.9895>}
 deform_wrist_dn_base_l {parent=hand_l, loc=<0 -1.5326 0>, rot=<0 0 0 -1>}
 deform_wrist_dn_l {parent=deform_wrist_dn_base_l, loc=<0 0 0>, rot=<0 0 0 1>}
 deform_wrist_up_base_l {parent=hand_l, loc=<0 2.3009 0>, rot=<0 0 0 -1>}
 deform_wrist_up_l {parent=deform_wrist_up_base_l, loc=<0 0 0>, rot=<0 0 0 1>}
 lowerarm_twist_02_l {parent=lowerarm_l, loc=<10.6707 0 0>, rot=<0 0 0 1>}
 lowerarm_twist_01_l {parent=lowerarm_l, loc=<18.316 0 0>, rot=<0 0 0 1>}
 deform_elbow_in_base_l {parent=lowerarm_l, loc=<0.1834 3.6062 -0.2572>, rot=<0 0 0 -1>}
 deform_elbow_in_l {parent=deform_elbow_in_base_l, loc=<0 0 0>, rot=<0 0 0 1>}
 deform_elbow_out_base_l {parent=lowerarm_l, loc=<0 -4 0>, rot=<0 0 0 1>}
 deform_elbow_out_l {parent=deform_elbow_out_base_l, loc=<0 0 0>, rot=<0 0 0 1>}
 upperarm_twist_01_l {parent=upperarm_l, loc=<7.2135 -0.1134 0>, rot=<0 0 0 1>}
 deform_upperarm_bk_base_l {parent=upperarm_twist_01_l, loc=<-6.4053 -4.3746 0.314>, rot=<0 0 0 -1>}
 deform_upperarm_bk_l {parent=deform_upperarm_bk_base_l, loc=<0 0 0>, rot=<0 0 0 1>}
 deform_upperarm_fr_base_l {parent=upperarm_twist_01_l, loc=<-5.6312 5.092 0.3354>, rot=<0 0 0 -1>}
 deform_upperarm_fr_l {parent=deform_upperarm_fr_base_l, loc=<0 0 0>, rot=<0 0 0 1>}
 deform_upperarm_up_base_l {parent=upperarm_twist_01_l, loc=<-8.891 0.4825 3.9517>, rot=<0 0 0 -1>}
 deform_upperarm_up_l {parent=deform_upperarm_up_base_l, loc=<0 0 0>, rot=<0 0 0 1>}
 upperarm_twist_02_l {parent=upperarm_l, loc=<19.7998 -0.1134 0>, rot=<0 0 0 1>}
 base_shoulderPad_l_02 {parent=clavicle_l, loc=<12.2306 3.9824 2.6454>, rot=<0.8939 0.1728 -0.1233 -0.3948>}
 dyn_shoulderPad_l_02 {parent=base_shoulderPad_l_02, loc=<0 0 0>, rot=<0 0 0 1>}
 base_shoulderPad_l_03 {parent=clavicle_l, loc=<13.8322 -1.6765 5.4324>, rot=<0.7071 0.0415 0.0059 -0.7059>}
 dyn_shoulderPad_l_03 {parent=base_shoulderPad_l_03, loc=<0 0 0>, rot=<0 0 0 1>}
 base_shoulderPad_l_04 {parent=clavicle_l, loc=<12.602 -6.416 2.0762>, rot=<0.1542 -0.2993 0.0862 -0.9377>}
 dyn_shoulderPad_l_04 {parent=base_shoulderPad_l_04, loc=<0 0 0>, rot=<0 0 0 1>}
 base_shoulderPad_l_01 {parent=clavicle_l, loc=<8.0696 6.9938 -1.6135>, rot=<0.8723 0.2165 -0.2957 -0.3236>}
 dyn_shoulderPad_l_01 {parent=base_shoulderPad_l_01, loc=<0 0 0>, rot=<0 0 0 1>}
 base_shoulderPad_l_05 {parent=clavicle_l, loc=<9.5881 -7.4827 -2.3724>, rot=<0.0716 -0.3834 0.0506 -0.9194>}
 dyn_shoulderPad_l_05 {parent=base_shoulderPad_l_05, loc=<0 0 0>, rot=<0 0 0 1>}
 base_cape_l_00 {parent=clavicle_l, loc=<11.158 -7.1525 -3.0942>, rot=<0.7509 0.0109 -0.6558 0.0773>}
 dyn_cape_l_00 {parent=base_cape_l_00, loc=<0 0 0>, rot=<0 0 0 1>}
 dyn_cape_l_01 {parent=dyn_cape_l_00, loc=<4.2929 0 0>, rot=<0 0 0 1>}
 dyn_cape_l_02 {parent=dyn_cape_l_01, loc=<21 0 0>, rot=<0 0 -0.0054 -1>}
 dyn_cape_l_03 {parent=dyn_cape_l_02, loc=<20.9988 -0.2273 0>, rot=<0 0 0 1>}
 dyn_cape_l_04 {parent=dyn_cape_l_03, loc=<21 0.0518 0>, rot=<0 0 0 1>}
 dyn_cape_l_05 {parent=dyn_cape_l_04, loc=<21.9995 0 0>, rot=<0 0 0 1>}
 clavicle_r {parent=spine_05, loc=<5.4831 -0.1053 3.2946>, rot=<0.7289 0.0465 -0.6829 0.009>}
 upperarm_r {parent=clavicle_r, loc=<-9.8658 1.2136 0.4105>, rot=<0.0164 -0.3245 -0.0269 -0.9454>}
 lowerarm_r {parent=upperarm_r, loc=<-26.465 0.8345 -0.259>, rot=<0.0036 0.0161 -0.2298 -0.9731>}
 hand_r {parent=lowerarm_r, loc=<-22.7355 0.3203 0.6106>, rot=<-0.6784 -0.0223 0.0735 -0.7307>}
 thumb_01_r {parent=hand_r, loc=<-1.3118 0.3191 2.459>, rot=<0.6546 -0.4042 0.0664 -0.6354>}
 thumb_02_r {parent=thumb_01_r, loc=<-5.9336 0.0448 -0.0401>, rot=<-0.0037 0.0213 0.298 -0.9543>}
 thumb_03_r {parent=thumb_02_r, loc=<-2.2223 -0.3682 0.0907>, rot=<0.0118 -0.013 0.0066 -0.9998>}
 index_metacarpal_r {parent=hand_r, loc=<-1.6535 -0.0183 2.96>, rot=<0.0158 -0.0124 -0.0173 -0.9996>}
 index_01_r {parent=index_metacarpal_r, loc=<-8.3008 -0.0357 0.0524>, rot=<-0.0044 0.0485 0.0924 -0.9945>}
 index_02_r {parent=index_01_r, loc=<-3.3587 0.3004 -0.0452>, rot=<-0.002 0.0027 0.3225 -0.9466>}
 index_03_r {parent=index_02_r, loc=<-2.5578 0.123 0.0252>, rot=<-0.002 0.0067 0.2558 -0.9667>}
 middle_metacarpal_r {parent=hand_r, loc=<-2.1856 0.1725 0.5984>, rot=<0.006 0.0168 -0.0329 -0.9993>}
 middle_01_r {parent=middle_metacarpal_r, loc=<-7.3275 0.2447 -0.0129>, rot=<-0.0208 0.0042 0.0858 -0.9961>}
 middle_02_r {parent=middle_01_r, loc=<-4.0123 -0.1401 -0.1175>, rot=<-0.0218 0.0258 0.3512 -0.9357>}
 middle_03_r {parent=middle_02_r, loc=<-2.6077 -0.0534 -0.1373>, rot=<0.0082 0.0188 0.1708 -0.9851>}
 ring_metacarpal_r {parent=hand_r, loc=<-2.2074 0.3625 -1.3638>, rot=<-0.0284 0.0722 -0.0458 -0.9959>}
 ring_01_r {parent=ring_metacarpal_r, loc=<-7.0295 0.6197 0.1446>, rot=<-0.0489 -0.0252 0.1892 -0.9804>}
 ring_02_r {parent=ring_01_r, loc=<-3.6174 -0.4647 0.1027>, rot=<0 0 0.2806 -0.9598>}
 ring_03_r {parent=ring_02_r, loc=<-2.4152 -0.114 -0.0199>, rot=<0 0 0.1189 -0.9929>}
 pinky_metacarpal_r {parent=hand_r, loc=<-1.7616 0.9753 -3.3814>, rot=<-0.0397 0.0814 -0.069 -0.9935>}
 pinky_01_r {parent=pinky_metacarpal_r, loc=<-7.1709 0.4831 0.1094>, rot=<-0.0654 -0.0035 0.2242 -0.9723>}
 pinky_02_r {parent=pinky_01_r, loc=<-2.8274 -0.0893 0.1043>, rot=<0.0037 0.0011 0.2854 -0.9584>}
 pinky_03_r {parent=pinky_02_r, loc=<-2.2046 -0.0307 0.0078>, rot=<-0.0002 0.0098 0.1445 -0.9895>}
 deform_wrist_dn_base_r {parent=hand_r, loc=<-0.0001 1.5328 0>, rot=<0 0 0 -1>}
 deform_wrist_dn_r {parent=deform_wrist_dn_base_r, loc=<0 0 0>, rot=<0 0 0 1>}
 deform_wrist_up_base_r {parent=hand_r, loc=<-0.0002 -2.3007 -0.0001>, rot=<0 0 0 -1>}
 deform_wrist_up_r {parent=deform_wrist_up_base_r, loc=<0 0 0>, rot=<0 0 0 1>}
 lowerarm_twist_02_r {parent=lowerarm_r, loc=<-10.6707 0 0>, rot=<0 0 0 -1>}
 lowerarm_twist_01_r {parent=lowerarm_r, loc=<-18.316 0 0>, rot=<0 0 0 -1>}
 deform_elbow_in_base_r {parent=lowerarm_r, loc=<-0.1834 -3.6062 0.2572>, rot=<0 0 0 -1>}
 deform_elbow_in_r {parent=deform_elbow_in_base_r, loc=<0 0 0>, rot=<0 0 0 1>}
 deform_elbow_out_base_r {parent=lowerarm_r, loc=<0 4 -0.0001>, rot=<0 0 0 -1>}
 deform_elbow_out_r {parent=deform_elbow_out_base_r, loc=<0 0 0>, rot=<0 0 0 1>}
 upperarm_twist_01_r {parent=upperarm_r, loc=<-7.2135 0.1134 0>, rot=<0 0 0 -1>}
 deform_upperarm_bk_base_r {parent=upperarm_twist_01_r, loc=<6.4053 4.3746 -0.314>, rot=<0 0 0 -1>}
 deform_upperarm_bk_r {parent=deform_upperarm_bk_base_r, loc=<0 0 0>, rot=<0 0 0 1>}
 deform_upperarm_fr_base_r {parent=upperarm_twist_01_r, loc=<5.6312 -5.092 -0.3354>, rot=<0 0 0 -1>}
 deform_upperarm_fr_r {parent=deform_upperarm_fr_base_r, loc=<0 0 0>, rot=<0 0 0 1>}
 deform_upperarm_up_base_r {parent=upperarm_twist_01_r, loc=<8.891 -0.4825 -3.9517>, rot=<0 0 0 -1>}
 deform_upperarm_up_r {parent=deform_upperarm_up_base_r, loc=<0 0 0>, rot=<0 0 0 1>}
 upperarm_twist_02_r {parent=upperarm_r, loc=<-19.7998 0.1134 0>, rot=<0 0 0 -1>}
 base_shoulderPad_r_02 {parent=clavicle_r, loc=<-12.2306 -3.9824 -2.6455>, rot=<0.8939 0.1728 -0.1233 -0.3948>}
 dyn_shoulderPad_r_02 {parent=base_shoulderPad_r_02, loc=<0 0 0>, rot=<0 0 0 1>}
 base_shoulderPad_r_03 {parent=clavicle_r, loc=<-13.8323 1.6765 -5.4322>, rot=<0.7071 0.0415 0.0059 -0.7059>}
 dyn_shoulderPad_r_03 {parent=base_shoulderPad_r_03, loc=<0 0 0>, rot=<0 0 0 1>}
 base_shoulderPad_r_04 {parent=clavicle_r, loc=<-12.602 6.4159 -2.0761>, rot=<0.1542 -0.2993 0.0862 -0.9377>}
 dyn_shoulderPad_r_04 {parent=base_shoulderPad_r_04, loc=<0 0 0>, rot=<0 0 0 -1>}
 base_shoulderPad_r_01 {parent=clavicle_r, loc=<-8.0696 -6.9938 1.6133>, rot=<0.8723 0.2165 -0.2957 -0.3236>}
 dyn_shoulderPad_r_01 {parent=base_shoulderPad_r_01, loc=<0 0 0>, rot=<0 0 0 -1>}
 base_shoulderPad_r_05 {parent=clavicle_r, loc=<-9.5881 7.4827 2.3721>, rot=<0.0716 -0.3834 0.0506 -0.9194>}
 dyn_shoulderPad_r_05 {parent=base_shoulderPad_r_05, loc=<0 0 0>, rot=<0 0 0 1>}
 base_cape_r_00 {parent=clavicle_r, loc=<-11.1586 7.1577 3.0939>, rot=<-0.0109 0.7509 -0.0773 -0.6558>}
 dyn_cape_r_00 {parent=base_cape_r_00, loc=<0 0 0>, rot=<0 0 0 1>}
 dyn_cape_r_01 {parent=dyn_cape_r_00, loc=<4.2931 0 0>, rot=<0.0087 0 0 -1>}
 dyn_cape_r_02 {parent=dyn_cape_r_01, loc=<20.9994 0 0>, rot=<0 0 -0.0054 -1>}
 dyn_cape_r_03 {parent=dyn_cape_r_02, loc=<20.999 -0.2273 0>, rot=<0 0 0 1>}
 dyn_cape_r_04 {parent=dyn_cape_r_03, loc=<20.9999 0.0518 0>, rot=<0 0 0 1>}
 dyn_cape_r_05 {parent=dyn_cape_r_04, loc=<21.9995 0 0>, rot=<0 0 0 1>}
 attach_backpack {parent=spine_05, loc=<-0.123 -13.473 0>, rot=<-0.0093 -0.707 0.0093 -0.707>}
 Backpack_BR {parent=attach_backpack, type=Null, loc=<0 0 0>, rot=<0 0 0 1>}
 attach_cape {parent=spine_05, loc=<9.017 -11.714 0>, rot=<0.0616 -0.7044 -0.0616 -0.7044>}
 cape {parent=attach_cape, type=Null, loc=<0 0 0>, rot=<0 0 0 1>}
 deform_pec_aim_base_l {parent=spine_04, loc=<3.2456 9.3309 -1.2237>, rot=<0.0603 -0.4033 0.3318 -0.8507>}
 deform_pec_aim_l {parent=deform_pec_aim_base_l, loc=<0 0 0>, rot=<0 0 0 1>}
 deform_pec_base_l {parent=deform_pec_aim_l, loc=<9.3665 0 0>, rot=<0 0 0 1>}
 deform_pec_l {parent=deform_pec_base_l, loc=<0 0 0>, rot=<0 0 0 1>}
 deform_pec_aim_base_r {parent=spine_04, loc=<3.2452 9.3309 1.2237>, rot=<-0.4033 -0.0603 0.8507 0.3318>}
 deform_pec_aim_r {parent=deform_pec_aim_base_r, loc=<0 0 0>, rot=<0 0 0 1>}
 deform_pec_base_r {parent=deform_pec_aim_r, loc=<-9.3667 -0.0003 -0.0004>, rot=<0 0 0 1>}
 deform_pec_r {parent=deform_pec_base_r, loc=<0 0 0>, rot=<0 0 0 1>}
 thigh_l {parent=pelvis, loc=<-1.0855 -0.2109 -8.7053>, rot=<0.0009 0.0355 0.0155 -0.9993>}
 calf_l {parent=thigh_l, loc=<-42.5284 -0.0019 -0.7297>, rot=<0.0006 -0.0161 -0.0391 -0.9991>}
 foot_l {parent=calf_l, loc=<-46.7401 -2.5014 -0.592>, rot=<0.0427 -0.0238 0.0353 -0.9982>}
 ball_l {parent=foot_l, loc=<-3.5658 10.8066 -0.1544>, rot=<0.0001 -0.0001 -0.7063 -0.7079>}
 calf_twist_02_l {parent=calf_l, loc=<-19.1546 -1.7943 -0.2638>, rot=<0.0237 0.0095 -0.0035 -0.9997>}
 calf_twist_01_l {parent=calf_l, loc=<-35.8483 -1.9197 -0.5794>, rot=<0.0237 0.0095 -0.0035 -0.9997>}
 deform_knee_in_base_l {parent=calf_l, loc=<0.8031 -5.0231 0.2456>, rot=<0 0 0 -1>}
 deform_knee_in_l {parent=deform_knee_in_base_l, loc=<0 0 0>, rot=<0 0 0 1>}
 deform_knee_out_base_l {parent=calf_l, loc=<0 4 0>, rot=<0 0 0 -1>}
 deform_knee_out_l {parent=deform_knee_out_base_l, loc=<0 0 0>, rot=<0 0 0 1>}
 thigh_twist_01_l {parent=thigh_l, loc=<-27.952 0 0>, rot=<0.0013 0 0 -1>}
 deform_groin_base_l {parent=thigh_l, loc=<0.4801 5.6776 -0.2486>, rot=<0 0 0 -1>}
 deform_groin_l {parent=deform_groin_base_l, loc=<0 0 0>, rot=<0 0 0 1>}
 thigh_r {parent=pelvis, loc=<-1.0855 -0.2109 8.7053>, rot=<0.0355 -0.0009 0.9993 0.0155>}
 calf_r {parent=thigh_r, loc=<42.5284 0.0019 0.7297>, rot=<0.0006 -0.0161 -0.0391 -0.9991>}
 foot_r {parent=calf_r, loc=<46.7401 2.5014 0.592>, rot=<0.0427 -0.0238 0.0353 -0.9982>}
 ball_r {parent=foot_r, loc=<3.5658 -10.8066 0.1544>, rot=<0.0001 -0.0001 -0.7063 -0.7079>}
 calf_twist_02_r {parent=calf_r, loc=<19.1546 1.7943 0.2638>, rot=<0.0237 0.0095 -0.0035 -0.9997>}
 calf_twist_01_r {parent=calf_r, loc=<35.8483 1.9197 0.5794>, rot=<0.0237 0.0095 -0.0035 -0.9997>}
 deform_knee_in_base_r {parent=calf_r, loc=<-0.8031 5.0231 -0.2457>, rot=<0 0 0 -1>}
 deform_knee_in_r {parent=deform_knee_in_base_r, loc=<0 0 0>, rot=<0 0 0 1>}
 deform_knee_out_base_r {parent=calf_r, loc=<0 -4 0>, rot=<0 0 0 -1>}
 deform_knee_out_r {parent=deform_knee_out_base_r, loc=<0 0 0>, rot=<0 0 0 1>}
 thigh_twist_01_r {parent=thigh_r, loc=<27.952 0 0>, rot=<0.0013 0 0 -1>}
 deform_groin_base_r {parent=thigh_r, loc=<-0.4801 -5.6776 0.2486>, rot=<0 0 0 -1>}
 deform_groin_r {parent=deform_groin_base_r, loc=<0 0 0>, rot=<0 0 0 1>}
 deform_glute_base_l {parent=pelvis, loc=<-7.3466 -9.0213 -9.1245>, rot=<0.0172 0.031 -0.4477 -0.8935>}
 deform_glute_l {parent=deform_glute_base_l, loc=<0 0 0>, rot=<0 0 0 1>}
 deform_glute_base_r {parent=pelvis, loc=<-7.3464 -9.0213 9.1245>, rot=<0.031 -0.0172 0.8935 -0.4477>}
 deform_glute_r {parent=deform_glute_base_r, loc=<0 0 0>, rot=<0 0 0 1>}
]
variables [
 Crouching high position {type=double, default=0}
 "Head Right Crouching " {type=double, default=0}
 LeftCalf Bend Alpha {type=double, default=0}
 LeftLeg Bwd Alpha {type=double, default=0}
 LeftLeg Fwd Alpha {type=double, default=0}
 LeftLeg Fwd Extreme Alpha {type=double, default=0}
 LeftLeg Out Alpha {type=double, default=0}
 RightCalf Bend Alpha {type=double, default=0}
 RightLeg Bwd Alpha {type=double, default=0}
 RightLeg Fwd Alpha {type=double, default=0}
 RightLeg Fwd Extreme Alpha {type=double, default=0}
 RightLeg Out Alpha {type=double, default=0}
 is Baseball Sliding {type=double, default=0}
 is Crouching {type=double, default=0}
 "is Knee Sliding " {type=double, default=0}
 is Moving 2D {type=double, default=0}
]
