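# Physics asset for the front skirt of the F_MED_OutlineLove_Flank female player body, built entirely from capsule bodies.
# A five-link front chain hangs from a small base body, with swing and twist limits generally loosening toward the tip.
# Left and right three-link side chains have tight twist and slightly differing swing limits, with no constraint on their top links.
# Unconstrained thigh and calf capsules provide leg shapes alongside the skirt chains.
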PhysicsAsset /BRCosmetics/Characters/Player/Female/Medium/Bodies/F_MED_OutlineLove_Flank/Meshes/F_MED_OutlineLove_Flank_Front_Physics
bodies {
 base_skirt_f_01 [{shapeType=Capsule, length=0.505, radius=0.505, shapeName=base_skirt_f_01_capsule, rotation=<rot 0 -90 0>}]
 calf_l [{
   shapeType=Capsule
   length=46.9381
   radius=9.0292
   shapeName=calf_r_capsule
   center=<-26.3973 -2.826 -0.2188>
   rotation=<rot 89.5737 179.9985 179.9988>
  }]
 calf_r [{
   shapeType=Capsule
   length=46.9381
   radius=9.0292
   shapeName=calf_r_capsule
   center=<26.3973 2.826 -0.2392>
   rotation=<rot 89.5737 179.9985 179.9988>
  }]
 dyn_skirt_f_01 [{
   shapeType=Capsule
   length=2.3461
   radius=2.8544
   shapeName=dyn_skirt_f_01_capsule
   center=<4.0348 -2.54 0.0209>
   rotation=<rot -89.4121 -0.0713 0>
  }]
 dyn_skirt_f_02 [{shapeType=Capsule, length=6.2125, radius=3.7839, shapeName=dyn_skirt_f_02_capsule, center=<5.1943 -2.54 0>}]
 dyn_skirt_f_03 [{shapeType=Capsule, length=4.5001, radius=4.4787, shapeName=dyn_skirt_f_03_capsule, center=<2.8929 -2.54 0>}]
 dyn_skirt_f_04 [{shapeType=Capsule, length=5.2235, radius=4.7481, shapeName=dyn_skirt_f_04_capsule, center=<3.3192 -2.54 0>}]
 dyn_skirt_f_05 [{shapeType=Capsule, length=7.4424, radius=4.22, shapeName=dyn_skirt_f_05_capsule, center=<2.2593 -2.54 0>}]
 dyn_skirt_l_1_02 [{shapeType=Capsule, length=9.4302, radius=7.0412, shapeName=dyn_skirt_l_1_02_capsule, center=<8.1134 0 -2.9057>, rotation=<rot 90 0 0>}]
 dyn_skirt_l_1_03 [{shapeType=Capsule, length=9.4302, radius=7.0412, shapeName=dyn_skirt_l_1_03_capsule, center=<8.1134 0 -2.4193>, rotation=<rot 90 0 0>}]
 dyn_skirt_l_1_04 [{shapeType=Capsule, length=9.4302, radius=7.0412, shapeName=dyn_skirt_l_1_04_capsule, center=<8.1134 0 -2.2871>, rotation=<rot 90 0 0>}]
 dyn_skirt_r_1_02 [{shapeType=Capsule, length=9.4302, radius=7.0412, shapeName=dyn_skirt_l_1_02_capsule, center=<8.1134 0 2.9057>, rotation=<rot 90 0 0>}]
 dyn_skirt_r_1_03 [{shapeType=Capsule, length=9.4302, radius=7.0412, shapeName=dyn_skirt_l_1_03_capsule, center=<8.1134 0 2.4193>, rotation=<rot 90 0 0>}]
 dyn_skirt_r_1_04 [{shapeType=Capsule, length=9.4302, radius=7.0412, shapeName=dyn_skirt_l_1_04_capsule, center=<8.1134 0 2.2871>, rotation=<rot 90 0 0>}]
 thigh_l [{
   shapeType=Capsule
   length=36.1478
   radius=8.9083
   shapeName=thigh_r_capsule
   center=<-21.8419 -0.8269 0.6041>
   rotation=<rot 87.6898 179.7339 179.8332>
  }]
 thigh_r [{
   shapeType=Capsule
   length=36.1478
   radius=8.9083
   shapeName=thigh_r_capsule
   center=<21.8419 0.8269 -0.6041>
   rotation=<rot 87.6898 179.7339 179.8332>
  }]
}
constraints [
 {bone1Name=dyn_skirt_f_01, bone2Name=base_skirt_f_01, swing1LimitDegrees=30, swing2LimitDegrees=30, twistLimitDegrees=1}
 {bone1Name=dyn_skirt_f_02, bone2Name=dyn_skirt_f_01, swing1LimitDegrees=30, swing2LimitDegrees=30, twistLimitDegrees=1}
 {bone1Name=dyn_skirt_f_03, bone2Name=dyn_skirt_f_02, swing1LimitDegrees=45, swing2LimitDegrees=40, twistLimitDegrees=15}
 {bone1Name=dyn_skirt_f_04, bone2Name=dyn_skirt_f_03, swing1LimitDegrees=45, swing2LimitDegrees=50, twistLimitDegrees=15}
 {bone1Name=dyn_skirt_f_05, bone2Name=dyn_skirt_f_04, swing1LimitDegrees=45, swing2LimitDegrees=60, twistLimitDegrees=15}
 {bone1Name=dyn_skirt_l_1_03, bone2Name=dyn_skirt_l_1_02, swing1LimitDegrees=30, swing2LimitDegrees=15, twistLimitDegrees=1}
 {bone1Name=dyn_skirt_l_1_04, bone2Name=dyn_skirt_l_1_03, swing1LimitDegrees=40, swing2LimitDegrees=17, twistLimitDegrees=1}
 {bone1Name=dyn_skirt_r_1_03, bone2Name=dyn_skirt_r_1_02, swing1LimitDegrees=25, swing2LimitDegrees=15, twistLimitDegrees=1}
 {bone1Name=dyn_skirt_r_1_04, bone2Name=dyn_skirt_r_1_03, swing1LimitDegrees=35, swing2LimitDegrees=25, twistLimitDegrees=1}
]
