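# Physics asset for the F_MED_OutlineLove_Flank female body, set up to simulate its skirt.
# Eight five-bone capsule chains, four per side, hang from small base bodies constrained to the pelvis, with locked roots and swing limits that generally widen downward.
# An angularly free constraint links the left and right fourth chains near their ends.
# Pelvis and leg capsules, plus a huge ground_mover capsule approximating a floor, provide collision shapes.
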PhysicsAsset /BRCosmetics/Characters/Player/Female/Medium/Bodies/F_MED_OutlineLove_Flank/Meshes/F_MED_OutlineLove_Flank_Physics
bodies {
 base_skirt_l_1_01 [{shapeType=Capsule, length=0.505, radius=0.505, shapeName=base_skirt_l_1_01_capsule, rotation=<rot 0 -90 0>}]
 base_skirt_l_2_01 [{shapeType=Capsule, length=0.505, radius=0.505, shapeName=base_skirt_l_2_01_capsule, rotation=<rot 0 -90 0>}]
 base_skirt_l_3_01 [{shapeType=Capsule, length=0.505, radius=0.505, shapeName=base_skirt_l_3_01_capsule, rotation=<rot 0 -90 0>}]
 base_skirt_l_4_01 [{shapeType=Capsule, length=0.505, radius=0.505, shapeName=base_skirt_l_4_01_capsule, rotation=<rot 0 -90 0>}]
 base_skirt_r_1_01 [{shapeType=Capsule, length=0.505, radius=0.505, shapeName=base_skirt_r_1_01_capsule, rotation=<rot 0 -90 0>}]
 base_skirt_r_2_01 [{shapeType=Capsule, length=0.505, radius=0.505, shapeName=base_skirt_r_2_01_capsule, rotation=<rot 0 -90 0>}]
 base_skirt_r_3_01 [{shapeType=Capsule, length=0.505, radius=0.505, shapeName=base_skirt_r_3_01_capsule, rotation=<rot 0 -90 0>}]
 base_skirt_r_4_01 [{shapeType=Capsule, length=0.505, radius=0.505, shapeName=base_skirt_r_4_01_capsule, rotation=<rot 0 -90 0>}]
 calf_l [{
   shapeType=Capsule
   length=46.9381
   radius=9.0292
   shapeName=calf_r_capsule
   center=<-26.3973 -2.826 -0.2188>
   rotation=<rot 89.5737 179.9985 179.9988>
  }]
 calf_r [{
   shapeType=Capsule
   length=46.9381
   radius=9.0292
   shapeName=calf_r_capsule
   center=<26.3973 2.826 -0.2392>
   rotation=<rot 89.5737 179.9985 179.9988>
  }]
 dyn_skirt_l_1_01 [{shapeType=Capsule, length=11.0348, radius=7.3357, shapeName=dyn_skirt_l_1_01_capsule, center=<5.4098 -5.5624 0>}]
 dyn_skirt_l_1_02 [{shapeType=Capsule, length=14.8745, radius=12.3428, shapeName=dyn_skirt_l_1_02_capsule, center=<7.6471 -12.2 0>}]
 dyn_skirt_l_1_03 [{shapeType=Capsule, length=12.6919, radius=12.5985, shapeName=dyn_skirt_l_1_03_capsule, center=<7.8055 -11.8 0>}]
 dyn_skirt_l_1_04 [{shapeType=Capsule, length=13.3964, radius=12.4116, shapeName=dyn_skirt_l_1_04_capsule, center=<7.7495 -9.9 0>}]
 dyn_skirt_l_1_05 [{shapeType=Capsule, length=13.5124, radius=12.647, shapeName=dyn_skirt_l_1_05_capsule, center=<6.8887 -9.5 0>}]
 dyn_skirt_l_2_01 [{shapeType=Capsule, length=11.0348, radius=7.3357, shapeName=dyn_skirt_l_2_01_capsule, center=<5.4098 -5.5624 0>}]
 dyn_skirt_l_2_02 [{shapeType=Capsule, length=12.0926, radius=12.0267, shapeName=dyn_skirt_l_2_02_capsule, center=<8.9763 -12.55 0>}]
 dyn_skirt_l_2_03 [{shapeType=Capsule, length=13.2083, radius=12.1479, shapeName=dyn_skirt_l_2_03_capsule, center=<9.8316 -11.9 0>}]
 dyn_skirt_l_2_04 [{shapeType=Capsule, length=13.6187, radius=12.3244, shapeName=dyn_skirt_l_2_04_capsule, center=<10.2924 -12 0>}]
 dyn_skirt_l_2_05 [{shapeType=Capsule, length=13.2422, radius=12.346, shapeName=dyn_skirt_l_2_05_capsule, center=<8.9849 -10.15 0>}]
 dyn_skirt_l_3_01 [{shapeType=Capsule, length=11.0348, radius=7.3357, shapeName=dyn_skirt_l_3_01_capsule, center=<5.4098 -5.5624 0>}]
 dyn_skirt_l_3_02 [{shapeType=Capsule, length=22.8545, radius=12.8235, shapeName=dyn_skirt_l_3_02_capsule, center=<8.4887 -12 -1.3>, rotation=<rot 0 0 -6.3>}]
 dyn_skirt_l_3_03 [{shapeType=Capsule, length=22.8545, radius=12.6858, shapeName=dyn_skirt_l_3_03_capsule, center=<8.8495 -11.3 -1.25>, rotation=<rot 0 0 -6.3>}]
 dyn_skirt_l_3_04 [{shapeType=Capsule, length=22.8545, radius=12.6633, shapeName=dyn_skirt_l_3_04_capsule, center=<9.8046 -12.7 -1.4>, rotation=<rot 0 0 -6.3>}]
 dyn_skirt_l_3_05 [{shapeType=Capsule, length=22.8545, radius=12.3449, shapeName=dyn_skirt_l_3_05_capsule, center=<7.7552 -11.4 0>, rotation=<rot 0 0 -6.3>}]
 dyn_skirt_l_4_01 [{shapeType=Capsule, length=11.0348, radius=7.3357, shapeName=dyn_skirt_l_4_01_capsule, center=<5.4098 -5.9281 0>}]
 dyn_skirt_l_4_02 [{shapeType=Capsule, length=20.4117, radius=12.356, shapeName=dyn_skirt_l_4_02_capsule, center=<7.4512 -10.64 0>}]
 dyn_skirt_l_4_03 [{shapeType=Capsule, length=20.4117, radius=12.5028, shapeName=dyn_skirt_l_4_03_capsule, center=<9.1806 -9.9 0>}]
 dyn_skirt_l_4_04 [{shapeType=Capsule, length=20.4117, radius=12.905, shapeName=dyn_skirt_l_4_04_capsule, center=<9.5127 -10.67 0>}]
 dyn_skirt_l_4_05 [{shapeType=Capsule, length=22.6, radius=12.5, shapeName=dyn_skirt_l_4_05_capsule, center=<9.4858 -9.8 0>}]
 dyn_skirt_r_1_01 [{shapeType=Capsule, length=11.0348, radius=7.3357, shapeName=dyn_skirt_r_1_01_capsule, center=<5.4098 -5.5624 0>}]
 dyn_skirt_r_1_02 [{shapeType=Capsule, length=14.8745, radius=12.3428, shapeName=dyn_skirt_l_1_02_capsule, center=<7.6471 -12.2 0>}]
 dyn_skirt_r_1_03 [{shapeType=Capsule, length=12.6919, radius=12.5985, shapeName=dyn_skirt_l_1_03_capsule, center=<7.8055 -11.8 0>}]
 dyn_skirt_r_1_04 [{shapeType=Capsule, length=13.3964, radius=12.4116, shapeName=dyn_skirt_l_1_04_capsule, center=<7.7495 -9.9 0>}]
 dyn_skirt_r_1_05 [{shapeType=Capsule, length=13.5124, radius=12.647, shapeName=dyn_skirt_l_1_05_capsule, center=<6.8887 -9.5 0>}]
 dyn_skirt_r_2_01 [{shapeType=Capsule, length=11.0348, radius=7.3357, shapeName=dyn_skirt_r_2_01_capsule, center=<5.4098 -5.5624 0>}]
 dyn_skirt_r_2_02 [{shapeType=Capsule, length=12.0926, radius=12.0267, shapeName=dyn_skirt_l_2_02_capsule, center=<8.9763 -12.55 0>}]
 dyn_skirt_r_2_03 [{shapeType=Capsule, length=13.2083, radius=12.1479, shapeName=dyn_skirt_l_2_03_capsule, center=<9.8316 -11.9 0>}]
 dyn_skirt_r_2_04 [{shapeType=Capsule, length=13.6187, radius=12.3244, shapeName=dyn_skirt_l_2_04_capsule, center=<10.2924 -12 0>}]
 dyn_skirt_r_2_05 [{shapeType=Capsule, length=13.2422, radius=12.346, shapeName=dyn_skirt_l_2_05_capsule, center=<8.9849 -10.15 0>}]
 dyn_skirt_r_3_01 [{shapeType=Capsule, length=11.0348, radius=7.3357, shapeName=dyn_skirt_r_3_01_capsule, center=<5.4098 -5.5624 0>}]
 dyn_skirt_r_3_02 [{shapeType=Capsule, length=22.8545, radius=12.8235, shapeName=dyn_skirt_l_3_02_capsule, center=<8.4887 -12 -1.3>, rotation=<rot 0 0 -6.3>}]
 dyn_skirt_r_3_03 [{shapeType=Capsule, length=22.8545, radius=12.6858, shapeName=dyn_skirt_l_3_03_capsule, center=<8.8495 -11.3 -1.25>, rotation=<rot 0 0 -6.3>}]
 dyn_skirt_r_3_04 [{shapeType=Capsule, length=22.8545, radius=12.6633, shapeName=dyn_skirt_l_3_04_capsule, center=<9.8046 -12.7 -1.4>, rotation=<rot 0 0 -6.3>}]
 dyn_skirt_r_3_05 [{shapeType=Capsule, length=22.8545, radius=12.3449, shapeName=dyn_skirt_l_3_05_capsule, center=<7.7552 -11.4 0>, rotation=<rot 0 0 -6.3>}]
 dyn_skirt_r_4_01 [{shapeType=Capsule, length=11.0348, radius=7.3357, shapeName=dyn_skirt_r_4_01_capsule, center=<5.4098 -5.9281 0>}]
 dyn_skirt_r_4_02 [{shapeType=Capsule, length=20.4117, radius=12.356, shapeName=dyn_skirt_l_4_02_capsule, center=<7.4512 -10.64 0>}]
 dyn_skirt_r_4_03 [{shapeType=Capsule, length=20.4117, radius=12.5028, shapeName=dyn_skirt_l_4_03_capsule, center=<9.1806 -9.9 0>}]
 dyn_skirt_r_4_04 [{shapeType=Capsule, length=20.4117, radius=12.905, shapeName=dyn_skirt_l_4_04_capsule, center=<9.5127 -10.67 0>}]
 dyn_skirt_r_4_05 [{shapeType=Capsule, length=22.6, radius=12.5, shapeName=dyn_skirt_l_4_05_capsule, center=<9.4858 -9.8 0>}]
 foot_l [{
   shapeType=Capsule
   length=14.2974
   radius=9.6069
   shapeName=foot_r_capsule
   center=<-2.4346 6.0548 -0.2885>
   rotation=<rot 0 6.9743 -89.7415>
  }]
 foot_r [{shapeType=Capsule, length=14.2974, radius=9.6069, shapeName=foot_r_capsule, center=<2.4346 -6.0548 0.2885>, rotation=<rot 0 6.9743 -89.7415>}]
 ground_mover [{shapeType=Capsule, length=0.5056, radius=337.6725, shapeName=floor_joint_capsule, center=<2.6954 -3.0549 -337.3088>, rotation=<rot 0 90 -90>}]
 pelvis [{
   shapeType=Capsule
   length=8.8492
   radius=8.3337
   shapeName=pelvis_capsule
   center=<-0.4636 -1.3543 -0.0343>
   rotation=<rot 2.0206 -90 0.0573>
  }]
 thigh_l [{
   shapeType=Capsule
   length=36.1478
   radius=8.9083
   shapeName=thigh_r_capsule
   center=<-21.8419 -0.8269 0.6041>
   rotation=<rot 87.6898 179.7339 179.8332>
  }]
 thigh_r [{
   shapeType=Capsule
   length=36.1478
   radius=8.9083
   shapeName=thigh_r_capsule
   center=<21.8419 0.8269 -0.6041>
   rotation=<rot 87.6898 179.7339 179.8332>
  }]
}
constraints [
 {
  bone1Name=base_skirt_l_1_01
  bone2Name=pelvis
  swing1LimitDegrees=45
  swing1Motion=Free
  swing2LimitDegrees=45
  swing2Motion=Free
  twistLimitDegrees=45
  twistMotion=Free
 }
 {
  bone1Name=dyn_skirt_l_1_01
  bone2Name=base_skirt_l_1_01
  swing1LimitDegrees=2
  swing1Motion=Locked
  swing2LimitDegrees=10
  swing2Motion=Locked
  twistLimitDegrees=1
  twistMotion=Locked
 }
 {bone1Name=dyn_skirt_l_1_02, bone2Name=dyn_skirt_l_1_01, swing1LimitDegrees=10, swing2LimitDegrees=5, twistLimitDegrees=1}
 {bone1Name=dyn_skirt_l_1_03, bone2Name=dyn_skirt_l_1_02, swing1LimitDegrees=30, swing2LimitDegrees=15, twistLimitDegrees=1}
 {bone1Name=dyn_skirt_l_1_04, bone2Name=dyn_skirt_l_1_03, swing1LimitDegrees=40, swing2LimitDegrees=17, twistLimitDegrees=1}
 {bone1Name=dyn_skirt_l_1_05, bone2Name=dyn_skirt_l_1_04, swing1LimitDegrees=30, swing2LimitDegrees=20, twistLimitDegrees=1}
 {
  bone1Name=base_skirt_l_2_01
  bone2Name=pelvis
  swing1LimitDegrees=45
  swing1Motion=Free
  swing2LimitDegrees=45
  swing2Motion=Free
  twistLimitDegrees=45
  twistMotion=Free
 }
 {
  bone1Name=dyn_skirt_l_2_01
  bone2Name=base_skirt_l_2_01
  swing1LimitDegrees=2
  swing1Motion=Locked
  swing2LimitDegrees=1
  swing2Motion=Locked
  twistLimitDegrees=1
  twistMotion=Locked
 }
 {bone1Name=dyn_skirt_l_2_02, bone2Name=dyn_skirt_l_2_01, swing1LimitDegrees=10, swing2LimitDegrees=10, twistLimitDegrees=1}
 {bone1Name=dyn_skirt_l_2_03, bone2Name=dyn_skirt_l_2_02, swing1LimitDegrees=15, swing2LimitDegrees=15, twistLimitDegrees=1}
 {bone1Name=dyn_skirt_l_2_04, bone2Name=dyn_skirt_l_2_03, swing1LimitDegrees=17, swing2LimitDegrees=17, twistLimitDegrees=1}
 {bone1Name=dyn_skirt_l_2_05, bone2Name=dyn_skirt_l_2_04, swing1LimitDegrees=20, swing2LimitDegrees=20, twistLimitDegrees=1}
 {
  bone1Name=base_skirt_l_3_01
  bone2Name=pelvis
  swing1LimitDegrees=45
  swing1Motion=Free
  swing2LimitDegrees=45
  swing2Motion=Free
  twistLimitDegrees=45
  twistMotion=Free
 }
 {
  bone1Name=dyn_skirt_l_3_01
  bone2Name=base_skirt_l_3_01
  swing1LimitDegrees=2
  swing1Motion=Locked
  swing2LimitDegrees=1
  swing2Motion=Locked
  twistLimitDegrees=1
  twistMotion=Locked
 }
 {bone1Name=dyn_skirt_l_3_02, bone2Name=dyn_skirt_l_3_01, swing1LimitDegrees=10, swing2LimitDegrees=10, twistLimitDegrees=1}
 {bone1Name=dyn_skirt_l_3_03, bone2Name=dyn_skirt_l_3_02, swing1LimitDegrees=15, swing2LimitDegrees=15, twistLimitDegrees=1}
 {bone1Name=dyn_skirt_l_3_04, bone2Name=dyn_skirt_l_3_03, swing1LimitDegrees=17, swing2LimitDegrees=17, twistLimitDegrees=1}
 {bone1Name=dyn_skirt_l_3_05, bone2Name=dyn_skirt_l_3_04, swing1LimitDegrees=20, swing2LimitDegrees=20, twistLimitDegrees=1}
 {
  bone1Name=base_skirt_l_4_01
  bone2Name=pelvis
  swing1LimitDegrees=45
  swing1Motion=Free
  swing2LimitDegrees=45
  swing2Motion=Free
  twistLimitDegrees=45
  twistMotion=Free
 }
 {
  bone1Name=dyn_skirt_l_4_01
  bone2Name=base_skirt_l_4_01
  swing1LimitDegrees=2
  swing1Motion=Locked
  swing2LimitDegrees=1
  swing2Motion=Locked
  twistLimitDegrees=1
  twistMotion=Locked
 }
 {bone1Name=dyn_skirt_l_4_02, bone2Name=dyn_skirt_l_4_01, swing1LimitDegrees=10, swing2LimitDegrees=10, twistLimitDegrees=1}
 {bone1Name=dyn_skirt_l_4_03, bone2Name=dyn_skirt_l_4_02, swing1LimitDegrees=15, swing2LimitDegrees=15, twistLimitDegrees=1}
 {bone1Name=dyn_skirt_l_4_04, bone2Name=dyn_skirt_l_4_03, swing1LimitDegrees=17, swing2LimitDegrees=17, twistLimitDegrees=1}
 {bone1Name=dyn_skirt_l_4_05, bone2Name=dyn_skirt_l_4_04, swing1LimitDegrees=20, swing2LimitDegrees=20, twistLimitDegrees=1}
 {
  bone1Name=base_skirt_r_1_01
  bone2Name=pelvis
  swing1LimitDegrees=45
  swing1Motion=Free
  swing2LimitDegrees=45
  swing2Motion=Free
  twistLimitDegrees=45
  twistMotion=Free
 }
 {
  bone1Name=dyn_skirt_r_1_01
  bone2Name=base_skirt_r_1_01
  swing1LimitDegrees=2
  swing1Motion=Locked
  swing2LimitDegrees=10
  swing2Motion=Locked
  twistLimitDegrees=1
  twistMotion=Locked
 }
 {bone1Name=dyn_skirt_r_1_02, bone2Name=dyn_skirt_r_1_01, swing1LimitDegrees=10, swing2LimitDegrees=5, twistLimitDegrees=1}
 {bone1Name=dyn_skirt_r_1_03, bone2Name=dyn_skirt_r_1_02, swing1LimitDegrees=25, swing2LimitDegrees=15, twistLimitDegrees=1}
 {bone1Name=dyn_skirt_r_1_04, bone2Name=dyn_skirt_r_1_03, swing1LimitDegrees=35, swing2LimitDegrees=25, twistLimitDegrees=1}
 {bone1Name=dyn_skirt_r_1_05, bone2Name=dyn_skirt_r_1_04, swing1LimitDegrees=30, swing2LimitDegrees=20, twistLimitDegrees=1}
 {
  bone1Name=base_skirt_r_2_01
  bone2Name=pelvis
  swing1LimitDegrees=45
  swing1Motion=Free
  swing2LimitDegrees=45
  swing2Motion=Free
  twistLimitDegrees=45
  twistMotion=Free
 }
 {
  bone1Name=dyn_skirt_r_2_01
  bone2Name=base_skirt_r_2_01
  swing1LimitDegrees=2
  swing1Motion=Locked
  swing2LimitDegrees=1
  swing2Motion=Locked
  twistLimitDegrees=1
  twistMotion=Locked
 }
 {bone1Name=dyn_skirt_r_2_02, bone2Name=dyn_skirt_r_2_01, swing1LimitDegrees=10, swing2LimitDegrees=10, twistLimitDegrees=1}
 {bone1Name=dyn_skirt_r_2_03, bone2Name=dyn_skirt_r_2_02, swing1LimitDegrees=15, swing2LimitDegrees=15, twistLimitDegrees=1}
 {bone1Name=dyn_skirt_r_2_04, bone2Name=dyn_skirt_r_2_03, swing1LimitDegrees=17, swing2LimitDegrees=17, twistLimitDegrees=1}
 {bone1Name=dyn_skirt_r_2_05, bone2Name=dyn_skirt_r_2_04, swing1LimitDegrees=20, swing2LimitDegrees=20, twistLimitDegrees=1}
 {
  bone1Name=base_skirt_r_3_01
  bone2Name=pelvis
  swing1LimitDegrees=45
  swing1Motion=Free
  swing2LimitDegrees=45
  swing2Motion=Free
  twistLimitDegrees=45
  twistMotion=Free
 }
 {
  bone1Name=dyn_skirt_r_3_01
  bone2Name=base_skirt_r_3_01
  swing1LimitDegrees=2
  swing1Motion=Locked
  swing2LimitDegrees=1
  swing2Motion=Locked
  twistLimitDegrees=1
  twistMotion=Locked
 }
 {bone1Name=dyn_skirt_r_3_02, bone2Name=dyn_skirt_r_3_01, swing1LimitDegrees=10, swing2LimitDegrees=10, twistLimitDegrees=1}
 {bone1Name=dyn_skirt_r_3_03, bone2Name=dyn_skirt_r_3_02, swing1LimitDegrees=15, swing2LimitDegrees=15, twistLimitDegrees=1}
 {bone1Name=dyn_skirt_r_3_04, bone2Name=dyn_skirt_r_3_03, swing1LimitDegrees=17, swing2LimitDegrees=17, twistLimitDegrees=1}
 {bone1Name=dyn_skirt_r_3_05, bone2Name=dyn_skirt_r_3_04, swing1LimitDegrees=20, swing2LimitDegrees=20, twistLimitDegrees=1}
 {
  bone1Name=base_skirt_r_4_01
  bone2Name=pelvis
  swing1LimitDegrees=45
  swing1Motion=Free
  swing2LimitDegrees=45
  swing2Motion=Free
  twistLimitDegrees=45
  twistMotion=Free
 }
 {
  bone1Name=dyn_skirt_r_4_01
  bone2Name=base_skirt_r_4_01
  swing1LimitDegrees=2
  swing1Motion=Locked
  swing2LimitDegrees=1
  swing2Motion=Locked
  twistLimitDegrees=1
  twistMotion=Locked
 }
 {bone1Name=dyn_skirt_r_4_02, bone2Name=dyn_skirt_r_4_01, swing1LimitDegrees=10, swing2LimitDegrees=10, twistLimitDegrees=1}
 {bone1Name=dyn_skirt_r_4_03, bone2Name=dyn_skirt_r_4_02, swing1LimitDegrees=15, swing2LimitDegrees=15, twistLimitDegrees=1}
 {bone1Name=dyn_skirt_r_4_04, bone2Name=dyn_skirt_r_4_03, swing1LimitDegrees=17, swing2LimitDegrees=17, twistLimitDegrees=1}
 {bone1Name=dyn_skirt_r_4_05, bone2Name=dyn_skirt_r_4_04, swing1LimitDegrees=20, swing2LimitDegrees=20, twistLimitDegrees=1}
 {
  bone1Name=dyn_skirt_r_4_04
  bone2Name=dyn_skirt_l_4_04
  swing1LimitDegrees=45
  swing1Motion=Free
  swing2LimitDegrees=45
  swing2Motion=Free
  twistLimitDegrees=45
  twistMotion=Free
 }
]
