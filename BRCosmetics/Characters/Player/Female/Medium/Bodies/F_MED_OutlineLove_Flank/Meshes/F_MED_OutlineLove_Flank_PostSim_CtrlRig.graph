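# Post-simulation Control Rig for the F_MED_OutlineLove_Flank female body that corrects skirt and cape bones according to player state.
# Construction adds FK and thigh reference controls; each thigh's rotation against its reference, gated by standing idle, drives that side's skirt offset through interpolation and curves.
# Crouch offsets fade with movement and a pelvis-height factor, which is stored after that use.
# Other offsets apply when not crouching, in FrontEnd, or constantly.
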ControlRigBlueprint /BRCosmetics/Characters/Player/Female/Medium/Bodies/F_MED_OutlineLove_Flank/Meshes/F_MED_OutlineLove_Flank_PostSim_CtrlRig
graphs [
 RigVMFunctionLibrary {links [], nodes []}
 RigVMModel {nodes [
   n1 = &RigUnit_PrepareForExecution()
   addFKCtrl = RigVMFunctionReferenceNode(
    Custom_FKCtrl=()
    Options_FKCtrl="(pelvis_1_#1=true,spine_3_#2=False,clavicle_5_#3=False,arm_7_#4=False,leg_9_#5=true,head_11_46DC15EF465CD2 ...(truncated)"
    Shape_Settings="(pelvis_2_#6=(bVisible=True,Name=\"Hexagon_Thin\",Color=(R=0.0117,G=0.8281,B=1,A=1),Transform=(Rotation=(X=0,Y=-0.7071,Z=0,W=0.7071),Translation=(X=2,Y=0,Z=0),Scale3D=(X=5,Y=5,Z=7))),spine_4_#7=(bVi ...(truncated)"
    ExecuteContext=@n1.ExecutePin
   )
   > Control Stack at Item_1 = RigVMFunctionReferenceNode(
    Add Bottom Null=False
    Add Null=true
    Add Secondary Control=False
    Control Suffix=_ctrl
    Item=(Name="thigh_r",Type=Bone)
    Name=ref_thigh_r
    Null Suffix=_null
    Parent=(Name="pelvis",Type=Bone)
    Secondary Settings="(Shape=(bVisible=True,Name=\"Default\",Color=(R=1,G=0,B=0,A=1),Transform=(Rotation=(X=0,Y=0,Z=0,W=1),Translation=(X=0,Y=0,Z=0),Scale3D=(X=1,Y=1,Z=1))),Proxy=(bIsProxy=False,ShapeVisibility=BasedOnSelection),DisplayName=\"None\",InitialSpace=LocalSpace,bUsePref ...(truncated)"
    Settings="(Shape=(bVisible=True,Name=\"Pyramid_Thin\",Color=(R=1,G=0,B=0,A=1),Transform=(Rotation=(X=0,Y=0,Z=0,W=1),Translation=(X=0,Y=0,Z=0),Scale3D=(X=1,Y=1,Z=1))),Proxy=(bIsProxy=False,ShapeVisibility=BasedOnSelection),DisplayName=\"None\",InitialSpace=LocalSpace,bUs ...(truncated)"
   )
   > Control Stack at Item_2 = RigVMFunctionReferenceNode(
    Add Bottom Null=False
    Add Null=true
    Add Secondary Control=False
    Control Suffix=_ctrl
    Item=(Name="thigh_l",Type=Bone)
    Name=ref_thigh_l
    Null Suffix=_null
    Parent=(Name="pelvis",Type=Bone)
    Secondary Settings="(Shape=(bVisible=True,Name=\"Default\",Color=(R=1,G=0,B=0,A=1),Transform=(Rotation=(X=0,Y=0,Z=0,W=1),Translation=(X=0,Y=0,Z=0),Scale3D=(X=1,Y=1,Z=1))),Proxy=(bIsProxy=False,ShapeVisibility=BasedOnSelection),DisplayName=\"None\",InitialSpace=LocalSpace,bUsePref ...(truncated)"
    Settings="(Shape=(bVisible=True,Name=\"Pyramid_Thin\",Color=(R=1,G=0,B=0,A=1),Transform=(Rotation=(X=0,Y=0,Z=0,W=1),Translation=(X=0,Y=0,Z=0),Scale3D=(X=1,Y=1,Z=1))),Proxy=(bIsProxy=False,ShapeVisibility=BasedOnSelection),DisplayName=\"None\",InitialSpace=LocalSpace,bUs ...(truncated)"
   )
   RigUnit_BeginExecution = &RigUnit_BeginExecution()
   setFKCtrl = RigVMFunctionReferenceNode(setFK_Ctrl=false, ExecuteContext=@RigUnit_BeginExecution.ExecutePin)
   n2 = &RigVMFunction_MathQuaternionToEuler(
    RotationOrder=ZXY
    Value=&RigUnit_GetRelativeTransformForItem(Child=(Type=Control,Name="ref_thigh_r_ctrl"), Parent=(Type=Bone,Name="thigh_r"), bChildInitial=False, bParentInitial=False).RelativeTransform.Rotation
   )
   Multiply = &RigVMFunction_MathDoubleMul(
    A=&RigVMFunction_MathDoubleRemap(SourceMaximum=20, SourceMinimum=4, TargetMaximum=1, TargetMinimum=0, Value=@n2.Result.Y).Result
    B=RigVMVariableNode() {variable=is Stading Idle}.Value
   )
   n3 = &RigVMFunction_AlphaInterp(InterpSpeedDecreasing=15, InterpSpeedIncreasing=7, Scale=1, bClampResult=true, Value=@Multiply.Result)
   n4 = &RigVMFunction_AnimEvalRichCurve(
    Curve="(EditorCurveData=(Keys=((InterpMode=RCIM_Linear,TangentMode=RCTM_Auto,TangentWeightMode=RCTWM_WeightedNone,Time=0,Value=0,ArriveTangent=0,ArriveTangentWeight=0,LeaveTangent=0,LeaveTangentWeight=0),(InterpMode=RCIM_Linear,TangentMode=RCTM_Auto,TangentWeight ...(truncated)"
    SourceMinimum=0
    TargetMinimum=0
    Value=@n3.Result
   )
   > &RigUnit_ModifyTransforms(
    ItemToModify="((Item=(Type=Bone,Name=\"base_skirt_r_2_01\"),Transform=(Rotation=(X=0,Y=0,Z=-0.1736,W=0.9848),Translation=(X=0,Y=-1.5,Z=0),Scale3D=(X=1,Y=1,Z=1))),(Item=(Type=Bone,Name=\"base_skirt_r_1_01\"),Transform=(Rotation=(X=0,Y=0,Z=-0.0436,W=0.999),Translation=(X=0,Y= ...(truncated)"
    WeightMinimum=0
    Weight=@n4.Result
   )
   n5 = &RigVMFunction_MathQuaternionToEuler(
    RotationOrder=ZXY
    Value=&RigUnit_GetRelativeTransformForItem(Child=(Type=Control,Name="ref_thigh_l_ctrl"), Parent=(Type=Bone,Name="thigh_l"), bChildInitial=False, bParentInitial=False).RelativeTransform.Rotation
   )
   n6 = &RigVMFunction_AlphaInterp(
    InterpSpeedDecreasing=15
    InterpSpeedIncreasing=7
    Scale=1
    bClampResult=true
    Value=&RigVMFunction_MathDoubleMul(
     A=&RigVMFunction_MathDoubleRemap(SourceMaximum=20, SourceMinimum=4, TargetMaximum=1, TargetMinimum=0, Value=@n5.Result.Y).Result
     B=RigVMVariableNode() {variable=is Stading Idle}.Value
    ).Result
   )
   n7 = &RigVMFunction_AnimEvalRichCurve(
    Curve="(EditorCurveData=(Keys=((InterpMode=RCIM_Linear,TangentMode=RCTM_Auto,TangentWeightMode=RCTWM_WeightedNone,Time=0,Value=0,ArriveTangent=0,ArriveTangentWeight=0,LeaveTangent=0,LeaveTangentWeight=0),(InterpMode=RCIM_Linear,TangentMode=RCTM_Auto,TangentWeight ...(truncated)"
    SourceMinimum=0
    TargetMinimum=0
    Value=@n6.Result
   )
   > &RigUnit_ModifyTransforms(
    ItemToModify="((Item=(Type=Bone,Name=\"base_skirt_l_2_01\"),Transform=(Rotation=(X=0,Y=0,Z=-0.1736,W=0.9848),Translation=(X=0,Y=0,Z=0),Scale3D=(X=1,Y=1,Z=1))),(Item=(Type=Bone,Name=\"base_skirt_l_1_01\"),Transform=(Rotation=(X=0,Y=0,Z=-0.0436,W=0.999),Translation=(X=0,Y=-0. ...(truncated)"
    WeightMinimum=0
    Weight=@n7.Result
   )
   Subtract = RigVMAggregateNode(A=1, B=RigVMVariableNode() {variable=is Moving 2d}.Value, C=RigVMVariableNode() {variable=Crouching high position}.Value)
   n8 = &RigVMFunction_AlphaInterp(
    InterpSpeedDecreasing=15
    InterpSpeedIncreasing=7
    Scale=1
    bClampResult=true
    Value=&RigVMFunction_MathDoubleMul(B=@Subtract.Result, A=RigVMVariableNode() {variable="is Crouching "}.Value).Result
   )
   n9 = &RigVMFunction_AnimEvalRichCurve(
    Curve="(EditorCurveData=(Keys=((InterpMode=RCIM_Linear,TangentMode=RCTM_Auto,TangentWeightMode=RCTWM_WeightedNone,Time=0,Value=0,ArriveTangent=0,ArriveTangentWeight=0,LeaveTangent=0,LeaveTangentWeight=0),(InterpMode=RCIM_Linear,TangentMode=RCTM_Auto,TangentWeight ...(truncated)"
    SourceMinimum=0
    TargetMinimum=0
    Value=@n8.Result
   )
   > &RigUnit_ModifyTransforms(
    ItemToModify="((Item=(Type=Bone,Name=\"dyn_skirt_l_1_01\"),Transform=(Rotation=(X=0,Y=0,Z=0,W=1),Translation=(X=0,Y=0,Z=0),Scale3D=(X=1,Y=1,Z=1))),(Item=(Type=Bone,Name=\"dyn_skirt_l_1_02\"),Transform=(Rotation=(X=0,Y=0,Z=0,W=1),Translation=(X=-5,Y=0,Z=0),Scale3D=(X=1,Y=1,Z ...(truncated)"
    WeightMinimum=0
    Weight=@n9.Result
   )
   n10 = &RigVMFunction_AnimEvalRichCurve(
    Curve=(EditorCurveData=(Keys=((),(Time=0.5,Value=1),(Time=0.8))))
    SourceMaximum=2
    SourceMinimum=0
    TargetMinimum=0
    Value=&RigVMFunction_MathDoubleRemap(
     SourceMaximum=75
     SourceMinimum=45
     TargetMaximum=1
     TargetMinimum=0
     bClamp=false
     Value=&RigUnit_GetRelativeTransformForItem(Child=(Type=Bone,Name="pelvis"), Parent=(Type=Bone,Name="root"), bChildInitial=False, bParentInitial=False).RelativeTransform.Translation.Z
    ).Result
   )
   > RigVMVariableNode(Value=@n10.Result) {variable=Crouching high position}
   > &RigUnit_ModifyTransforms(
    ItemToModify="((Item=(Type=Bone,Name=\"base_skirt_l_1_01\"),Transform=(Rotation=(X=0,Y=0,Z=0,W=1),Translation=(X=0,Y=-2,Z=0),Scale3D=(X=1,Y=1,Z=1))),(Item=(Type=Bone,Name=\"base_skirt_r_1_01\"),Transform=(Rotation=(X=0,Y=0,Z=0,W=1),Translation=(X=0,Y=-2,Z=0),Scale3D=(X=1,Y= ...(truncated)"
    WeightMinimum=0
    Weight=&RigVMFunction_MathDoubleSub(A=1, B=RigVMVariableNode() {variable="is Crouching "}.Value).Result
   )
   > &RigUnit_ModifyTransforms(
    ItemToModify=((Item=(Type=Bone,Name="dyn_skirt_r_4_03"),Transform=(Rotation=(X=-0.2603,Y=0.0187,Z=0.0682,W=0.9629),Translation=(X=0,Y=0,Z=0),Scale3D=(X=1,Y=1,Z=1))))
    WeightMinimum=0
    Weight=RigVMVariableNode() {variable="is FrontEnd "}.Value
   )
   > &RigUnit_ModifyTransforms(
    ItemToModify=((Item=(Type=Bone,Name="dyn_skirt_l_4_03"),Transform=(Rotation=(X=0.2603,Y=-0.0187,Z=0.0682,W=0.9629),Translation=(X=0,Y=0,Z=0),Scale3D=(X=1,Y=1,Z=1))))
    WeightMinimum=0
    Weight=RigVMVariableNode() {variable="is FrontEnd "}.Value
   )
   > &RigUnit_ModifyTransforms(
    ItemToModify=((Item=(Type=Bone,Name="dyn_skirt_f_02"),Transform=(Rotation=(X=0,Y=0,Z=0,W=-1),Translation=(X=0,Y=-3,Z=0),Scale3D=(X=1,Y=1,Z=1))))
    Weight=1
    WeightMinimum=0
   )
   n11 = RigVMVariableNode() {variable="is FrontEnd "}
   > &RigUnit_ModifyTransforms(
    ItemToModify=((Item=(Type=Bone,Name="dyn_cape_l_02"),Transform=(Rotation=(X=0,Y=0,Z=0,W=-1),Translation=(X=0,Y=-2.14,Z=0),Scale3D=(X=1,Y=1,Z=1))))
    WeightMinimum=0
    Weight=@n11.Value
   )
   > &RigUnit_ModifyTransforms(
    ItemToModify=((Item=(Type=Bone,Name="dyn_cape_l_03"),Transform=(Rotation=(X=-0.0045,Y=-0.0609,Z=-0.0731,W=0.9955),Translation=(X=0.2,Y=0,Z=-1),Scale3D=(X=1,Y=1,Z=1))))
    WeightMinimum=0
    Weight=@n11.Value
   )
   > &RigUnit_ModifyTransforms(
    ItemToModify=((Item=(Type=Bone,Name="dyn_cape_l_05"),Transform=(Rotation=(X=0,Y=-0.0785,Z=0,W=-0.9969),Translation=(X=0,Y=0,Z=0),Scale3D=(X=1,Y=1,Z=1))))
    WeightMinimum=0
    Weight=@n11.Value
   )
   > &RigUnit_ModifyTransforms(
    ItemToModify=((Item=(Type=Bone,Name="dyn_cape_r_02"),Transform=(Rotation=(X=0,Y=0,Z=0,W=-1),Translation=(X=0,Y=-2.14,Z=0),Scale3D=(X=1,Y=1,Z=1))))
    WeightMinimum=0
    Weight=@n11.Value
   )
   > &RigUnit_ModifyTransforms(
    ItemToModify=((Item=(Type=Bone,Name="dyn_cape_r_03"),Transform=(Rotation=(X=0.0045,Y=0.0609,Z=-0.0731,W=0.9955),Translation=(X=-0.2,Y=0,Z=-1),Scale3D=(X=1,Y=1,Z=1))))
    WeightMinimum=0
    Weight=@n11.Value
   )
   > &RigUnit_ModifyTransforms(
    ItemToModify=((Item=(Type=Bone,Name="dyn_cape_r_05"),Transform=(Rotation=(X=0,Y=-0.0785,Z=0,W=0.9969),Translation=(X=0,Y=0,Z=0),Scale3D=(X=1,Y=1,Z=1))))
    WeightMinimum=0
    Weight=@n11.Value
   )
   RigVMDispatchNode(Enabled=True, ExecuteContext=(), Prefix="right leg ", ScreenColor=(R=1,G=1,B=1,A=1), ScreenDuration=0.0, Value=@n2.Result.Z) {template="DISPATCH_RigVMDispatch_Print(in Prefix,in Value,in Enabled,in ScreenDuration,in ScreenColor)"}
   RigVMDispatchNode(Enabled=True, ExecuteContext=(), Prefix=left leg, ScreenColor=(R=1,G=1,B=1,A=1), ScreenDuration=0.0, Value=@n5.Result.Z) {template="DISPATCH_RigVMDispatch_Print(in Prefix,in Value,in Enabled,in ScreenDuration,in ScreenColor)"}
   RigVMDispatchNode(Enabled=True, ExecuteContext=(), ScreenColor=(R=1,G=1,B=1,A=1), ScreenDuration=0.0, Value=@n10.Result) {template="DISPATCH_RigVMDispatch_Print(in Prefix,in Value,in Enabled,in ScreenDuration,in ScreenColor)"}
  ]}
 Subtract_1_ContainedGraph {nodes [
   n1 = RigVMFunctionEntryNode()
   Subtract = &RigVMFunction_MathDoubleSub(A=@n1.A, B=@n1.B)
   RigVMFunctionReturnNode(Result=&RigVMFunction_MathDoubleSub(B=@n1.C, A=@Subtract.Result).Result)
  ]}
]
hierarchy [
 root {loc=<0 0 0>, rot=<0 0 0 1>}
 pelvis {parent=root, loc=<0 -0.5301 98.0571>, rot=<0.0127 0.707 -0.0127 -0.707>}
 base_skirt_l_1_01 {parent=pelvis, loc=<4.166 6.9665 -7.6793>, rot=<-0.0656 0.2585 0.9626 0.0469>}
 dyn_skirt_l_1_01 {parent=base_skirt_l_1_01, loc=<0 0 0>, rot=<0 0 0 1>}
 dyn_skirt_l_1_02 {parent=dyn_skirt_l_1_01, loc=<14.25 -0.7852 0>, rot=<0 0 0 1>}
 dyn_skirt_l_1_03 {parent=dyn_skirt_l_1_02, loc=<14.25 0.0573 0>, rot=<0 0 0 1>}
 dyn_skirt_l_1_04 {parent=dyn_skirt_l_1_03, loc=<14.1708 0.4023 0>, rot=<0 0 0 1>}
 dyn_skirt_l_1_05 {parent=dyn_skirt_l_1_04, loc=<14.25 0.2496 0>, rot=<0 0 0 1>}
 base_skirt_l_2_01 {parent=pelvis, loc=<3.9355 0.8334 -12.8604>, rot=<-0.0754 0.6804 0.7266 0.0591>}
 dyn_skirt_l_2_01 {parent=base_skirt_l_2_01, loc=<0 0 0>, rot=<0 0 0 1>}
 dyn_skirt_l_2_02 {parent=dyn_skirt_l_2_01, loc=<13.875 -0.7937 0>, rot=<0 0 0 1>}
 dyn_skirt_l_2_03 {parent=dyn_skirt_l_2_02, loc=<13.875 0.2935 0>, rot=<0 0 0 1>}
 dyn_skirt_l_2_04 {parent=dyn_skirt_l_2_03, loc=<13.8792 0.4429 0>, rot=<0 0 0 1>}
 dyn_skirt_l_2_05 {parent=dyn_skirt_l_2_04, loc=<13.875 0.0625 0>, rot=<0 0 0 1>}
 base_skirt_l_3_01 {parent=pelvis, loc=<3.6892 -6.0558 -11.2447>, rot=<0.0598 -0.9231 -0.3771 -0.0466>}
 dyn_skirt_l_3_01 {parent=base_skirt_l_3_01, loc=<0 0 0>, rot=<0 0 0 1>}
 dyn_skirt_l_3_02 {parent=dyn_skirt_l_3_01, loc=<13.5 -1.1683 0>, rot=<0 0 0 1>}
 dyn_skirt_l_3_03 {parent=dyn_skirt_l_3_02, loc=<13.5 0.7842 0>, rot=<0 0 0 1>}
 dyn_skirt_l_3_04 {parent=dyn_skirt_l_3_03, loc=<13.5736 0.5761 0>, rot=<0 0 0 1>}
 dyn_skirt_l_3_05 {parent=dyn_skirt_l_3_04, loc=<13.5 -0.0513 0>, rot=<0 0 0 1>}
 base_skirt_l_4_01 {parent=pelvis, loc=<3.6196 -9.4021 -4.6745>, rot=<0.049 -0.9948 -0.0844 -0.0281>}
 dyn_skirt_l_4_01 {parent=base_skirt_l_4_01, loc=<0 0 0>, rot=<0 0 0 1>}
 dyn_skirt_l_4_02 {parent=dyn_skirt_l_4_01, loc=<13.35 -1.529 0>, rot=<0 0 0 1>}
 dyn_skirt_l_4_03 {parent=dyn_skirt_l_4_02, loc=<13.35 0.6206 0>, rot=<0 0 0 1>}
 dyn_skirt_l_4_04 {parent=dyn_skirt_l_4_03, loc=<13.3551 0.4836 0>, rot=<0 0 0 1>}
 dyn_skirt_l_4_05 {parent=dyn_skirt_l_4_04, loc=<13.35 0.4241 0>, rot=<0 0 0 1>}
 base_skirt_f_01 {parent=pelvis, loc=<4.257 8.909 0>, rot=<0 0 0.9995 0.033>}
 dyn_skirt_f_01 {parent=base_skirt_f_01, loc=<0 0 0>, rot=<0 0 0 1>}
 dyn_skirt_f_02 {parent=dyn_skirt_f_01, loc=<9.1819 -0.0498 0>, rot=<0 0 0 1>}
 dyn_skirt_f_03 {parent=dyn_skirt_f_02, loc=<13.4045 -0.0269 0>, rot=<0 0 0 1>}
 dyn_skirt_f_04 {parent=dyn_skirt_f_03, loc=<13.4207 0.1293 0>, rot=<0 0 0 1>}
 dyn_skirt_f_05 {parent=dyn_skirt_f_04, loc=<13.1369 0.0896 0>, rot=<0 0 0 1>}
 spine_01 {parent=pelvis, loc=<6.1208 -0.9328 0>, rot=<0 0 0.034 -0.9994>}
 spine_02 {parent=spine_01, loc=<6.5997 -0.2095 0>, rot=<0 0 0.0238 -0.9997>}
 spine_03 {parent=spine_02, loc=<6.4914 -0.1237 0>, rot=<0 0 0.0149 -0.9999>}
 spine_04 {parent=spine_03, loc=<6.9598 0.0794 0>, rot=<0 0 -0.0132 -0.9999>}
 spine_05 {parent=spine_04, loc=<6.9987 0 0>, rot=<0 0 -0.0015 -1>}
 neck_01 {parent=spine_05, loc=<11.3138 0.1529 0>, rot=<0 0 -0.2003 -0.9797>}
 neck_02 {parent=neck_01, loc=<4.5627 0 0>, rot=<0 0 0.0504 -0.9987>}
 head {parent=neck_02, loc=<5.3914 0 0>, rot=<0 0 0.115 -0.9934>}
 clavicle_l {parent=spine_05, loc=<5.4831 -0.1053 -3.2946>, rot=<0.0465 -0.7289 -0.009 -0.6829>}
 upperarm_l {parent=clavicle_l, loc=<9.8658 -1.2136 -0.4105>, rot=<0.0164 -0.3245 -0.0269 -0.9454>}
 lowerarm_l {parent=upperarm_l, loc=<26.465 -0.8345 0.259>, rot=<0.0036 0.0161 -0.2298 -0.9731>}
 hand_l {parent=lowerarm_l, loc=<22.7355 -0.3203 -0.6105>, rot=<-0.6784 -0.0223 0.0735 -0.7307>}
 thumb_01_l {parent=hand_l, loc=<1.3118 -0.3191 -2.459>, rot=<0.6546 -0.4042 0.0664 -0.6354>}
 thumb_02_l {parent=thumb_01_l, loc=<5.9336 -0.0448 0.0401>, rot=<-0.0037 0.0213 0.298 -0.9543>}
 thumb_03_l {parent=thumb_02_l, loc=<2.2223 0.3682 -0.0907>, rot=<0.0118 -0.013 0.0066 -0.9998>}
 index_metacarpal_l {parent=hand_l, loc=<1.6535 0.0183 -2.96>, rot=<0.0158 -0.0124 -0.0173 -0.9996>}
 index_01_l {parent=index_metacarpal_l, loc=<8.3008 0.0357 -0.0524>, rot=<-0.0044 0.0485 0.0924 -0.9945>}
 index_02_l {parent=index_01_l, loc=<3.3587 -0.3004 0.0452>, rot=<-0.002 0.0027 0.3225 -0.9466>}
 index_03_l {parent=index_02_l, loc=<2.5578 -0.123 -0.0252>, rot=<-0.002 0.0067 0.2558 -0.9667>}
 middle_metacarpal_l {parent=hand_l, loc=<2.1856 -0.1725 -0.5984>, rot=<0.006 0.0168 -0.0329 -0.9993>}
 middle_01_l {parent=middle_metacarpal_l, loc=<7.3275 -0.2447 0.0129>, rot=<-0.0208 0.0042 0.0858 -0.9961>}
 middle_02_l {parent=middle_01_l, loc=<4.0123 0.1401 0.1175>, rot=<-0.0218 0.0258 0.3512 -0.9357>}
 middle_03_l {parent=middle_02_l, loc=<2.6077 0.0534 0.1373>, rot=<0.0082 0.0188 0.1708 -0.9851>}
 ring_metacarpal_l {parent=hand_l, loc=<2.2074 -0.3625 1.3638>, rot=<-0.0284 0.0722 -0.0458 -0.9959>}
 ring_01_l {parent=ring_metacarpal_l, loc=<7.0295 -0.6197 -0.1446>, rot=<-0.0489 -0.0252 0.1892 -0.9804>}
 ring_02_l {parent=ring_01_l, loc=<3.6174 0.4647 -0.1027>, rot=<0 0 0.2806 -0.9598>}
 ring_03_l {parent=ring_02_l, loc=<2.4152 0.114 0.0199>, rot=<0 0 0.1189 -0.9929>}
 pinky_metacarpal_l {parent=hand_l, loc=<1.7616 -0.9753 3.3814>, rot=<-0.0397 0.0814 -0.069 -0.9935>}
 pinky_01_l {parent=pinky_metacarpal_l, loc=<7.1709 -0.4831 -0.1094>, rot=<-0.0654 -0.0035 0.2242 -0.9723>}
 pinky_02_l {parent=pinky_01_l, loc=<2.8274 0.0893 -0.1043>, rot=<0.0037 0.0011 0.2854 -0.9584>}
 pinky_03_l {parent=pinky_02_l, loc=<2.2046 0.0307 -0.0078>, rot=<-0.0002 0.0098 0.1445 -0.9895>}
 deform_wrist_dn_base_l {parent=hand_l, loc=<0 -1.5326 0>, rot=<0 0 0 -1>}
 deform_wrist_dn_l {parent=deform_wrist_dn_base_l, loc=<0 0 0>, rot=<0 0 0 1>}
 deform_wrist_up_base_l {parent=hand_l, loc=<0 2.3009 0>, rot=<0 0 0 -1>}
 deform_wrist_up_l {parent=deform_wrist_up_base_l, loc=<0 0 0>, rot=<0 0 0 1>}
 lowerarm_twist_02_l {parent=lowerarm_l, loc=<10.6707 0 0>, rot=<0 0 0 1>}
 lowerarm_twist_01_l {parent=lowerarm_l, loc=<18.316 0 0>, rot=<0 0 0 1>}
 deform_elbow_in_base_l {parent=lowerarm_l, loc=<0.1834 3.6062 -0.2572>, rot=<0 0 0 -1>}
 deform_elbow_in_l {parent=deform_elbow_in_base_l, loc=<0 0 0>, rot=<0 0 0 1>}
 deform_elbow_out_base_l {parent=lowerarm_l, loc=<0 -4 0>, rot=<0 0 0 1>}
 deform_elbow_out_l {parent=deform_elbow_out_base_l, loc=<0 0 0>, rot=<0 0 0 1>}
 upperarm_twist_01_l {parent=upperarm_l, loc=<7.2135 -0.1134 0>, rot=<0 0 0 1>}
 deform_upperarm_bk_base_l {parent=upperarm_twist_01_l, loc=<-6.4053 -4.3746 0.314>, rot=<0 0 0 -1>}
 deform_upperarm_bk_l {parent=deform_upperarm_bk_base_l, loc=<0 0 0>, rot=<0 0 0 1>}
 deform_upperarm_fr_base_l {parent=upperarm_twist_01_l, loc=<-5.6312 5.092 0.3354>, rot=<0 0 0 -1>}
 deform_upperarm_fr_l {parent=deform_upperarm_fr_base_l, loc=<0 0 0>, rot=<0 0 0 1>}
 deform_upperarm_up_base_l {parent=upperarm_twist_01_l, loc=<-8.891 0.4825 3.9517>, rot=<0 0 0 -1>}
 deform_upperarm_up_l {parent=deform_upperarm_up_base_l, loc=<0 0 0>, rot=<0 0 0 1>}
 upperarm_twist_02_l {parent=upperarm_l, loc=<19.7998 -0.1134 0>, rot=<0 0 0 1>}
 base_shoulderPad_l_02 {parent=clavicle_l, loc=<12.2306 3.9824 2.6454>, rot=<0.8939 0.1728 -0.1233 -0.3948>}
 dyn_shoulderPad_l_02 {parent=base_shoulderPad_l_02, loc=<0 0 0>, rot=<0 0 0 1>}
 base_shoulderPad_l_03 {parent=clavicle_l, loc=<13.8322 -1.6765 5.4324>, rot=<0.7071 0.0415 0.0059 -0.7059>}
 dyn_shoulderPad_l_03 {parent=base_shoulderPad_l_03, loc=<0 0 0>, rot=<0 0 0 1>}
 base_shoulderPad_l_04 {parent=clavicle_l, loc=<12.602 -6.416 2.0762>, rot=<0.1542 -0.2993 0.0862 -0.9377>}
 dyn_shoulderPad_l_04 {parent=base_shoulderPad_l_04, loc=<0 0 0>, rot=<0 0 0 1>}
 base_shoulderPad_l_01 {parent=clavicle_l, loc=<8.0696 6.9938 -1.6135>, rot=<0.8723 0.2165 -0.2957 -0.3236>}
 dyn_shoulderPad_l_01 {parent=base_shoulderPad_l_01, loc=<0 0 0>, rot=<0 0 0 1>}
 base_shoulderPad_l_05 {parent=clavicle_l, loc=<9.5881 -7.4827 -2.3724>, rot=<0.0716 -0.3834 0.0506 -0.9194>}
 dyn_shoulderPad_l_05 {parent=base_shoulderPad_l_05, loc=<0 0 0>, rot=<0 0 0 1>}
 base_cape_l_00 {parent=clavicle_l, loc=<11.158 -7.1525 -3.0942>, rot=<0.7509 0.0109 -0.6558 0.0773>}
 dyn_cape_l_00 {parent=base_cape_l_00, loc=<0 0 0>, rot=<0 0 0 1>}
 dyn_cape_l_01 {parent=dyn_cape_l_00, loc=<4.2929 0 0>, rot=<0 0 0 1>}
 dyn_cape_l_02 {parent=dyn_cape_l_01, loc=<21 0 0>, rot=<0 0 -0.0054 -1>}
 dyn_cape_l_03 {parent=dyn_cape_l_02, loc=<20.9988 -0.2273 0>, rot=<0 0 0 1>}
 dyn_cape_l_04 {parent=dyn_cape_l_03, loc=<21 0.0518 0>, rot=<0 0 0 1>}
 dyn_cape_l_05 {parent=dyn_cape_l_04, loc=<21.9995 0 0>, rot=<0 0 0 1>}
 clavicle_r {parent=spine_05, loc=<5.4831 -0.1053 3.2946>, rot=<0.7289 0.0465 -0.6829 0.009>}
 upperarm_r {parent=clavicle_r, loc=<-9.8658 1.2136 0.4105>, rot=<0.0164 -0.3245 -0.0269 -0.9454>}
 lowerarm_r {parent=upperarm_r, loc=<-26.465 0.8345 -0.259>, rot=<0.0036 0.0161 -0.2298 -0.9731>}
 hand_r {parent=lowerarm_r, loc=<-22.7355 0.3203 0.6106>, rot=<-0.6784 -0.0223 0.0735 -0.7307>}
 thumb_01_r {parent=hand_r, loc=<-1.3118 0.3191 2.459>, rot=<0.6546 -0.4042 0.0664 -0.6354>}
 thumb_02_r {parent=thumb_01_r, loc=<-5.9336 0.0448 -0.0401>, rot=<-0.0037 0.0213 0.298 -0.9543>}
 thumb_03_r {parent=thumb_02_r, loc=<-2.2223 -0.3682 0.0907>, rot=<0.0118 -0.013 0.0066 -0.9998>}
 index_metacarpal_r {parent=hand_r, loc=<-1.6535 -0.0183 2.96>, rot=<0.0158 -0.0124 -0.0173 -0.9996>}
 index_01_r {parent=index_metacarpal_r, loc=<-8.3008 -0.0357 0.0524>, rot=<-0.0044 0.0485 0.0924 -0.9945>}
 index_02_r {parent=index_01_r, loc=<-3.3587 0.3004 -0.0452>, rot=<-0.002 0.0027 0.3225 -0.9466>}
 index_03_r {parent=index_02_r, loc=<-2.5578 0.123 0.0252>, rot=<-0.002 0.0067 0.2558 -0.9667>}
 middle_metacarpal_r {parent=hand_r, loc=<-2.1856 0.1725 0.5984>, rot=<0.006 0.0168 -0.0329 -0.9993>}
 middle_01_r {parent=middle_metacarpal_r, loc=<-7.3275 0.2447 -0.0129>, rot=<-0.0208 0.0042 0.0858 -0.9961>}
 middle_02_r {parent=middle_01_r, loc=<-4.0123 -0.1401 -0.1175>, rot=<-0.0218 0.0258 0.3512 -0.9357>}
 middle_03_r {parent=middle_02_r, loc=<-2.6077 -0.0534 -0.1373>, rot=<0.0082 0.0188 0.1708 -0.9851>}
 ring_metacarpal_r {parent=hand_r, loc=<-2.2074 0.3625 -1.3638>, rot=<-0.0284 0.0722 -0.0458 -0.9959>}
 ring_01_r {parent=ring_metacarpal_r, loc=<-7.0295 0.6197 0.1446>, rot=<-0.0489 -0.0252 0.1892 -0.9804>}
 ring_02_r {parent=ring_01_r, loc=<-3.6174 -0.4647 0.1027>, rot=<0 0 0.2806 -0.9598>}
 ring_03_r {parent=ring_02_r, loc=<-2.4152 -0.114 -0.0199>, rot=<0 0 0.1189 -0.9929>}
 pinky_metacarpal_r {parent=hand_r, loc=<-1.7616 0.9753 -3.3814>, rot=<-0.0397 0.0814 -0.069 -0.9935>}
 pinky_01_r {parent=pinky_metacarpal_r, loc=<-7.1709 0.4831 0.1094>, rot=<-0.0654 -0.0035 0.2242 -0.9723>}
 pinky_02_r {parent=pinky_01_r, loc=<-2.8274 -0.0893 0.1043>, rot=<0.0037 0.0011 0.2854 -0.9584>}
 pinky_03_r {parent=pinky_02_r, loc=<-2.2046 -0.0307 0.0078>, rot=<-0.0002 0.0098 0.1445 -0.9895>}
 deform_wrist_dn_base_r {parent=hand_r, loc=<-0.0001 1.5328 0>, rot=<0 0 0 -1>}
 deform_wrist_dn_r {parent=deform_wrist_dn_base_r, loc=<0 0 0>, rot=<0 0 0 1>}
 deform_wrist_up_base_r {parent=hand_r, loc=<-0.0002 -2.3007 -0.0001>, rot=<0 0 0 -1>}
 deform_wrist_up_r {parent=deform_wrist_up_base_r, loc=<0 0 0>, rot=<0 0 0 1>}
 lowerarm_twist_02_r {parent=lowerarm_r, loc=<-10.6707 0 0>, rot=<0 0 0 -1>}
 lowerarm_twist_01_r {parent=lowerarm_r, loc=<-18.316 0 0>, rot=<0 0 0 -1>}
 deform_elbow_in_base_r {parent=lowerarm_r, loc=<-0.1834 -3.6062 0.2572>, rot=<0 0 0 -1>}
 deform_elbow_in_r {parent=deform_elbow_in_base_r, loc=<0 0 0>, rot=<0 0 0 1>}
 deform_elbow_out_base_r {parent=lowerarm_r, loc=<0 4 -0.0001>, rot=<0 0 0 -1>}
 deform_elbow_out_r {parent=deform_elbow_out_base_r, loc=<0 0 0>, rot=<0 0 0 1>}
 upperarm_twist_01_r {parent=upperarm_r, loc=<-7.2135 0.1134 0>, rot=<0 0 0 -1>}
 deform_upperarm_bk_base_r {parent=upperarm_twist_01_r, loc=<6.4053 4.3746 -0.314>, rot=<0 0 0 -1>}
 deform_upperarm_bk_r {parent=deform_upperarm_bk_base_r, loc=<0 0 0>, rot=<0 0 0 1>}
 deform_upperarm_fr_base_r {parent=upperarm_twist_01_r, loc=<5.6312 -5.092 -0.3354>, rot=<0 0 0 -1>}
 deform_upperarm_fr_r {parent=deform_upperarm_fr_base_r, loc=<0 0 0>, rot=<0 0 0 1>}
 deform_upperarm_up_base_r {parent=upperarm_twist_01_r, loc=<8.891 -0.4825 -3.9517>, rot=<0 0 0 -1>}
 deform_upperarm_up_r {parent=deform_upperarm_up_base_r, loc=<0 0 0>, rot=<0 0 0 1>}
 upperarm_twist_02_r {parent=upperarm_r, loc=<-19.7998 0.1134 0>, rot=<0 0 0 -1>}
 base_shoulderPad_r_02 {parent=clavicle_r, loc=<-12.2306 -3.9824 -2.6455>, rot=<0.8939 0.1728 -0.1233 -0.3948>}
 dyn_shoulderPad_r_02 {parent=base_shoulderPad_r_02, loc=<0 0 0>, rot=<0 0 0 1>}
 base_shoulderPad_r_03 {parent=clavicle_r, loc=<-13.8323 1.6765 -5.4322>, rot=<0.7071 0.0415 0.0059 -0.7059>}
 dyn_shoulderPad_r_03 {parent=base_shoulderPad_r_03, loc=<0 0 0>, rot=<0 0 0 1>}
 base_shoulderPad_r_04 {parent=clavicle_r, loc=<-12.602 6.4159 -2.0761>, rot=<0.1542 -0.2993 0.0862 -0.9377>}
 dyn_shoulderPad_r_04 {parent=base_shoulderPad_r_04, loc=<0 0 0>, rot=<0 0 0 -1>}
 base_shoulderPad_r_01 {parent=clavicle_r, loc=<-8.0696 -6.9938 1.6133>, rot=<0.8723 0.2165 -0.2957 -0.3236>}
 dyn_shoulderPad_r_01 {parent=base_shoulderPad_r_01, loc=<0 0 0>, rot=<0 0 0 -1>}
 base_shoulderPad_r_05 {parent=clavicle_r, loc=<-9.5881 7.4827 2.3721>, rot=<0.0716 -0.3834 0.0506 -0.9194>}
 dyn_shoulderPad_r_05 {parent=base_shoulderPad_r_05, loc=<0 0 0>, rot=<0 0 0 1>}
 base_cape_r_00 {parent=clavicle_r, loc=<-11.1586 7.1577 3.0939>, rot=<-0.0109 0.7509 -0.0773 -0.6558>}
 dyn_cape_r_00 {parent=base_cape_r_00, loc=<0 0 0>, rot=<0 0 0 1>}
 dyn_cape_r_01 {parent=dyn_cape_r_00, loc=<4.2931 0 0>, rot=<0.0087 0 0 -1>}
 dyn_cape_r_02 {parent=dyn_cape_r_01, loc=<20.9994 0 0>, rot=<0 0 -0.0054 -1>}
 dyn_cape_r_03 {parent=dyn_cape_r_02, loc=<20.999 -0.2273 0>, rot=<0 0 0 1>}
 dyn_cape_r_04 {parent=dyn_cape_r_03, loc=<20.9999 0.0518 0>, rot=<0 0 0 1>}
 dyn_cape_r_05 {parent=dyn_cape_r_04, loc=<21.9995 0 0>, rot=<0 0 0 1>}
 attach_backpack {parent=spine_05, loc=<-0.123 -13.473 0>, rot=<-0.0093 -0.707 0.0093 -0.707>}
 attach_cape {parent=spine_05, loc=<9.017 -11.714 0>, rot=<0.0616 -0.7044 -0.0616 -0.7044>}
 deform_pec_aim_base_l {parent=spine_04, loc=<3.2456 9.3309 -1.2237>, rot=<0.0603 -0.4033 0.3318 -0.8507>}
 deform_pec_aim_l {parent=deform_pec_aim_base_l, loc=<0 0 0>, rot=<0 0 0 1>}
 deform_pec_base_l {parent=deform_pec_aim_l, loc=<9.3665 0 0>, rot=<0 0 0 1>}
 deform_pec_l {parent=deform_pec_base_l, loc=<0 0 0>, rot=<0 0 0 1>}
 deform_pec_aim_base_r {parent=spine_04, loc=<3.2452 9.3309 1.2237>, rot=<-0.4033 -0.0603 0.8507 0.3318>}
 deform_pec_aim_r {parent=deform_pec_aim_base_r, loc=<0 0 0>, rot=<0 0 0 1>}
 deform_pec_base_r {parent=deform_pec_aim_r, loc=<-9.3667 -0.0003 -0.0004>, rot=<0 0 0 1>}
 deform_pec_r {parent=deform_pec_base_r, loc=<0 0 0>, rot=<0 0 0 1>}
 thigh_l {parent=pelvis, loc=<-1.0855 -0.2109 -8.7053>, rot=<0.0009 0.0355 0.0155 -0.9993>}
 calf_l {parent=thigh_l, loc=<-42.5284 -0.0019 -0.7297>, rot=<0.0006 -0.0161 -0.0391 -0.9991>}
 foot_l {parent=calf_l, loc=<-46.7401 -2.5014 -0.592>, rot=<0.0427 -0.0238 0.0353 -0.9982>}
 ball_l {parent=foot_l, loc=<-3.5658 10.8066 -0.1544>, rot=<0.0001 -0.0001 -0.7063 -0.7079>}
 calf_twist_02_l {parent=calf_l, loc=<-19.1546 -1.7943 -0.2638>, rot=<0.0237 0.0095 -0.0035 -0.9997>}
 calf_twist_01_l {parent=calf_l, loc=<-35.8483 -1.9197 -0.5794>, rot=<0.0237 0.0095 -0.0035 -0.9997>}
 deform_knee_in_base_l {parent=calf_l, loc=<0.8031 -5.0231 0.2456>, rot=<0 0 0 -1>}
 deform_knee_in_l {parent=deform_knee_in_base_l, loc=<0 0 0>, rot=<0 0 0 1>}
 deform_knee_out_base_l {parent=calf_l, loc=<0 4 0>, rot=<0 0 0 -1>}
 deform_knee_out_l {parent=deform_knee_out_base_l, loc=<0 0 0>, rot=<0 0 0 1>}
 thigh_twist_01_l {parent=thigh_l, loc=<-27.952 0 0>, rot=<0.0013 0 0 -1>}
 deform_groin_base_l {parent=thigh_l, loc=<0.4801 5.6776 -0.2486>, rot=<0 0 0 -1>}
 deform_groin_l {parent=deform_groin_base_l, loc=<0 0 0>, rot=<0 0 0 1>}
 thigh_r {parent=pelvis, loc=<-1.0855 -0.2109 8.7053>, rot=<0.0355 -0.0009 0.9993 0.0155>}
 calf_r {parent=thigh_r, loc=<42.5284 0.0019 0.7297>, rot=<0.0006 -0.0161 -0.0391 -0.9991>}
 foot_r {parent=calf_r, loc=<46.7401 2.5014 0.592>, rot=<0.0427 -0.0238 0.0353 -0.9982>}
 ball_r {parent=foot_r, loc=<3.5658 -10.8066 0.1544>, rot=<0.0001 -0.0001 -0.7063 -0.7079>}
 calf_twist_02_r {parent=calf_r, loc=<19.1546 1.7943 0.2638>, rot=<0.0237 0.0095 -0.0035 -0.9997>}
 calf_twist_01_r {parent=calf_r, loc=<35.8483 1.9197 0.5794>, rot=<0.0237 0.0095 -0.0035 -0.9997>}
 deform_knee_in_base_r {parent=calf_r, loc=<-0.8031 5.0231 -0.2457>, rot=<0 0 0 -1>}
 deform_knee_in_r {parent=deform_knee_in_base_r, loc=<0 0 0>, rot=<0 0 0 1>}
 deform_knee_out_base_r {parent=calf_r, loc=<0 -4 0>, rot=<0 0 0 -1>}
 deform_knee_out_r {parent=deform_knee_out_base_r, loc=<0 0 0>, rot=<0 0 0 1>}
 thigh_twist_01_r {parent=thigh_r, loc=<27.952 0 0>, rot=<0.0013 0 0 -1>}
 deform_groin_base_r {parent=thigh_r, loc=<-0.4801 -5.6776 0.2486>, rot=<0 0 0 -1>}
 deform_groin_r {parent=deform_groin_base_r, loc=<0 0 0>, rot=<0 0 0 1>}
 deform_glute_base_l {parent=pelvis, loc=<-7.3466 -9.0213 -9.1245>, rot=<0.0172 0.031 -0.4477 -0.8935>}
 deform_glute_l {parent=deform_glute_base_l, loc=<0 0 0>, rot=<0 0 0 1>}
 deform_glute_base_r {parent=pelvis, loc=<-7.3464 -9.0213 9.1245>, rot=<0.031 -0.0172 0.8935 -0.4477>}
 deform_glute_r {parent=deform_glute_base_r, loc=<0 0 0>, rot=<0 0 0 1>}
 base_skirt_r_1_01 {parent=pelvis, loc=<4.1661 6.9665 7.6793>, rot=<0.0656 -0.2585 0.9626 0.0469>}
 dyn_skirt_r_1_01 {parent=base_skirt_r_1_01, loc=<0 0 0>, rot=<0 0 0 1>}
 dyn_skirt_r_1_02 {parent=dyn_skirt_r_1_01, loc=<14.2501 -0.7852 0>, rot=<0 0 0 1>}
 dyn_skirt_r_1_03 {parent=dyn_skirt_r_1_02, loc=<14.25 0.0573 -0.0001>, rot=<0 0 0 1>}
 dyn_skirt_r_1_04 {parent=dyn_skirt_r_1_03, loc=<14.1709 0.4023 0>, rot=<0 0 0 1>}
 dyn_skirt_r_1_05 {parent=dyn_skirt_r_1_04, loc=<14.25 0.2496 0>, rot=<0 0 0 1>}
 base_skirt_r_2_01 {parent=pelvis, loc=<3.9352 0.8334 12.8604>, rot=<0.0754 -0.6804 0.7266 0.0591>}
 dyn_skirt_r_2_01 {parent=base_skirt_r_2_01, loc=<0 0 0>, rot=<0 0 0 1>}
 dyn_skirt_r_2_02 {parent=dyn_skirt_r_2_01, loc=<13.8747 -0.7938 0>, rot=<0 0 0 1>}
 dyn_skirt_r_2_03 {parent=dyn_skirt_r_2_02, loc=<13.875 0.2935 0>, rot=<0 0 0 1>}
 dyn_skirt_r_2_04 {parent=dyn_skirt_r_2_03, loc=<13.8791 0.443 0>, rot=<0 0 0 1>}
 dyn_skirt_r_2_05 {parent=dyn_skirt_r_2_04, loc=<13.875 0.0625 0>, rot=<0 0 0 1>}
 base_skirt_r_3_01 {parent=pelvis, loc=<3.6893 -6.0558 11.2447>, rot=<0.0598 -0.9231 0.3771 0.0466>}
 dyn_skirt_r_3_01 {parent=base_skirt_r_3_01, loc=<0 0 0>, rot=<0 0 0 1>}
 dyn_skirt_r_3_02 {parent=dyn_skirt_r_3_01, loc=<13.5 -1.1682 0.0001>, rot=<0 0 0 1>}
 dyn_skirt_r_3_03 {parent=dyn_skirt_r_3_02, loc=<13.5001 0.7842 0>, rot=<0 0 0 1>}
 dyn_skirt_r_3_04 {parent=dyn_skirt_r_3_03, loc=<13.5735 0.5762 0>, rot=<0 0 0 1>}
 dyn_skirt_r_3_05 {parent=dyn_skirt_r_3_04, loc=<13.5 -0.0514 0>, rot=<0 0 0 1>}
 base_skirt_r_4_01 {parent=pelvis, loc=<3.6194 -9.4021 4.6745>, rot=<0.049 -0.9948 0.0844 0.0281>}
 dyn_skirt_r_4_01 {parent=base_skirt_r_4_01, loc=<0 0 0>, rot=<0 0 0 1>}
 dyn_skirt_r_4_02 {parent=dyn_skirt_r_4_01, loc=<13.3498 -1.5291 0>, rot=<0 0 0 -1>}
 dyn_skirt_r_4_03 {parent=dyn_skirt_r_4_02, loc=<13.35 0.6206 0>, rot=<0 0 0 -1>}
 dyn_skirt_r_4_04 {parent=dyn_skirt_r_4_03, loc=<13.355 0.4837 0>, rot=<0 0 0 -1>}
 dyn_skirt_r_4_05 {parent=dyn_skirt_r_4_04, loc=<13.3501 0.4241 0>, rot=<0 0 0 -1>}
 ground_mover {parent=root, loc=<0 0 0>, rot=<0 0 0 1>}
]
variables [
 Crouching high position {type=double, default=0}
 "is Crouching " {type=double, default=0}
 "is FrontEnd " {type=double, default=0}
 is Moving 2d {type=double, default=0}
 is Stading Idle {type=double, default=0}
]
