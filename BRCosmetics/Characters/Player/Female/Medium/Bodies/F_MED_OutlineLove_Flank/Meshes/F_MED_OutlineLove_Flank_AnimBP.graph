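# Post-process-style AnimBlueprint for the OutlineLove Flank female body cosmetic, animating its skirt and cape over the parent mesh's pose.
# A pre-sim Control Rig and bone tweaks feed chained skirt RigidBody sims, then two cape sims selected by back bling.
# Corrective bone offsets and a post-sim Control Rig finish the pose.
# The EventGraph converts movement states into float alphas, adjusts gravity for skydiving and lowers skirt sim weight during emotes.
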AnimBlueprint /BRCosmetics/Characters/Player/Female/Medium/Bodies/F_MED_OutlineLove_Flank/Meshes/F_MED_OutlineLove_Flank_AnimBP
target_skeleton=&/F_MED_OutlineLove_Flank_Skeleton
graphs [
 AnimGraph {nodes [
   n1 = Use cached pose 'Cape Normal Physics'()
   n2 = Use cached pose 'Cape Backpack Physics'()
   n3 = Get isWearingPack() -> (isWearingPack:bool)
   n4 = "Transform (Modify) Bone - Bone: base_skirt_l_3_01"(
    ComponentPose=Local To Component(LocalPose=Blend Poses by bool(BlendPose_0=@n2, BlendPose_1=@n1, bActiveValue=@n3))
    Alpha=Get Skydive Left() -> (Skydive Left:double)
    Translation="0,0,0"
    Rotation="0,-10,0"
    Scale="1,1,1"
   ) {
    rotationMode=BMM_Additive
    rotationSpace=BCS_BoneSpace
    translationMode=BMM_Additive
    translationSpace=BCS_BoneSpace
    alphaScaleBiasClamp {bInterpResult=true}
    boneToModify {boneName=base_skirt_l_3_01}
    rotation {yaw=-10}
   }
   n5 = "Transform (Modify) Bone - Bone: base_skirt_r_3_01"(ComponentPose=@n4, Alpha=Get Skydive Right() -> (Skydive Right:double), Translation="0,0,0", Rotation="0,-10,0", Scale="1,1,1") {
    rotationMode=BMM_Additive
    rotationSpace=BCS_BoneSpace
    translationMode=BMM_Additive
    translationSpace=BCS_BoneSpace
    alphaScaleBiasClamp {bInterpResult=true}
    boneToModify {boneName=base_skirt_r_3_01}
    rotation {yaw=-10}
   }
   n6 = "Transform (Modify) Bone - Bone: base_cape_l_00"(ComponentPose=@n5, Alpha=1, Translation="0,1,0", Rotation="0,0,7", Scale="1,1,1") {
    rotationMode=BMM_Additive
    rotationSpace=BCS_BoneSpace
    scaleMode=BMM_Additive
    translationMode=BMM_Additive
    translationSpace=BCS_BoneSpace
    boneToModify {boneName=base_cape_l_00}
    rotation {roll=7}
    translation {y=1}
   }
   n7 = "Transform (Modify) Bone - Bone: base_shoulderPad_l_05"(ComponentPose=@n6, Alpha=1, Translation="0,-1.5,0", Rotation="0,0,0", Scale="1,1,1") {
    rotationMode=BMM_Additive
    rotationSpace=BCS_BoneSpace
    scaleMode=BMM_Additive
    translationMode=BMM_Additive
    translationSpace=BCS_BoneSpace
    boneToModify {boneName=base_shoulderPad_l_05}
    translation {y=-1.5}
   }
   n8 = "Transform (Modify) Bone - Bone: base_cape_r_00"(ComponentPose=@n7, Alpha=1, Translation="0,1,0", Rotation="0,0,-7", Scale="1,1,1") {
    rotationMode=BMM_Additive
    rotationSpace=BCS_BoneSpace
    scaleMode=BMM_Additive
    translationMode=BMM_Additive
    translationSpace=BCS_BoneSpace
    boneToModify {boneName=base_cape_r_00}
    rotation {roll=-7}
    translation {y=1}
   }
   n9 = "Transform (Modify) Bone - Bone: base_shoulderPad_r_05"(ComponentPose=@n8, Alpha=1, Translation="0,1.7,0", Rotation="0,0,0", Scale="1,1,1") {
    rotationMode=BMM_Additive
    rotationSpace=BCS_BoneSpace
    scaleMode=BMM_Additive
    translationMode=BMM_Additive
    translationSpace=BCS_BoneSpace
    boneToModify {boneName=base_shoulderPad_r_05}
    translation {y=1.7}
   }
   n10 = Control Rig!(
    Source=Component To Local(ComponentPose=@n9)
    Alpha
    is Stading Idle:double=Get is Standing Idle CR() -> (is Standing Idle CR:double)
    "is Crouching ":double=Get is Crouching CR() -> (is Crouching CR:double)
    is Moving 2d:double=Get is Moving 2D CR() -> (is Moving 2D CR:double)
    "is FrontEnd ":double=Get Is Front End/Idle CR() -> (Is Front End/Idle CR:double)
   ) -> (Pose) {controlRigAssetReference {blueprintRigClass=&/F_MED_OutlineLove_Flank_PostSim_CtrlRig.F_MED_OutlineLove_Flank_PostSim_CtrlRig_C}}
   Output Pose(Result=@n10) {layerGroup=None}
   n11 = Control Rig!(
    Source=Copy Pose From Mesh() {bUseAttachedParent=true}
    Alpha
    "is Knee Sliding ":double=Get is Knee Sliding CR() -> (is Knee Sliding CR:double)
    is Baseball Sliding:double=Get is Baseball Sliding CR() -> (is Baseball Sliding CR:double)
    is Crouching:double=Get is Crouching CR()
    is Moving 2D:double=Get is Moving 2D CR()
   ) -> (Pose) {controlRigAssetReference {blueprintRigClass=&/F_MED_OutlineLove_Flank_CtrlRig.F_MED_OutlineLove_Flank_CtrlRig_C}}
   n12 = "Transform (Modify) Bone - Bone: ground_mover"(ComponentPose=Local To Component(LocalPose=@n11), Alpha=1, Translation="0,0,4", Rotation="0,0,0", Scale="1,1,1") {
    rotationMode=BMM_Additive
    rotationSpace=BCS_BoneSpace
    translationMode=BMM_Additive
    translationSpace=BCS_BoneSpace
    boneToModify {boneName=ground_mover}
    translation {z=4}
   }
   n13 = Get Tactical Sprint Jump() -> (Tactical Sprint Jump:double)
   n14 = "Transform (Modify) Bone - Bone: base_cape_l_00"(ComponentPose=@n12, Alpha=@n13, Translation="0,0,0", Rotation="0,-30,0", Scale="1,1,1") {
    rotationMode=BMM_Additive
    rotationSpace=BCS_BoneSpace
    scaleMode=BMM_Additive
    scaleSpace=BCS_BoneSpace
    translationMode=BMM_Additive
    translationSpace=BCS_BoneSpace
    alphaScaleBiasClamp {bInterpResult=true}
    boneToModify {boneName=base_cape_l_00}
    rotation {yaw=-30}
   }
   n15 = "Transform (Modify) Bone - Bone: base_cape_r_00"(ComponentPose=@n14, Alpha=@n13, Translation="0,0,0", Rotation="0,-30,0", Scale="1,1,1") {
    rotationMode=BMM_Additive
    rotationSpace=BCS_BoneSpace
    scaleMode=BMM_Additive
    scaleSpace=BCS_BoneSpace
    translationMode=BMM_Additive
    translationSpace=BCS_BoneSpace
    alphaScaleBiasClamp {bInterpResult=true}
    boneToModify {boneName=base_cape_r_00}
    rotation {yaw=-30}
   }
   n16 = RigidBody!(ComponentPose=@n15, Alpha=Get AlphaFE() -> (AlphaFE:double), OverrideWorldGravity:Vector=Get Gravity Override(), ExternalForce) -> (Pose) {
    bClampLinearTranslationLimitToRefPose=true
    bDefaultToSkeletalMeshPhysicsAsset=false
    bOverrideWorldGravity=true
    overridePhysicsAsset=&/F_MED_OutlineLove_Flank_Physics
    simulationSpace=BaseBoneSpace
    baseBoneRef {boneName=pelvis}
    componentAppliedLinearAccClamp=<0 0 0>
    simSpaceSettings {
     dampingAlpha=0
     maxAngularAcceleration=16
     maxAngularVelocity=8
     maxLinearAcceleration=2800
     maxLinearVelocity=700
     velocityScaleZ=0.3
     worldAlpha=1
     externalLinearDragV=<0.2 0.2 0.2>
    }
   }
   n17 = RigidBody(ComponentPose=@n16, Alpha=Get AlphaFE(), OverrideWorldGravity=Get Gravity Override()) {
    bClampLinearTranslationLimitToRefPose=true
    bDefaultToSkeletalMeshPhysicsAsset=false
    bOverrideWorldGravity=true
    overridePhysicsAsset=&/F_MED_OutlineLove_Flank_Front_Physics
    simulationSpace=BaseBoneSpace
    baseBoneRef {boneName=pelvis}
    componentAppliedLinearAccClamp=<0 0 0>
    simSpaceSettings {
     dampingAlpha=0
     maxAngularAcceleration=16
     maxAngularVelocity=8
     maxLinearAcceleration=2800
     maxLinearVelocity=700
     velocityScaleZ=0.3
     worldAlpha=1
     externalLinearDragV=<0.2 0.2 0.2>
    }
   }
   Save cached pose 'Skirt Physics '(Pose=Component To Local(ComponentPose=@n17)) {cacheName="Skirt Physics "}
   n18 = Get GravityOverride Params() -> (GravityOverride Params:Vector)
   n19 = RigidBody!(
    ComponentPose=Local To Component(LocalPose=Use cached pose 'Skirt Physics '())
    Alpha
    OverrideWorldGravity=@n18
    ExternalForce
    SimSpaceSettings:Sim Space Settings Structure=(WorldAlpha=1,VelocityScaleZ=1,DampingAlpha=1,MaxLinearVelocity=10000,MaxAngularVelocity=10000,MaxLinearAcceleration=10000,MaxAngularAcceleration=10000,ExternalLinearDragV=(X=0,Y=0,Z=0),ExternalLinearVelocity=(X=0,Y=0,Z=0),ExternalAngularVelocity=(X=0,Y=0,Z=0))
   ) -> (Pose) {
    bClampLinearTranslationLimitToRefPose=true
    bOverrideWorldGravity=true
    overridePhysicsAsset=&/F_MED_OutlineLove_Flank_Cape_Physics
    simulationSpace=BaseBoneSpace
    baseBoneRef {boneName=spine_05}
    componentAppliedLinearAccClamp=<0 0 0>
    simSpaceSettings {worldAlpha=1}
   }
   Save cached pose 'Cape Normal Physics'(Pose=Component To Local(ComponentPose=@n19)) {cacheName=Cape Normal Physics}
   n20 = RigidBody(
    ComponentPose=Local To Component(LocalPose=Use cached pose 'Skirt Physics '())
    OverrideWorldGravity=@n18
    SimSpaceSettings=(WorldAlpha=1,VelocityScaleZ=1,DampingAlpha=1,MaxLinearVelocity=10000,MaxAngularVelocity=10000,MaxLinearAcceleration=10000,MaxAngularAcceleration=10000,ExternalLinearDragV=(X=0,Y=0,Z=0),ExternalLinearVelocity=(X=0,Y=0,Z=0),ExternalAngularVelocity=(X=0,Y=0,Z=0))
   ) {
    bClampLinearTranslationLimitToRefPose=true
    bOverrideWorldGravity=true
    overridePhysicsAsset=&/F_MED_OutlineLove_Flank_Cape_Pack_Physics
    simulationSpace=BaseBoneSpace
    baseBoneRef {boneName=spine_05}
    componentAppliedLinearAccClamp=<0 0 0>
    simSpaceSettings {worldAlpha=1}
   }
   Save cached pose 'Cape Backpack Physics'(Pose=Component To Local(ComponentPose=@n20)) {cacheName=Cape Backpack Physics}
   Get Standing Fixes Coat() -> (Standing Fixes Coat:double)
  ]}
 EventGraph {nodes [
   Event BlueprintUpdateAnimation()
   > n1 = Sequence!(execute) -> (then_0, then_1, then_2:Exec, then_3:Exec)
   > n2 = "Gravity Override State Function (Parameter Asset) (Nativized)"(
    CharacterPartAnimInstance=Self-Reference()
    GameTimeInSeconds=GetGameTimeInSeconds()
    Parameters=&/F_MED_OutlineLove_Flank_Parameters
   )
   > Set GravityOverride Params(GravityOverride Params:Vector=@n2) -> (Output_Get:Vector)
   > Set RBAN Settings Out(RBAN Settings Out=@n2.OutRigidBodyAnimNodeParameters)
   n3 = CheckForBackbling()
   > Set isWearingPack(isWearingPack:bool=OR Boolean!(A=@n3, B=@n3.bIsBackblingACape, C:bool=@n3.bIsBackblingLong) -> (ReturnValue)) -> (Output_Get:bool)
   > Set isWearingCape(isWearingCape=@n3.bIsBackblingLong)
   Set is Knee Sliding CR(execute=@n1.then_1, is Knee Sliding CR:double="To Float (Boolean)"(InBool=Get bIsKneeSliding())) -> (Output_Get:double)
   > Set is Baseball Sliding CR(is Baseball Sliding CR:double="To Float (Boolean)"(InBool=Get bIsBaseballSliding())) -> (Output_Get:double)
   > Set is Crouching CR(is Crouching CR:double="To Float (Boolean)"(InBool=Get bIsCrouching())) -> (Output_Get:double)
   > Set is Moving 2D CR(is Moving 2D CR:double="To Float (Boolean)"(InBool=Get bIsMoving2D())) -> (Output_Get:double)
   n4 = NOT Boolean(A=Get bIsCrouching())
   n5 = Get bIsMoving2D()
   > Set is Standing Idle CR(is Standing Idle CR:double="To Float (Boolean)"(InBool=AND Boolean(A=@n5, B=@n4))) -> (Output_Get:double)
   n6 = OR Boolean!(A=Get bIsJumping(), B=Get bIsFalling()) -> (ReturnValue)
   n7 = Get bIsTacticalSprinting()
   > Set Tactical Sprint Jump(Tactical Sprint Jump:double="To Float (Boolean)"(InBool=AND Boolean(A=@n7, B=@n6))) -> (Output_Get:double)
   > Set Is Front End/Idle CR(Is Front End/Idle CR:double="To Float (Boolean)"(InBool=Get bIsInFrontEnd())) -> (Output_Get:double)
   > n8 = Branch(Condition=OR Boolean(A=Get bIsSkydiveFloating(), B=Get bIsSkydiveDiveMode()))
   > Set Gravity Override(Gravity Override="0,0,200")
   Set Gravity Override(execute=@n8.else, Gravity Override="0,0,-980")
   n9 = PlayerMovement(execute=@n1.then_2, AnimInstance=Self-Reference(), XDotProductCoefficient=100, YDotProductCoefficient=100, ZDotProductCoefficient=100)
   n10 = BreakVector(InVec=MakeVector(X=@n9, Y=@n9.YMovement))
   > n11 = Set Left(Left:double="To Float (Boolean)"(InBool="float < float"(A=@n10, B=-50))) -> (Output_Get:double)
   > n12 = Set Right(Right:double="To Float (Boolean)"(InBool="float > float"(A=@n10, B=50))) -> (Output_Get:double)
   > Set Forward(Forward:double="To Float (Boolean)"(InBool="float > float"(A=@n10.Y, B=40))) -> (Output_Get:double)
   > Set Backward(Backward:double="To Float (Boolean)"(InBool="float < float"(A=@n10.Y, B=-40))) -> (Output_Get:double)
   n13 = NOT Boolean(A=Get bIsSkydiveDiveMode())
   n14 = Get bIsSkyDiving()
   > Set Skydive Left(Skydive Left:double=float * float!(A=@n11, B:bool=@n14, C:bool=@n13) -> (ReturnValue) {is=title K2Node_PromotableOperator}) -> (Output_Get:double)
   > Set Skydive Right(Skydive Right:double=float * float(A=@n12, B=@n14, C=@n13)) -> (Output_Get:double)
   n15 = Get bIsPlayingEmote()
   > n16 = Branch(Condition=@n15)
   > Set AlphaFE(AlphaFE:double=0.4) -> (Output_Get:double)
   Set AlphaFE(execute=@n16.else, AlphaFE=0.8)
   AND Boolean(A=Get bIsInFrontEnd(), B=@n15)
   "float > float"(A=@n10.Y, B=40)
   Set Sim Space Settings Params(Sim Space Settings Params:Sim Space Settings Structure) -> (Output_Get:Sim Space Settings Structure)
  ]}
]
variables [
 AlphaFE {type=number}
 Backward {type=number}
 Forward {type=number}
 Gravity Override {type=Vector}
 GravityOverride Params {type=Vector}
 is Baseball Sliding CR {type=number}
 is Crouching CR {type=number}
 Is Front End/Idle CR {type=number}
 is Knee Sliding CR {type=number}
 is Moving 2D CR {type=number}
 is Standing Idle CR {type=number}
 isWearingCape {type=boolean}
 isWearingPack {type=boolean}
 Left {type=number}
 RBAN Settings Out {type=FortRigidBodyAnimNodeParameters}
 Right {type=number}
 Sim Space Settings Params {
  type=SimSpaceSettings
  default {
   dampingAlpha=0
   maxAngularAcceleration=16
   maxAngularVelocity=8
   maxLinearAcceleration=2800
   maxLinearVelocity=700
   velocityScaleZ=1
   worldAlpha=1
   externalLinearDragV=<0.2 0.2 0.2>
  }
 }
 Skydive Left {type=number}
 Skydive Right {type=number}
 Standing Fixes Coat {type=number}
 Tactical Sprint Jump {type=number}
]
