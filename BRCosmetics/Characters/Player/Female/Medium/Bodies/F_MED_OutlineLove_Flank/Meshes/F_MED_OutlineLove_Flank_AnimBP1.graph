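# Post-processes the OutlineLove Flank female cosmetic body, adding secondary body and cape motion to a pose copied from its attached parent mesh.
# A Control Rig fed movement-state floats precedes two RigidBody simulations based at pelvis and spine_05, then a post-sim rig.
# The event graph flips gravity upward during skydiving for both simulations.
# "Standing Idle" is actually moving-and-not-crouching; cape-base offsets have zero alpha, and directional flags go unused here.
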AnimBlueprint /BRCosmetics/Characters/Player/Female/Medium/Bodies/F_MED_OutlineLove_Flank/Meshes/F_MED_OutlineLove_Flank_AnimBP1
target_skeleton=&/F_MED_OutlineLove_Flank_Skeleton
graphs [
 AnimGraph {nodes [
   Component To Local()
   Local To Component()
   n1 = Control Rig!(
    Source=Copy Pose From Mesh() {bUseAttachedParent=true}
    Alpha
    "is Knee Sliding ":double=Get is Knee Sliding CR() -> (is Knee Sliding CR:double)
    is Baseball Sliding:double=Get is Baseball Sliding CR() -> (is Baseball Sliding CR:double)
    is Crouching:double=Get is Crouching CR() -> (is Crouching CR:double)
    is Moving 2D:double=Get is Moving 2D CR() -> (is Moving 2D CR:double)
   ) -> (Pose) {controlRigAssetReference {blueprintRigClass=&/F_MED_OutlineLove_Flank_CtrlRig.F_MED_OutlineLove_Flank_CtrlRig_C}}
   n2 = "Transform (Modify) Bone - Bone: ground_mover"(ComponentPose=Local To Component(LocalPose=@n1), Alpha=1, Translation="0,0,4", Rotation="0,0,0", Scale="1,1,1") {
    rotationMode=BMM_Additive
    rotationSpace=BCS_BoneSpace
    translationMode=BMM_Additive
    translationSpace=BCS_BoneSpace
    boneToModify {boneName=ground_mover}
    translation {z=4}
   }
   n3 = Get Gravity Override()
   n4 = RigidBody!(ComponentPose=@n2, Alpha, OverrideWorldGravity:Vector=@n3, ExternalForce) -> (Pose) {
    bClampLinearTranslationLimitToRefPose=true
    bDefaultToSkeletalMeshPhysicsAsset=false
    bOverrideWorldGravity=true
    overridePhysicsAsset=&/F_MED_OutlineLove_Flank_Physics
    simulationSpace=BaseBoneSpace
    baseBoneRef {boneName=pelvis}
    componentAppliedLinearAccClamp=<0 0 0>
    simSpaceSettings {
     dampingAlpha=0
     maxAngularAcceleration=16
     maxAngularVelocity=8
     maxLinearAcceleration=2800
     maxLinearVelocity=700
     velocityScaleZ=0.3
     worldAlpha=1
     externalLinearDragV=<0.5 0.5 0.5>
    }
   }
   n5 = RigidBody(ComponentPose=@n4, OverrideWorldGravity=@n3) {
    bClampLinearTranslationLimitToRefPose=true
    bOverrideWorldGravity=true
    overridePhysicsAsset=&/F_MED_OutlineLove_Flank_Cape_Physics
    simulationSpace=BaseBoneSpace
    baseBoneRef {boneName=spine_05}
    componentAppliedLinearAccClamp=<0 0 0>
    simSpaceSettings {
     dampingAlpha=0
     maxAngularAcceleration=16
     maxAngularVelocity=8
     maxLinearAcceleration=2800
     maxLinearVelocity=700
     worldAlpha=1
     externalLinearDragV=<0.2 0.2 0.2>
    }
   }
   n6 = "Transform (Modify) Bone - Bone: base_cape_l_00"(ComponentPose=@n5, Alpha=0, Translation="0,0,0", Rotation="0,15,0", Scale="1,1,1") {
    rotationMode=BMM_Additive
    rotationSpace=BCS_BoneSpace
    scaleMode=BMM_Additive
    scaleSpace=BCS_BoneSpace
    translationMode=BMM_Additive
    translationSpace=BCS_BoneSpace
    alphaScaleBiasClamp {bInterpResult=true}
    boneToModify {boneName=base_cape_l_00}
    rotation {yaw=15}
   }
   n7 = "Transform (Modify) Bone - Bone: base_cape_r_00"(ComponentPose=@n6, Alpha=0, Translation="0,0,0", Rotation="0,15,0", Scale="1,1,1") {
    rotationMode=BMM_Additive
    rotationSpace=BCS_BoneSpace
    scaleMode=BMM_Additive
    scaleSpace=BCS_BoneSpace
    translationMode=BMM_Additive
    translationSpace=BCS_BoneSpace
    alphaScaleBiasClamp {bInterpResult=true}
    boneToModify {boneName=base_cape_r_00}
    rotation {yaw=15}
   }
   n8 = Control Rig!(Source=Component To Local(ComponentPose=@n7), Alpha, is Stading Idle:double=Get is Standing Idle CR() -> (is Standing Idle CR:double)) -> (Pose) {controlRigAssetReference {blueprintRigClass=&/F_MED_OutlineLove_Flank_PostSim_CtrlRig.F_MED_OutlineLove_Flank_PostSim_CtrlRig_C}}
   Output Pose(Result=@n8) {layerGroup=None}
   Get Forward() -> (Forward:double)
   Get Standing Fixes Coat() -> (Standing Fixes Coat:double)
  ]}
 EventGraph {nodes [
   Event BlueprintUpdateAnimation()
   > n1 = Sequence()
   > Set is Knee Sliding CR(is Knee Sliding CR:double="To Float (Boolean)"(InBool=Get bIsKneeSliding())) -> (Output_Get:double)
   > Set is Baseball Sliding CR(is Baseball Sliding CR:double="To Float (Boolean)"(InBool=Get bIsBaseballSliding())) -> (Output_Get:double)
   > Set is Crouching CR(is Crouching CR:double="To Float (Boolean)"(InBool=Get bIsCrouching())) -> (Output_Get:double)
   > Set is Moving 2D CR(is Moving 2D CR:double="To Float (Boolean)"(InBool=Get bIsMoving2D())) -> (Output_Get:double)
   n2 = NOT Boolean(A=Get bIsCrouching())
   n3 = Get bIsMoving2D()
   > Set is Standing Idle CR(is Standing Idle CR:double="To Float (Boolean)"(InBool=AND Boolean(A=@n3, B=@n2))) -> (Output_Get:double)
   > n4 = Branch(Condition=OR Boolean(A=Get bIsSkydiveFloating(), B=Get bIsSkydiveDiveMode()))
   > Set Gravity Override(Gravity Override="0,0,200")
   Set Gravity Override(execute=@n4.else, Gravity Override="0,0,-980")
   n5 = PlayerMovement(execute=@n1.then_1, AnimInstance=Self-Reference(), XDotProductCoefficient=100, YDotProductCoefficient=100, ZDotProductCoefficient=100)
   n6 = BreakVector(InVec=MakeVector(X=@n5, Y=@n5.YMovement))
   > Set Left(Left:double="To Float (Boolean)"(InBool="float < float"(A=@n6, B=-50))) -> (Output_Get:double)
   > Set Right(Right:double="To Float (Boolean)"(InBool="float > float"(A=@n6, B=50))) -> (Output_Get:double)
   > Set Forward(Forward:double="To Float (Boolean)"(InBool="float > float"(A=@n6.Y, B=40))) -> (Output_Get:double)
   > Set Backward(Backward:double="To Float (Boolean)"(InBool="float < float"(A=@n6.Y, B=-40))) -> (Output_Get:double)
   "float > float"(A=@n6.Y, B=40)
  ]}
]
variables [
 Backward {type=number}
 Forward {type=number}
 Gravity Override {type=Vector}
 is Baseball Sliding CR {type=number}
 is Crouching CR {type=number}
 is Knee Sliding CR {type=number}
 is Moving 2D CR {type=number}
 is Standing Idle CR {type=number}
 Left {type=number}
 Right {type=number}
 Standing Fixes Coat {type=number}
]
