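# Control Rig for the F_MED_OutlineLove_Flank face accessory mesh that poses the hair tail bones as the head moves.
# Setup adds FK controls through a shared function, and the forward solve calls a companion set-FK function with its flag off.
# One corrective rotates Master_hairTail, weighted by head rotation relative to spine_05 through a remapped Euler angle.
# Two spherical pose readers on the head weight further transform offsets on the hair tail bones.
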ControlRigBlueprint /BRCosmetics/Characters/Player/Female/Medium/Bodies/F_MED_OutlineLove_Flank/Meshes/Parts/F_MED_OutlineLove_Flank_FaceAcc_CtrlRig
graphs [
 RigVMFunctionLibrary {links [], nodes []}
 RigVMModel {nodes [
   n1 = &RigUnit_PrepareForExecution()
   addFKCtrl = RigVMFunctionReferenceNode(
    Custom_FKCtrl=()
    Options_FKCtrl="(pelvis_1_#1=False,spine_3_#2=False,clavicle_5_#3=true,arm_7_#4=true,leg_9_#5=False,head_11_46DC15EF465CD2 ...(truncated)"
    Shape_Settings="(pelvis_2_#6=(bVisible=True,Name=\"Hexagon_Thin\",Color=(R=0.0117,G=0.8281,B=1,A=1),Transform=(Rotation=(X=0,Y=-0.7071,Z=0,W=0.7071),Translation=(X=2,Y=0,Z=0),Scale3D=(X=5,Y=5,Z=7))),spine_4_#7=(bVi ...(truncated)"
    ExecuteContext=@n1.ExecutePin
   )
   RigUnit_BeginExecution = &RigUnit_BeginExecution()
   setFKCtrl = RigVMFunctionReferenceNode(setFK_Ctrl=false, ExecuteContext=@RigUnit_BeginExecution.ExecutePin)
   n2 = &RigUnit_GetRelativeTransformForItem(Child=(Type=Bone,Name="head"), Parent=(Type=Bone,Name="spine_05"), bChildInitial=False, bParentInitial=False)
   > &RigUnit_ModifyTransforms(
    ItemToModify=((Item=(Type=Bone,Name="Master_hairTail"),Transform=(Rotation=(X=0,Y=0,Z=0.515,W=0.8572),Translation=(X=0,Y=0,Z=0),Scale3D=(X=1,Y=1,Z=1))))
    WeightMinimum=0
    Weight=&RigVMFunction_MathDoubleRemap(
     SourceMaximum=-60
     SourceMinimum=4
     TargetMaximum=1
     TargetMinimum=0
     Value=&RigVMFunction_MathQuaternionToEuler(RotationOrder=ZXY, Value=@n2.RelativeTransform.Rotation).Result.Z
    ).Result
   )
   > n3 = &RigUnit_SphericalPoseReader(
    ActiveRegionScaleFactors=(PositiveWidth=1,NegativeWidth=1,PositiveHeight=1,NegativeHeight=1)
    ActiveRegionSize=0.1
    Debug=(bDrawDebug=false,bDraw2D=False,bDrawLocalAxes=False,DebugScale=25,DebugSegments=20,DebugThickness=0.25)
    DriverAxis=(X=0,Y=1,Z=0)
    DriverItem=(Type=Bone,Name="head")
    FalloffRegionScaleFactors=(PositiveWidth=1,NegativeWidth=1,PositiveHeight=1,NegativeHeight=1)
    FalloffSize=0.3
    FlipHeightScaling=False
    FlipWidthScaling=False
    OptionalParentItem=(Type=Bone,Name="None")
    RotationOffset=(X=160,Y=0,Z=-25)
   )
   > &RigUnit_ModifyTransforms(
    ItemToModify="((Item=(Type=Bone,Name=\"Master_hairTail\"),Transform=(Rotation=(X=0,Y=0.2164,Z=0,W=0.9763),Translation=(X=0,Y=0,Z=0),Scale3D=(X=1,Y=1,Z=1))),(Item=(Type=Bone,Name=\"base_hairTail_01\"),Transform=(Rotation=(X=-0.3126,Y=0.1797,Z=-0.1178,W=0.9253),Translation=(X ...(truncated)"
    WeightMinimum=0
    Weight=@n3.OutputParam
   )
   > n4 = &RigUnit_SphericalPoseReader(
    ActiveRegionScaleFactors=(PositiveWidth=1,NegativeWidth=1,PositiveHeight=1,NegativeHeight=1)
    ActiveRegionSize=0.1
    Debug=(bDrawDebug=false,bDraw2D=False,bDrawLocalAxes=False,DebugScale=25,DebugSegments=20,DebugThickness=0.25)
    DriverAxis=(X=0,Y=1,Z=0)
    DriverItem=(Type=Bone,Name="head")
    FalloffRegionScaleFactors=(PositiveWidth=1,NegativeWidth=1,PositiveHeight=1,NegativeHeight=1)
    FalloffSize=0.3
    FlipHeightScaling=False
    FlipWidthScaling=False
    OptionalParentItem=(Type=Bone,Name="None")
    RotationOffset=(X=0,Y=0,Z=-25)
   )
   > &RigUnit_ModifyTransforms(
    ItemToModify="((Item=(Type=Bone,Name=\"Master_hairTail\"),Transform=(Rotation=(X=-0.0226,Y=-0.2578,Z=-0.0842,W=0.9623),Translation=(X=0,Y=0,Z=0),Scale3D=(X=1,Y=1,Z=1))),(Item=(Type=Bone,Name=\"base_hairTail_01\"),Transform=(Rotation=(X=0.2335,Y=-0.138,Z=-0.1862,W=0.9443),Tr ...(truncated)"
    WeightMinimum=0
    Weight=@n4.OutputParam
   )
  ]}
]
hierarchy [
 root {loc=<0 0 0>, rot=<0 0 0 1>}
 pelvis {parent=root, loc=<0 -0.5301 98.0571>, rot=<0.0127 0.707 -0.0127 -0.707>}
 spine_01 {parent=pelvis, loc=<6.1208 -0.9328 0>, rot=<0 0 0.034 -0.9994>}
 spine_02 {parent=spine_01, loc=<6.5997 -0.2095 0>, rot=<0 0 0.0238 -0.9997>}
 spine_03 {parent=spine_02, loc=<6.4914 -0.1237 0>, rot=<0 0 0.0149 -0.9999>}
 spine_04 {parent=spine_03, loc=<6.9598 0.0794 0>, rot=<0 0 -0.0132 -0.9999>}
 spine_05 {parent=spine_04, loc=<6.9987 0 0>, rot=<0 0 -0.0015 -1>}
 neck_01 {parent=spine_05, loc=<11.3138 0.1529 0>, rot=<0 0 -0.2003 -0.9797>}
 neck_02 {parent=neck_01, loc=<4.5627 0 0>, rot=<0 0 0.0504 -0.9987>}
 head {parent=neck_02, loc=<5.3914 0 0>, rot=<0 0 0.115 -0.9934>}
 base_hairFront_l_01 {parent=head, loc=<2.8473 2.7209 -6.4363>, rot=<0 0 1 -0.0041>}
 dyn_hairFront_l_01 {parent=base_hairFront_l_01, loc=<0 0 0>, rot=<0 0 0 1>}
 dyn_hairFront_l_02 {parent=dyn_hairFront_l_01, loc=<9.7013 -0.3564 -0.4202>, rot=<0.0152 0.0344 0.241 -0.9698>}
 dyn_hairFront_l_03 {parent=dyn_hairFront_l_02, loc=<6.6295 -0.0494 -0.0278>, rot=<0.0035 -0.0209 0.0369 -0.9991>}
 dyn_hairFront_l_04 {parent=dyn_hairFront_l_03, loc=<6.7548 0.1851 -0.0075>, rot=<0.0013 -0.0072 -0.0812 -0.9967>}
 dyn_hairFront_l_05 {parent=dyn_hairFront_l_04, loc=<6.6578 -0.0986 0>, rot=<-0.0024 0 0 -1>}
 base_hairFront_r_02 {parent=head, loc=<2.8474 2.7209 6.4363>, rot=<0 0 -0.0041 -1>}
 dyn_hairFront_r_01 {parent=base_hairFront_r_02, loc=<0 0 0>, rot=<0 0 0 1>}
 dyn_hairFront_r_02 {parent=dyn_hairFront_r_01, loc=<-9.6514 0.3179 -0.4126>, rot=<0.0152 0.0344 0.241 -0.9698>}
 dyn_hairFront_r_03 {parent=dyn_hairFront_r_02, loc=<-6.5894 0.3042 0.5523>, rot=<0.0035 -0.0209 0.0369 -0.9991>}
 dyn_hairFront_r_04 {parent=dyn_hairFront_r_03, loc=<-6.7757 -0.3299 -0.2365>, rot=<0.0013 -0.0072 -0.0812 -0.9967>}
 dyn_hairFront_r_05 {parent=dyn_hairFront_r_04, loc=<-6.6575 0.0812 0.0004>, rot=<-0.0024 0 0 -1>}
 Master_hairTail {parent=head, loc=<0.8098 -7.4595 0>, rot=<0 0 0.9781 -0.2079>}
 base_hairTail_01 {parent=Master_hairTail, loc=<4.021 -0.5485 0>, rot=<0 0 0 1>}
 dyn_hairTail_01 {parent=base_hairTail_01, loc=<0 0 0>, rot=<0 0 0 1>}
 dyn_hairTail_02 {parent=dyn_hairTail_01, loc=<7.7011 -0.7328 0>, rot=<0 0 0.0034 -1>}
 dyn_hairTail_03 {parent=dyn_hairTail_02, loc=<7.3804 0.0892 0>, rot=<0 0 0.127 -0.9919>}
 dyn_hairTail_04 {parent=dyn_hairTail_03, loc=<7.2704 -0.1498 0>, rot=<0 0 0.1705 -0.9854>}
 dyn_hairTail_05 {parent=dyn_hairTail_04, loc=<8.466 0.0275 0>, rot=<0 0 -0.0652 -0.9979>}
 clavicle_l {parent=spine_05, loc=<5.4831 -0.1053 -3.2946>, rot=<0.0465 -0.7289 -0.009 -0.6829>}
 upperarm_l {parent=clavicle_l, loc=<9.8658 -1.2136 -0.4105>, rot=<0.0164 -0.3245 -0.0269 -0.9454>}
 lowerarm_l {parent=upperarm_l, loc=<26.465 -0.8345 0.259>, rot=<0.0036 0.0161 -0.2298 -0.9731>}
 hand_l {parent=lowerarm_l, loc=<22.7355 -0.3203 -0.6105>, rot=<-0.6784 -0.0223 0.0735 -0.7307>}
 deform_wrist_dn_base_l {parent=hand_l, loc=<0 -1.5326 0>, rot=<0 0 0 -1>}
 deform_wrist_dn_l {parent=deform_wrist_dn_base_l, loc=<0 0 0>, rot=<0 0 0 1>}
 deform_wrist_up_base_l {parent=hand_l, loc=<0 2.3009 0>, rot=<0 0 0 -1>}
 deform_wrist_up_l {parent=deform_wrist_up_base_l, loc=<0 0 0>, rot=<0 0 0 1>}
 deform_elbow_in_base_l {parent=lowerarm_l, loc=<0.1834 3.6062 -0.2572>, rot=<0 0 0 -1>}
 deform_elbow_in_l {parent=deform_elbow_in_base_l, loc=<0 0 0>, rot=<0 0 0 1>}
 deform_elbow_out_base_l {parent=lowerarm_l, loc=<0 -4 0>, rot=<0 0 0 1>}
 deform_elbow_out_l {parent=deform_elbow_out_base_l, loc=<0 0 0>, rot=<0 0 0 1>}
 upperarm_twist_01_l {parent=upperarm_l, loc=<7.2135 -0.1134 0>, rot=<0 0 0 1>}
 deform_upperarm_bk_base_l {parent=upperarm_twist_01_l, loc=<-6.4053 -4.3746 0.314>, rot=<0 0 0 -1>}
 deform_upperarm_bk_l {parent=deform_upperarm_bk_base_l, loc=<0 0 0>, rot=<0 0 0 1>}
 deform_upperarm_fr_base_l {parent=upperarm_twist_01_l, loc=<-5.6312 5.092 0.3354>, rot=<0 0 0 -1>}
 deform_upperarm_fr_l {parent=deform_upperarm_fr_base_l, loc=<0 0 0>, rot=<0 0 0 1>}
 deform_upperarm_up_base_l {parent=upperarm_twist_01_l, loc=<-8.891 0.4825 3.9517>, rot=<0 0 0 -1>}
 deform_upperarm_up_l {parent=deform_upperarm_up_base_l, loc=<0 0 0>, rot=<0 0 0 1>}
 base_shoulderPad_l_02 {parent=clavicle_l, loc=<12.2306 3.9824 2.6454>, rot=<0.8939 0.1728 -0.1233 -0.3948>}
 dyn_shoulderPad_l_02 {parent=base_shoulderPad_l_02, loc=<0 0 0>, rot=<0 0 0 1>}
 base_shoulderPad_l_03 {parent=clavicle_l, loc=<13.8322 -1.6765 5.4324>, rot=<0.7071 0.0415 0.0059 -0.7059>}
 dyn_shoulderPad_l_03 {parent=base_shoulderPad_l_03, loc=<0 0 0>, rot=<0 0 0 1>}
 base_shoulderPad_l_04 {parent=clavicle_l, loc=<12.602 -6.416 2.0762>, rot=<0.1542 -0.2993 0.0862 -0.9377>}
 dyn_shoulderPad_l_04 {parent=base_shoulderPad_l_04, loc=<0 0 0>, rot=<0 0 0 1>}
 base_shoulderPad_l_01 {parent=clavicle_l, loc=<8.0696 6.9938 -1.6135>, rot=<0.8723 0.2165 -0.2957 -0.3236>}
 dyn_shoulderPad_l_01 {parent=base_shoulderPad_l_01, loc=<0 0 0>, rot=<0 0 0 1>}
 base_shoulderPad_l_05 {parent=clavicle_l, loc=<9.5881 -7.4827 -2.3724>, rot=<0.0716 -0.3834 0.0506 -0.9194>}
 dyn_shoulderPad_l_05 {parent=base_shoulderPad_l_05, loc=<0 0 0>, rot=<0 0 0 1>}
 base_cape_l_00 {parent=clavicle_l, loc=<11.158 -7.1525 -3.0942>, rot=<0.7509 0.0109 -0.6558 0.0773>}
 dyn_cape_l_00 {parent=base_cape_l_00, loc=<0 0 0>, rot=<0 0 0 1>}
 dyn_cape_l_01 {parent=dyn_cape_l_00, loc=<4.2929 0 0>, rot=<0 0 0 1>}
 dyn_cape_l_02 {parent=dyn_cape_l_01, loc=<21 0 0>, rot=<0 0 -0.0054 -1>}
 clavicle_r {parent=spine_05, loc=<5.4831 -0.1053 3.2946>, rot=<0.7289 0.0465 -0.6829 0.009>}
 upperarm_r {parent=clavicle_r, loc=<-9.8658 1.2136 0.4105>, rot=<0.0164 -0.3245 -0.0269 -0.9454>}
 lowerarm_r {parent=upperarm_r, loc=<-26.465 0.8345 -0.259>, rot=<0.0036 0.0161 -0.2298 -0.9731>}
 hand_r {parent=lowerarm_r, loc=<-22.7355 0.3203 0.6106>, rot=<-0.6784 -0.0223 0.0735 -0.7307>}
 deform_wrist_dn_base_r {parent=hand_r, loc=<-0.0001 1.5328 0>, rot=<0 0 0 -1>}
 deform_wrist_dn_r {parent=deform_wrist_dn_base_r, loc=<0 0 0>, rot=<0 0 0 1>}
 deform_wrist_up_base_r {parent=hand_r, loc=<-0.0002 -2.3007 -0.0001>, rot=<0 0 0 -1>}
 deform_wrist_up_r {parent=deform_wrist_up_base_r, loc=<0 0 0>, rot=<0 0 0 1>}
 deform_elbow_in_base_r {parent=lowerarm_r, loc=<-0.1834 -3.6062 0.2572>, rot=<0 0 0 -1>}
 deform_elbow_in_r {parent=deform_elbow_in_base_r, loc=<0 0 0>, rot=<0 0 0 1>}
 deform_elbow_out_base_r {parent=lowerarm_r, loc=<0 4 -0.0001>, rot=<0 0 0 -1>}
 deform_elbow_out_r {parent=deform_elbow_out_base_r, loc=<0 0 0>, rot=<0 0 0 1>}
 upperarm_twist_01_r {parent=upperarm_r, loc=<-7.2135 0.1134 0>, rot=<0 0 0 -1>}
 deform_upperarm_bk_base_r {parent=upperarm_twist_01_r, loc=<6.4053 4.3746 -0.314>, rot=<0 0 0 -1>}
 deform_upperarm_bk_r {parent=deform_upperarm_bk_base_r, loc=<0 0 0>, rot=<0 0 0 1>}
 deform_upperarm_fr_base_r {parent=upperarm_twist_01_r, loc=<5.6312 -5.092 -0.3354>, rot=<0 0 0 -1>}
 deform_upperarm_fr_r {parent=deform_upperarm_fr_base_r, loc=<0 0 0>, rot=<0 0 0 1>}
 deform_upperarm_up_base_r {parent=upperarm_twist_01_r, loc=<8.891 -0.4825 -3.9517>, rot=<0 0 0 -1>}
 deform_upperarm_up_r {parent=deform_upperarm_up_base_r, loc=<0 0 0>, rot=<0 0 0 1>}
 base_shoulderPad_r_02 {parent=clavicle_r, loc=<-12.2306 -3.9824 -2.6455>, rot=<0.8939 0.1728 -0.1233 -0.3948>}
 dyn_shoulderPad_r_02 {parent=base_shoulderPad_r_02, loc=<0 0 0>, rot=<0 0 0 1>}
 base_shoulderPad_r_03 {parent=clavicle_r, loc=<-13.8323 1.6765 -5.4322>, rot=<0.7071 0.0415 0.0059 -0.7059>}
 dyn_shoulderPad_r_03 {parent=base_shoulderPad_r_03, loc=<0 0 0>, rot=<0 0 0 1>}
 base_shoulderPad_r_04 {parent=clavicle_r, loc=<-12.602 6.4159 -2.0761>, rot=<0.1542 -0.2993 0.0862 -0.9377>}
 dyn_shoulderPad_r_04 {parent=base_shoulderPad_r_04, loc=<0 0 0>, rot=<0 0 0 1>}
 base_shoulderPad_r_01 {parent=clavicle_r, loc=<-8.0696 -6.9938 1.6133>, rot=<0.8723 0.2165 -0.2957 -0.3236>}
 dyn_shoulderPad_r_01 {parent=base_shoulderPad_r_01, loc=<0 0 0>, rot=<0 0 0 -1>}
 base_shoulderPad_r_05 {parent=clavicle_r, loc=<-9.5881 7.4827 2.3721>, rot=<0.0716 -0.3834 0.0506 -0.9194>}
 dyn_shoulderPad_r_05 {parent=base_shoulderPad_r_05, loc=<0 0 0>, rot=<0 0 0 1>}
 base_cape_r_00 {parent=clavicle_r, loc=<-11.1586 7.1577 3.0939>, rot=<-0.0109 0.7509 -0.0773 -0.6558>}
 dyn_cape_r_00 {parent=base_cape_r_00, loc=<0 0 0>, rot=<0 0 0 1>}
 dyn_cape_r_01 {parent=dyn_cape_r_00, loc=<4.2931 0 0>, rot=<0.0087 0 0 -1>}
 dyn_cape_r_02 {parent=dyn_cape_r_01, loc=<20.9994 0 0>, rot=<0 0 -0.0054 -1>}
 deform_pec_aim_base_l {parent=spine_04, loc=<3.2456 9.3309 -1.2237>, rot=<0.0603 -0.4033 0.3318 -0.8507>}
 deform_pec_aim_l {parent=deform_pec_aim_base_l, loc=<0 0 0>, rot=<0 0 0 1>}
 deform_pec_base_l {parent=deform_pec_aim_l, loc=<9.3665 0 0>, rot=<0 0 0 1>}
 deform_pec_l {parent=deform_pec_base_l, loc=<0 0 0>, rot=<0 0 0 1>}
 deform_pec_aim_base_r {parent=spine_04, loc=<3.2452 9.3309 1.2237>, rot=<-0.4033 -0.0603 0.8507 0.3318>}
 deform_pec_aim_r {parent=deform_pec_aim_base_r, loc=<0 0 0>, rot=<0 0 0 1>}
 deform_pec_base_r {parent=deform_pec_aim_r, loc=<-9.3667 -0.0003 -0.0004>, rot=<0 0 0 1>}
 deform_pec_r {parent=deform_pec_base_r, loc=<0 0 0>, rot=<0 0 0 1>}
 thigh_l {parent=pelvis, loc=<-1.0855 -0.2109 -8.7053>, rot=<0.0009 0.0355 0.0155 -0.9993>}
 calf_l {parent=thigh_l, loc=<-42.5284 -0.0019 -0.7297>, rot=<0.0006 -0.0161 -0.0391 -0.9991>}
 deform_knee_in_base_l {parent=calf_l, loc=<0.8031 -5.0231 0.2456>, rot=<0 0 0 -1>}
 deform_knee_in_l {parent=deform_knee_in_base_l, loc=<0 0 0>, rot=<0 0 0 1>}
 deform_knee_out_base_l {parent=calf_l, loc=<0 4 0>, rot=<0 0 0 -1>}
 deform_knee_out_l {parent=deform_knee_out_base_l, loc=<0 0 0>, rot=<0 0 0 1>}
 deform_groin_base_l {parent=thigh_l, loc=<0.4801 5.6776 -0.2486>, rot=<0 0 0 -1>}
 deform_groin_l {parent=deform_groin_base_l, loc=<0 0 0>, rot=<0 0 0 1>}
 thigh_r {parent=pelvis, loc=<-1.0855 -0.2109 8.7053>, rot=<0.0355 -0.0009 0.9993 0.0155>}
 calf_r {parent=thigh_r, loc=<42.5284 0.0019 0.7297>, rot=<0.0006 -0.0161 -0.0391 -0.9991>}
 deform_knee_in_base_r {parent=calf_r, loc=<-0.8031 5.0231 -0.2457>, rot=<0 0 0 -1>}
 deform_knee_in_r {parent=deform_knee_in_base_r, loc=<0 0 0>, rot=<0 0 0 1>}
 deform_knee_out_base_r {parent=calf_r, loc=<0 -4 0>, rot=<0 0 0 -1>}
 deform_knee_out_r {parent=deform_knee_out_base_r, loc=<0 0 0>, rot=<0 0 0 1>}
 deform_groin_base_r {parent=thigh_r, loc=<-0.4801 -5.6776 0.2486>, rot=<0 0 0 -1>}
 deform_groin_r {parent=deform_groin_base_r, loc=<0 0 0>, rot=<0 0 0 1>}
 deform_glute_base_l {parent=pelvis, loc=<-7.3466 -9.0213 -9.1245>, rot=<0.0172 0.031 -0.4477 -0.8935>}
 deform_glute_l {parent=deform_glute_base_l, loc=<0 0 0>, rot=<0 0 0 1>}
 deform_glute_base_r {parent=pelvis, loc=<-7.3464 -9.0213 9.1245>, rot=<0.031 -0.0172 0.8935 -0.4477>}
 deform_glute_r {parent=deform_glute_base_r, loc=<0 0 0>, rot=<0 0 0 1>}
]
