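# Physics asset for the front-hair face accessory of the OutlineLove Flank medium female player cosmetic, defining capsule bodies for hair physics.
# Left and right five-segment hair chains hang from base bodies linked to the head, with tight swing and near-locked twist limits between segments.
# Upper spine, neck, head, clavicle, upper arm and shoulder pad capsules approximate the body around the hair, joined by free constraints.
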PhysicsAsset /BRCosmetics/Characters/Player/Female/Medium/Bodies/F_MED_OutlineLove_Flank/Meshes/Parts/F_MED_OutlineLove_Flank_FaceAcc_FHair_Physics
bodies {
 base_hairFront_l_01 [{shapeType=Capsule, length=0.505, radius=0.505, shapeName=base_hairFront_l_01_capsule, rotation=<rot 0 -90 0>}]
 base_hairFront_r_02 [{shapeType=Capsule, length=0.505, radius=0.505, shapeName=base_hairFront_r_02_capsule, rotation=<rot 0 -90 0>}]
 clavicle_l [{
   shapeType=Capsule
   length=9.5596
   radius=7.461
   shapeName=clavicle_l_capsule
   center=<3.7353 -1.6729 -1.3862>
   rotation=<rot 82.6577 -0.005 -0.0051>
  }]
 clavicle_r [{
   shapeType=Capsule
   length=9.5596
   radius=7.461
   shapeName=clavicle_l_capsule
   center=<-3.7556 1.3682 1.2288>
   rotation=<rot 82.6577 -0.005 -0.0051>
  }]
 dyn_hairFront_l_01 [{
   shapeType=Capsule
   length=4.7335
   radius=2
   shapeName=dyn_hairFront_l_01_capsule
   center=<4.7443 -2.0548 -1.177>
   rotation=<rot 76.867 179.9775 180>
  }]
 dyn_hairFront_l_02 [{shapeType=Capsule, length=4.7335, radius=2, shapeName=dyn_hairFront_l_02_capsule, center=<3.2658 0 -0.4398>, rotation=<rot 90 0 0>}]
 dyn_hairFront_l_03 [{shapeType=Capsule, length=4.7335, radius=2, shapeName=dyn_hairFront_l_03_capsule, center=<3.2658 0 -0.2269>, rotation=<rot 90 0 0>}]
 dyn_hairFront_l_04 [{shapeType=Capsule, length=4.7335, radius=2, shapeName=dyn_hairFront_l_04_capsule, center=<3.2658 0 -0.3319>, rotation=<rot 90 0 0>}]
 dyn_hairFront_l_05 [{shapeType=Capsule, length=6.3624, radius=2, shapeName=dyn_hairFront_l_05_capsule, center=<3.802 0.3041 -0.3156>, rotation=<rot 90 11.8885 0>}]
 dyn_hairFront_r_01 [{
   shapeType=Capsule
   length=4.7335
   radius=2
   shapeName=dyn_hairFront_l_01_capsule
   center=<-5.4456 2.2795 0.8852>
   rotation=<rot 75.4499 179.9999 180>
  }]
 dyn_hairFront_r_02 [{
   shapeType=Capsule
   length=4.7335
   radius=2
   shapeName=dyn_hairFront_l_02_capsule
   center=<-3.2594 0 0.4258>
   rotation=<rot 87.5988 0.0001 0.0002>
  }]
 dyn_hairFront_r_03 [{shapeType=Capsule, length=4.7335, radius=2, shapeName=dyn_hairFront_l_03_capsule, center=<-3.2658 -0.2273 0.2704>, rotation=<rot 90 0 0>}]
 dyn_hairFront_r_04 [{shapeType=Capsule, length=4.7335, radius=2, shapeName=dyn_hairFront_l_04_capsule, center=<-3.2658 0 0.1735>, rotation=<rot 90 0 0>}]
 dyn_hairFront_r_05 [{
   shapeType=Capsule
   length=6.0718
   radius=2.2714
   shapeName=dyn_hairFront_l_05_capsule
   center=<-3.672 -0.1393 0.3228>
   rotation=<rot 90 18.1565 0>
  }]
 dyn_shoulderPad_l_01 [{
   shapeType=Capsule
   length=3.6998
   radius=2.0135
   shapeName=dyn_shoulderPad_l_01_capsule
   center=<-0.185 1.3066 0.0058>
   rotation=<rot 0 -0.0001 -1.8548>
  }]
 dyn_shoulderPad_l_02 [{
   shapeType=Capsule
   length=5.1924
   radius=2
   shapeName=dyn_shoulderPad_l_02_capsule
   center=<-0.2356 1.419 -0.013>
   rotation=<rot 6.6538 -0.0003 2.5623>
  }]
 dyn_shoulderPad_r_01 [{
   shapeType=Capsule
   length=3.6998
   radius=2.0135
   shapeName=dyn_shoulderPad_l_01_capsule
   center=<0.185 -1.3066 0.0058>
   rotation=<rot 0 -0.0001 -1.8548>
  }]
 dyn_shoulderPad_r_02 [{
   shapeType=Capsule
   length=5.1924
   radius=2
   shapeName=dyn_shoulderPad_l_02_capsule
   center=<0.2356 -1.419 -0.013>
   rotation=<rot 6.6538 -0.0003 2.5623>
  }]
 head [{shapeType=Capsule, length=3.4095, radius=5.9904, shapeName=head_capsule, center=<1.1701 3.1654 0>, rotation=<rot 90 0 0>}]
 neck_01 [{shapeType=Capsule, length=3.9305, radius=4.2803, shapeName=neck_01_capsule, center=<2.1021 0.4842 0>, rotation=<rot 90 0 0>}]
 spine_04 [{shapeType=Capsule, length=5.4197, radius=11.3695, shapeName=spine_04_capsule, center=<-0.2925 2.0226 0>, rotation=<rot 0 -90 0>}]
 spine_05 [
  {shapeType=Capsule, length=10.2283, radius=10.4175, shapeName=spine_05_capsule, center=<-3.5692 0.5704 0>, rotation=<rot 0 -90 0>}
  {shapeType=Capsule, length=10.2357, radius=10.3743, shapeName=spine_05_capsule, center=<0.8631 -2.8022 0>}
  {shapeType=Capsule, length=33.0073, radius=4.7612, shapeName=spine_05_capsule, center=<12.9329 -6.3602 0>}
 ]
 upperarm_l [{shapeType=Capsule, length=26.386, radius=5.6141, shapeName=upperarm_l_capsule, center=<14.2936 0 -0.5693>, rotation=<rot 90 180 0>}]
 upperarm_r [{shapeType=Capsule, length=26.386, radius=5.6141, shapeName=upperarm_l_capsule, center=<-14.2936 1.0983 0.5693>, rotation=<rot 90 180 0>}]
}
constraints [
 {
  bone1Name=spine_05
  bone2Name=spine_04
  swing1LimitDegrees=45
  swing1Motion=Free
  swing2LimitDegrees=45
  swing2Motion=Free
  twistLimitDegrees=45
  twistMotion=Free
 }
 {
  bone1Name=clavicle_l
  bone2Name=spine_05
  swing1LimitDegrees=45
  swing1Motion=Free
  swing2LimitDegrees=45
  swing2Motion=Free
  twistLimitDegrees=45
  twistMotion=Free
 }
 {
  bone1Name=clavicle_r
  bone2Name=spine_05
  swing1LimitDegrees=45
  swing1Motion=Free
  swing2LimitDegrees=45
  swing2Motion=Free
  twistLimitDegrees=45
  twistMotion=Free
 }
 {
  bone1Name=upperarm_l
  bone2Name=clavicle_l
  swing1LimitDegrees=45
  swing1Motion=Free
  swing2LimitDegrees=45
  swing2Motion=Free
  twistLimitDegrees=45
  twistMotion=Free
 }
 {
  bone1Name=upperarm_r
  bone2Name=clavicle_r
  swing1LimitDegrees=45
  swing1Motion=Free
  swing2LimitDegrees=45
  swing2Motion=Free
  twistLimitDegrees=45
  twistMotion=Free
 }
 {
  bone1Name=base_hairFront_l_01
  bone2Name=head
  swing1LimitDegrees=45
  swing1Motion=Free
  swing2LimitDegrees=45
  swing2Motion=Free
  twistLimitDegrees=45
  twistMotion=Free
 }
 {bone1Name=dyn_hairFront_l_01, bone2Name=base_hairFront_l_01, swing1LimitDegrees=15, swing2LimitDegrees=5, twistLimitDegrees=1}
 {bone1Name=dyn_hairFront_l_02, bone2Name=dyn_hairFront_l_01, swing1LimitDegrees=25, swing2LimitDegrees=5, twistLimitDegrees=1}
 {bone1Name=dyn_hairFront_l_03, bone2Name=dyn_hairFront_l_02, swing1LimitDegrees=25, swing2LimitDegrees=5, twistLimitDegrees=1}
 {bone1Name=dyn_hairFront_l_04, bone2Name=dyn_hairFront_l_03, swing1LimitDegrees=25, swing2LimitDegrees=5, twistLimitDegrees=1}
 {bone1Name=dyn_hairFront_l_05, bone2Name=dyn_hairFront_l_04, swing1LimitDegrees=25, swing2LimitDegrees=5, twistLimitDegrees=1}
 {
  bone1Name=base_hairFront_r_02
  bone2Name=head
  swing1LimitDegrees=45
  swing1Motion=Free
  swing2LimitDegrees=45
  swing2Motion=Free
  twistLimitDegrees=45
  twistMotion=Free
 }
 {bone1Name=dyn_hairFront_r_01, bone2Name=base_hairFront_r_02, swing1LimitDegrees=15, swing2LimitDegrees=5, twistLimitDegrees=1}
 {bone1Name=dyn_hairFront_r_02, bone2Name=dyn_hairFront_r_01, swing1LimitDegrees=25, swing2LimitDegrees=5, twistLimitDegrees=1}
 {bone1Name=dyn_hairFront_r_03, bone2Name=dyn_hairFront_r_02, swing1LimitDegrees=25, swing2LimitDegrees=5, twistLimitDegrees=1}
 {bone1Name=dyn_hairFront_r_04, bone2Name=dyn_hairFront_r_03, swing1LimitDegrees=25, swing2LimitDegrees=5, twistLimitDegrees=1}
 {bone1Name=dyn_hairFront_r_05, bone2Name=dyn_hairFront_r_04, swing1LimitDegrees=25, swing2LimitDegrees=5, twistLimitDegrees=1}
 {
  bone1Name=head
  bone2Name=neck_01
  swing1LimitDegrees=45
  swing1Motion=Free
  swing2LimitDegrees=45
  swing2Motion=Free
  twistLimitDegrees=45
  twistMotion=Free
 }
 {
  bone1Name=neck_01
  bone2Name=spine_05
  swing1LimitDegrees=45
  swing1Motion=Free
  swing2LimitDegrees=45
  swing2Motion=Free
  twistLimitDegrees=45
  twistMotion=Free
 }
 {
  bone1Name=dyn_shoulderPad_l_01
  bone2Name=clavicle_l
  swing1LimitDegrees=45
  swing1Motion=Free
  swing2LimitDegrees=45
  swing2Motion=Free
  twistLimitDegrees=45
  twistMotion=Free
 }
 {
  bone1Name=dyn_shoulderPad_l_02
  bone2Name=clavicle_l
  swing1LimitDegrees=45
  swing1Motion=Free
  swing2LimitDegrees=45
  swing2Motion=Free
  twistLimitDegrees=45
  twistMotion=Free
 }
 {
  bone1Name=dyn_shoulderPad_r_01
  bone2Name=clavicle_r
  swing1LimitDegrees=45
  swing1Motion=Free
  swing2LimitDegrees=45
  swing2Motion=Free
  twistLimitDegrees=45
  twistMotion=Free
 }
 {
  bone1Name=dyn_shoulderPad_r_02
  bone2Name=clavicle_r
  swing1LimitDegrees=45
  swing1Motion=Free
  swing2LimitDegrees=45
  swing2Motion=Free
  twistLimitDegrees=45
  twistMotion=Free
 }
]
