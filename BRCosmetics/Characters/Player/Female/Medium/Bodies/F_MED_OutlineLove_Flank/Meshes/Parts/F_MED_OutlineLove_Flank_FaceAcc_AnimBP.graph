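# Anim Blueprint for the OutlineLove Flank female character's FaceAcc part, driving hair and accessory physics.
# The attached-parent pose, blended with body-mesh shoulder bones, feeds a Control Rig, then spine_05-relative RigidBody simulations using parameter-asset gravity.
# Without backbling or cape, two physics assets simulate in series.
# With either, only FHair simulates, then constraints and additive offsets adjust hair-tail bones to head and spine_05.
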AnimBlueprint /BRCosmetics/Characters/Player/Female/Medium/Bodies/F_MED_OutlineLove_Flank/Meshes/Parts/F_MED_OutlineLove_Flank_FaceAcc_AnimBP
target_skeleton=&/F_MED_OutlineLove_Flank_FaceAcc_Skeleton
graphs [
 AnimGraph {nodes [
   Output Pose(Result=Blend Poses by bool(
     BlendPose_0=Use cached pose 'Backpack/Cape setup'()
     BlendPose_1=Use cached pose 'Dynamics setup'()
     bActiveValue=Get is Wearing Backbling or Cape() -> (is Wearing Backbling or Cape:bool)
    ) {blendType=Linear}) {layerGroup=None}
   n1 = Layered blend per bone(
    BasePose=Copy Pose From Mesh() {bUseAttachedParent=true}
    BlendPoses_0=Copy Pose From Mesh(SourceMeshComponent=Get Body_Mesh()) {bUseAttachedParent=true}
   ) {layerSetup [{branchFilters [{boneName=clavicle_l}, {boneName=clavicle_r}, {boneName=upperarm_r}, {boneName=upperarm_l}]}]}
   Save cached pose 'Input Pose'(Pose=@n1) {cacheName=Input Pose}
   n2 = Get GravityOverride Params() -> (GravityOverride Params:Vector)
   n3 = RigidBody!(ComponentPose=Local To Component(LocalPose=Use cached pose 'Control rig'()), Alpha, OverrideWorldGravity:Vector=@n2, ExternalForce) -> (Pose) {
    bAllowPoseScalePropagation=false
    bClampLinearTranslationLimitToRefPose=true
    bOverrideWorldGravity=true
    overridePhysicsAsset=&/F_MED_OutlineLove_Flank_FaceAcc_Physics
    simulationSpace=BaseBoneSpace
    baseBoneRef {boneName=spine_05}
    componentAppliedLinearAccClamp=<0 0 0>
    simSpaceSettings {
     dampingAlpha=0
     maxAngularAcceleration=12
     maxAngularVelocity=8
     maxLinearAcceleration=1200
     maxLinearVelocity=700
     velocityScaleZ=0.4
     worldAlpha=1
     externalLinearDragV=<0.2 0.2 0.2>
    }
   }
   n4 = RigidBody(ComponentPose=@n3, OverrideWorldGravity=@n2) {
    bAllowPoseScalePropagation=false
    bClampLinearTranslationLimitToRefPose=true
    bOverrideWorldGravity=true
    overridePhysicsAsset=&/F_MED_OutlineLove_Flank_FaceAcc_FHair_Physics
    simulationSpace=BaseBoneSpace
    baseBoneRef {boneName=spine_05}
    componentAppliedLinearAccClamp=<0 0 0>
    simSpaceSettings {
     dampingAlpha=0
     maxAngularAcceleration=16
     maxAngularVelocity=8
     maxLinearAcceleration=1200
     maxLinearVelocity=700
     velocityScaleZ=0.4
     worldAlpha=1
     externalLinearDragV=<0.2 0.2 0.2>
    }
   }
   Save cached pose 'Dynamics setup'(Pose=Component To Local(ComponentPose=@n4)) {cacheName=Dynamics setup}
   n5 = Control Rig(Source=Use cached pose 'Input Pose'()) {controlRigAssetReference {blueprintRigClass=&/F_MED_OutlineLove_Flank_FaceAcc_CtrlRig.F_MED_OutlineLove_Flank_FaceAcc_CtrlRig_C}}
   Save cached pose 'Control rig'(Pose=@n5) {cacheName=Control rig}
   n6 = RigidBody(ComponentPose=Local To Component(LocalPose=Use cached pose 'Control rig'()), OverrideWorldGravity=@n2) {
    bAllowPoseScalePropagation=false
    bClampLinearTranslationLimitToRefPose=true
    bOverrideWorldGravity=true
    overridePhysicsAsset=&F_MED_OutlineLove_Flank_FaceAcc_FHair_Physics
    simulationSpace=BaseBoneSpace
    baseBoneRef {boneName=spine_05}
    componentAppliedLinearAccClamp=<0 0 0>
    simSpaceSettings {
     dampingAlpha=0
     maxAngularAcceleration=16
     maxAngularVelocity=8
     maxLinearAcceleration=1200
     maxLinearVelocity=700
     velocityScaleZ=0.4
     worldAlpha=1
     externalLinearDragV=<0.2 0.2 0.2>
    }
   }
   n7 = "Constraint - base_hairTail_01 (PP)"!(ComponentPose=@n6, Alpha=1, ConstraintWeights_0=0.3, ConstraintWeights_1:float=0.7) -> (Pose) {
    boneToModify {boneName=base_hairTail_01}
    constraintSetup [{targetBone {boneName=head}}, {targetBone {boneName=spine_05}}]
    constraintWeights [1,1]
   }
   n8 = "Constraint - dyn_hairTail_02 (PP)"!(ComponentPose=@n7, Alpha=1, ConstraintWeights_0=0.2, ConstraintWeights_1:float=0.8) -> (Pose) {
    boneToModify {boneName=dyn_hairTail_02}
    constraintSetup [{targetBone {boneName=head}}, {targetBone {boneName=spine_05}}]
    constraintWeights [1,1]
   }
   n9 = "Constraint - dyn_hairTail_03 (PP)"!(ComponentPose=@n8, Alpha=1, ConstraintWeights_0=0, ConstraintWeights_1:float=1) -> (Pose) {
    boneToModify {boneName=dyn_hairTail_03}
    constraintSetup [{targetBone {boneName=head}}, {targetBone {boneName=spine_05}}]
    constraintWeights [1,1]
   }
   n10 = "Constraint - dyn_hairTail_04 (PP)"!(ComponentPose=@n9, Alpha=1, ConstraintWeights_0=0, ConstraintWeights_1:float=1) -> (Pose) {
    boneToModify {boneName=dyn_hairTail_04}
    constraintSetup [{targetBone {boneName=head}}, {targetBone {boneName=spine_05}}]
    constraintWeights [1,1]
   }
   n11 = "Constraint - dyn_hairTail_05 (PP)"!(ComponentPose=@n10, Alpha=1, ConstraintWeights_0=0, ConstraintWeights_1:float=1) -> (Pose) {
    boneToModify {boneName=dyn_hairTail_05}
    constraintSetup [{targetBone {boneName=head}}, {targetBone {boneName=spine_05}}]
    constraintWeights [1,1]
   }
   n12 = "Transform (Modify) Bone - Bone: dyn_hairTail_03"(ComponentPose=@n11, Alpha=1, Translation="0,-2,0", Rotation="0,0,0", Scale="1,1,1") {
    rotationMode=BMM_Additive
    rotationSpace=BCS_BoneSpace
    translationMode=BMM_Additive
    translationSpace=BCS_BoneSpace
    boneToModify {boneName=dyn_hairTail_03}
    translation {y=-2}
   }
   n13 = "Transform (Modify) Bone - Bone: dyn_hairTail_04"(ComponentPose=@n12, Alpha=1, Translation="-0.000,-1.400,-0.000", Rotation="0,0,0", Scale="1,1,1") {
    rotationMode=BMM_Additive
    rotationSpace=BCS_BoneSpace
    translationMode=BMM_Additive
    translationSpace=BCS_BoneSpace
    boneToModify {boneName=dyn_hairTail_04}
    translation {y=-1.4}
   }
   Save cached pose 'Backpack/Cape setup'(Pose=Component To Local(ComponentPose=@n13)) {cacheName=Backpack/Cape setup}
  ]}
 EventGraph {nodes [
   Event OnCharacterCustomizationLoaded()
   > n1 = Cast To PlayerPawn_Athena(Object=TryGetPawnOwner())
   > Set Body_Mesh(Body_Mesh=GetSkeletalMeshForPartType(self=@n1))
   Event BlueprintInitializeAnimation()
   > Set Skel Mesh(Skel Mesh=GetOwnerSkeletalMeshComponent())
   Event BlueprintUpdateAnimation()
   > n2 = Sequence!(execute) -> (then_0, then_1, then_2:Exec)
   > n3 = "Gravity Override State Function (Parameter Asset) (Nativized)"(
    CharacterPartAnimInstance=Self-Reference()
    GameTimeInSeconds=GetGameTimeInSeconds()
    Parameters=&/F_MED_OutlineLove_Flank_FA_Parameters
   )
   > Set GravityOverride Params(GravityOverride Params:Vector=@n3) -> (Output_Get:Vector)
   > Set RBAN Settings Out(RBAN Settings Out=@n3.OutRigidBodyAnimNodeParameters)
   Set Hair Up(
    execute=@n2.then_2
    Hair Up:double=GetAlphaFromReferenceTransformDelta(SkeletalMeshComponent=Get Skel Mesh(), BoneName=spine_05, BaseName=head, InputRange=(X=0,Y=70), TwistAxis="0,0,1")
   ) -> (Output_Get:double)
   > n4 = Branch(Condition=OR Boolean(A=Get bIsSkydiveFloating(), B=Get bIsSkydiveDiveMode()))
   > Set Gravity Override BackHair(Gravity Override BackHair:Vector="0,0,200") -> (Output_Get:Vector)
   Set Gravity Override BackHair(execute=@n4.else, Gravity Override BackHair="0,0,-980")
   n5 = CheckForBackbling()
   Set is Wearing Backbling(
    execute=@n2.then_1
    is Wearing Backbling:bool=AND Boolean!(A=@n5, B=NOT Boolean(A=@n5.bIsBackblingACape), C:bool=NOT Boolean(A=@n5.bIsBackblingLong)) -> (ReturnValue)
   ) -> (Output_Get:bool)
   > Set is Wearing Cape(is Wearing Cape:bool=OR Boolean(A=@n5.bIsBackblingACape, B=@n5.bIsBackblingLong)) -> (Output_Get:bool)
   n6 = OR Boolean(A=Get is Wearing Backbling() -> (is Wearing Backbling:bool), B=Get is Wearing Cape() -> (is Wearing Cape:bool))
   > Set is Wearing Backbling or Cape(is Wearing Backbling or Cape:bool=@n6) -> (Output_Get:bool)
   > Set is Wearing Backbling or Cape Alpha(is Wearing Backbling or Cape Alpha:double="To Float (Boolean)"(InBool=@n6)) -> (Output_Get:double)
   > n7 = Branch(Condition=Get bIsSkydiveFloating())
   > Set Gravity Override(Gravity Override="-500,-500,500")
   n8 = Branch(execute=@n7.else, Condition=Get bIsSkydiveDiveMode())
   > Set Gravity Override(Gravity Override="0,0,500")
   Set Gravity Override(execute=@n8.else, Gravity Override="0,0,-980")
   Set Sim Space Settings Params(Sim Space Settings Params:Sim Space Settings Structure) -> (Output_Get:Sim Space Settings Structure)
  ]}
]
variables [
 Body_Mesh {type=SkeletalMeshComponent}
 Gravity Override {type=Vector}
 Gravity Override BackHair {type=Vector}
 GravityOverride Params {type=Vector}
 Hair Up {type=number}
 is Wearing Backbling {type=boolean}
 is Wearing Backbling or Cape {type=boolean}
 is Wearing Backbling or Cape Alpha {type=number}
 is Wearing Cape {type=boolean}
 RBAN Settings Out {type=FortRigidBodyAnimNodeParameters}
 Sim Space Settings Params {
  type=SimSpaceSettings
  default {
   dampingAlpha=1
   maxAngularAcceleration=10000
   maxAngularVelocity=10000
   maxLinearAcceleration=10000
   maxLinearVelocity=10000
   velocityScaleZ=1
   worldAlpha=0
   externalLinearDragV=<0 0 0>
  }
 }
 Skel Mesh {type=SkeletalMeshComponent}
]
