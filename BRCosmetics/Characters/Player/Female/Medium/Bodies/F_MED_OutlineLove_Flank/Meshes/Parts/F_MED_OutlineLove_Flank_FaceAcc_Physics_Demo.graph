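# Physics asset for the face accessory mesh of the female medium OutlineLove Flank body (Demo), covering a dynamic hair tail and upper body.
# Five dyn_hairTail capsules form a chain from base_hairTail_01, with identical swing and twist limits on each joint.
# Spine and upper-arm capsules give the torso volume, while neck and tail-base capsules are tiny.
# Free angular constraints link these body bones and attach the tail base to neck_02.
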PhysicsAsset /BRCosmetics/Characters/Player/Female/Medium/Bodies/F_MED_OutlineLove_Flank/Meshes/Parts/F_MED_OutlineLove_Flank_FaceAcc_Physics_Demo
bodies {
 base_hairTail_01 [{shapeType=Capsule, length=0.505, radius=0.505, shapeName=base_hairTail_01_capsule, rotation=<rot 0 -90 0>}]
 dyn_hairTail_01 [{shapeType=Capsule, length=3.3586, radius=5.3329, shapeName=dyn_hairTail_01_capsule, center=<5.675 2.5455 0>, rotation=<rot 90 -4.7439 0>}]
 dyn_hairTail_02 [{shapeType=Capsule, length=3.0449, radius=5.3329, shapeName=dyn_hairTail_02_capsule, center=<4.1287 2.4805 0>, rotation=<rot 90 0 0>}]
 dyn_hairTail_03 [{shapeType=Capsule, length=3.0449, radius=5.3329, shapeName=dyn_hairTail_03_capsule, center=<4.4148 1.8584 0>, rotation=<rot 90 0 0>}]
 dyn_hairTail_04 [{shapeType=Capsule, length=3.0449, radius=5.3329, shapeName=dyn_hairTail_04_capsule, center=<4.1002 2.3916 0>, rotation=<rot 90 0 0>}]
 dyn_hairTail_05 [{shapeType=Capsule, length=3.0449, radius=5.3329, shapeName=dyn_hairTail_05_capsule, center=<2.6892 2.1442 0>, rotation=<rot 90 0 0>}]
 neck_01 [{shapeType=Capsule, length=0.505, radius=0.505, shapeName=neck_01_capsule, rotation=<rot 0 -90 0>}]
 neck_02 [{shapeType=Capsule, length=0.505, radius=0.505, shapeName=neck_02_capsule, rotation=<rot 0 -90 0>}]
 spine_01 [{shapeType=Capsule, length=10.9972, radius=7.6234, shapeName=spine_01_capsule, center=<3.9724 0.1714 0>, rotation=<rot 0 -90 0>}]
 spine_02 [{shapeType=Capsule, length=15.2591, radius=7.0453, shapeName=spine_02_capsule, center=<5.9866 0 0>, rotation=<rot 0 -90 0>}]
 spine_03 [{shapeType=Capsule, length=19.1431, radius=8.6402, shapeName=spine_03_capsule, center=<5.9866 0.7741 0>, rotation=<rot 0 -90 0>}]
 spine_04 [{shapeType=Capsule, length=27.4513, radius=9.8801, shapeName=spine_04_capsule, center=<3.9105 0.6497 0>}]
 spine_05 [{shapeType=Capsule, length=34.7305, radius=9.4656, shapeName=spine_05_capsule, center=<3.328 0.6578 0>, rotation=<rot 0 -90 0>}]
 upperarm_l [{shapeType=Capsule, length=25.9882, radius=6.0118, shapeName=upperarm_l_capsule, center=<14.2936 0 -0.5693>, rotation=<rot 90 180 0>}]
 upperarm_r [{shapeType=Capsule, length=25.9882, radius=6.0118, shapeName=upperarm_l_capsule, center=<-14.2936 0 0.5832>, rotation=<rot 90 180 0>}]
}
constraints [
 {
  bone1Name=base_hairTail_01
  bone2Name=neck_02
  swing1LimitDegrees=45
  swing1Motion=Free
  swing2LimitDegrees=45
  swing2Motion=Free
  twistLimitDegrees=45
  twistMotion=Free
 }
 {bone1Name=dyn_hairTail_01, bone2Name=base_hairTail_01, swing1LimitDegrees=25, swing2LimitDegrees=50, twistLimitDegrees=10}
 {bone1Name=dyn_hairTail_02, bone2Name=dyn_hairTail_01, swing1LimitDegrees=25, swing2LimitDegrees=50, twistLimitDegrees=10}
 {bone1Name=dyn_hairTail_03, bone2Name=dyn_hairTail_02, swing1LimitDegrees=25, swing2LimitDegrees=50, twistLimitDegrees=10}
 {bone1Name=dyn_hairTail_04, bone2Name=dyn_hairTail_03, swing1LimitDegrees=25, swing2LimitDegrees=50, twistLimitDegrees=10}
 {bone1Name=dyn_hairTail_05, bone2Name=dyn_hairTail_04, swing1LimitDegrees=25, swing2LimitDegrees=50, twistLimitDegrees=10}
 {
  bone1Name=neck_02
  bone2Name=neck_01
  swing1LimitDegrees=45
  swing1Motion=Free
  swing2LimitDegrees=45
  swing2Motion=Free
  twistLimitDegrees=45
  twistMotion=Free
 }
 {
  bone1Name=neck_01
  bone2Name=spine_05
  swing1LimitDegrees=45
  swing1Motion=Free
  swing2LimitDegrees=45
  swing2Motion=Free
  twistLimitDegrees=45
  twistMotion=Free
 }
 {
  bone1Name=spine_05
  bone2Name=spine_04
  swing1LimitDegrees=45
  swing1Motion=Free
  swing2LimitDegrees=45
  swing2Motion=Free
  twistLimitDegrees=45
  twistMotion=Free
 }
 {
  bone1Name=spine_04
  bone2Name=spine_03
  swing1LimitDegrees=45
  swing1Motion=Free
  swing2LimitDegrees=45
  swing2Motion=Free
  twistLimitDegrees=45
  twistMotion=Free
 }
 {
  bone1Name=spine_03
  bone2Name=spine_02
  swing1LimitDegrees=45
  swing1Motion=Free
  swing2LimitDegrees=45
  swing2Motion=Free
  twistLimitDegrees=45
  twistMotion=Free
 }
 {
  bone1Name=spine_02
  bone2Name=spine_01
  swing1LimitDegrees=45
  swing1Motion=Free
  swing2LimitDegrees=45
  swing2Motion=Free
  twistLimitDegrees=45
  twistMotion=Free
 }
 {
  bone1Name=upperarm_l
  bone2Name=spine_05
  swing1LimitDegrees=45
  swing1Motion=Free
  swing2LimitDegrees=45
  swing2Motion=Free
  twistLimitDegrees=45
  twistMotion=Free
 }
 {
  bone1Name=upperarm_r
  bone2Name=spine_05
  swing1LimitDegrees=45
  swing1Motion=Free
  swing2LimitDegrees=45
  swing2Motion=Free
  twistLimitDegrees=45
  twistMotion=Free
 }
]
